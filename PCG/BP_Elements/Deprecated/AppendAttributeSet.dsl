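; Deprecated PCG Blueprint element that merges multiple input attribute sets into a single new attribute set.
; Execution creates a new param data object with one metadata entry, then loops over the input params, adding each one's attributes and copying the values at key 0 into that entry.
; The node is colored red and titled to point to the native Add Attributes node as its replacement.
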(fn ExecuteWithContext
  (params (InContext :type PCGContext%20Structure%20%28by%20ref%29) (Input :type PCGData%20Collection%20Structure%20%28by%20ref%29))
  (bind _outmetadata (LocalVariables|GetOutMetadata))
  (bind _returnvalue (Game|ConstructObjectfromClass "/Script/PCG.PCGParamData"))
  (bind _returnvalue_2 (Metadata|MutableMetadata _returnvalue))
  (bind _output_get (LocalVariables|SetOutMetadata _returnvalue_2))
  (bind _returnvalue_3 (PCG|Metadata|AddEntry _output_get -1))
  (LocalVariables|SetOutMetaDataEntry _returnvalue_3)
  (bind _returnvalue_4 (Data|GetParams Input))
  (for _array_element _returnvalue_4
    (bind _aspcgattribute_set (Utilities|Casting|CastToPCGParamData (PCG|TaggedData|BreakTaggedData _array_element)))
    (bind _returnvalue_5 (Metadata|MutableMetadata _aspcgattribute_set))
    (PCG|Metadata|AddAttributes _outmetadata _returnvalue_5)
    (PCG|Metadata|SetAttributesbyKey _outmetadata (LocalVariables|GetOutMetaDataEntry) _returnvalue_5 0))
  (return (Utilities|Struct|MakePCGDataCollection (Utilities|Array|MakeArray (PCG|TaggedData|MakeTaggedData _returnvalue)))))

(fn NodeColorOverride
  (return "(R=1.000000,G=0.000000,B=0.000000,A=1.000000)"))

(fn NodeTitleOverride
  (return "DEPRECATED, Use Native Add Attributes"))
