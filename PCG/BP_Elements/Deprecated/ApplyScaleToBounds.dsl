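; A deprecated PCG Blueprint element that bakes each point's scale into its bounds; its red node and title call for replacement with the native node.
; Execution converts each spatial input on the first custom input pin to point data, runs a point loop, and copies the target actor to the output.
; The loop body multiplies bounds min and max by scale and rebuilds the transform from location and rotation, leaving scale at default.
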(fn ExecuteWithContext
  (params (InContext :type PCGContext%20Structure%20%28by%20ref%29) (Input :type PCGData%20Collection%20Structure%20%28by%20ref%29))
  (bind _result (Utilities|Set|ToArray (PCG|Input&Output|CustomInputLabels self)))
  (bind _returnvalue (Data|GetInputsbyPinLabel Input (Utilities|Array|Get(aref) _result)))
  (for _array_element _returnvalue
    (bind _aspcgspatial_data (Utilities|Casting|CastToPCGSpatialData (PCG|TaggedData|BreakTaggedData _array_element)))
    (bind _outdata (PCG|FlowControl|PointLoop InContext (SpatialData|ToPointDatawithContext _aspcgspatial_data InContext) 0))
    (Variables|/Script/PCG.PCGSpatialData|SetTargetActor (Variables|/Script/PCG.PCGSpatialData|GetTargetActor _aspcgspatial_data) _outdata)
    (Utilities|Array|Add (LocalVariables|GetOutTaggedDatas) (PCG|TaggedData|MakeTaggedData _outdata)))
  (return (Utilities|Struct|MakePCGDataCollection (LocalVariables|GetOutTaggedDatas))))

(fn NodeColorOverride
  (return "(R=1.000000,G=0.000000,B=0.000000,A=1.000000)"))

(fn NodeTitleOverride
  (return "ApplyScaleToBounds: Deprecated, Replace BP with Native node"))

(fn PointLoopBody
  (params (InContext :type PCGContext%20Structure%20%28by%20ref%29) (InData :type PCGPoint%20Data%20Object%20Reference) (InPoint :type PCGPoint%20Structure%20%28by%20ref%29) (OutMetadata :type PCGMetadata%20Object%20Reference) (Iteration :type Integer64))
  (bind (_transform _boundsmin _boundsmax) (Utilities|Struct|BreakPCGPoint InPoint))
  (bind _structout (Utilities|Struct|SetmembersinPCGPoint InPoint (Math|Transform|MakeTransform (.location _transform) (.rotation _transform)) (* _boundsmin (.scale _transform)) (* (.scale _transform) _boundsmax)))
  (return _structout true))
