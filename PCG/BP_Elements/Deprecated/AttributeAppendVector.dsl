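; Deprecated PCG Blueprint node, titled "Append" plus the type, that combines float attributes from separate input pins into one float, Vector or Vector4 point attribute.
; It renames the last attribute on pin 0 and copies those from the other pins into that same input metadata, modifying it in place.
; A point loop over pin 0 then writes the combined value to new point data, with optional debug printing for Vector4.
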(fn ExecuteWithContext
  (params (InContext :type PCGContext%20Structure%20%28by%20ref%29) (Input :type PCGData%20Collection%20Structure%20%28by%20ref%29))
  (bind _outputattributename (Variables|Default|GetOutputAttributeName))
  (LocalVariables|SetContext InContext)
  (Utilities|FlowControl|Sequence
    (:then_0
      ;[1 1 1] Get Metadata by Pin
      (bind _result (Utilities|Set|ToArray (PCG|Input&Output|CustomInputLabels self)))
      (Utilities|FlowControl|Sequence
        (:then_0
          (bind _returnvalue (Data|GetInputsbyPinLabel Input (Utilities|Array|Get(aref) _result)))
          (bind _aspcgspatial_data (Utilities|Casting|CastToPCGSpatialData (PCG|TaggedData|BreakTaggedData (Utilities|Array|Get(aref) _returnvalue))))
          (bind _output_get (LocalVariables|SetSpatialDataPin0 _aspcgspatial_data))
          (bind _returnvalue_2 (Metadata|MutableMetadata _output_get))
          (LocalVariables|SetMetadataPin0 _returnvalue_2))
        (:then_1
          (bind _returnvalue_3 (Data|GetInputsbyPinLabel Input (Utilities|Array|Get(aref) _result 1)))
          (bind _aspcgspatial_data_2 (Utilities|Casting|CastToPCGSpatialData (PCG|TaggedData|BreakTaggedData (Utilities|Array|Get(aref) _returnvalue_3))))
          (bind _returnvalue_4 (Metadata|MutableMetadata _aspcgspatial_data_2))
          (LocalVariables|SetMetadataPin1 _returnvalue_4))
        (:then_2
          (bind _returnvalue_5 (Data|GetInputsbyPinLabel Input (Utilities|Array|Get(aref) _result 2)))
          (bind _aspcgspatial_data_3 (Utilities|Casting|CastToPCGSpatialData (PCG|TaggedData|BreakTaggedData (Utilities|Array|Get(aref) _returnvalue_5))))
          (bind _returnvalue_6 (Metadata|MutableMetadata _aspcgspatial_data_3))
          (LocalVariables|SetMetadataPin2 _returnvalue_6))
        (:then_3
          (if (Utilities|Enum|Equal(Enum) (Variables|Default|GetAttribute%20Type) "NewEnumerator2")
            (bind _returnvalue_7 (Data|GetInputsbyPinLabel Input (Utilities|Array|Get(aref) _result 3)))
            (bind _aspcgspatial_data_4 (Utilities|Casting|CastToPCGSpatialData (PCG|TaggedData|BreakTaggedData (Utilities|Array|Get(aref) _returnvalue_7))))
            (bind _returnvalue_8 (Metadata|MutableMetadata _aspcgspatial_data_4))
            (LocalVariables|SetMetadataPin3 _returnvalue_8)))))
    (:then_1
      ;[1 1 1] Copy all Metadata to MetadataPin0
      (PCG|Metadata|RenameAttribute (LocalVariables|GetMetadataPin0) (Utilities|Array|Get(acopy) (PCG|Metadata|GetAttributes (LocalVariables|GetMetadataPin0)) (Utilities|Array|LastIndex (PCG|Metadata|GetAttributes (LocalVariables|GetMetadataPin0)))) (Variables|Hidden|GetInputAttributeName0))
      (PCG|Metadata|Advanced|CopyAttribute (LocalVariables|GetMetadataPin0) (LocalVariables|GetMetadataPin1) (Utilities|Array|Get(acopy) (PCG|Metadata|GetAttributes (LocalVariables|GetMetadataPin1)) (Utilities|Array|LastIndex (PCG|Metadata|GetAttributes (LocalVariables|GetMetadataPin1)))) (Variables|Hidden|GetInputAttributeName1))
      (PCG|Metadata|Advanced|CopyAttribute (LocalVariables|GetMetadataPin0) (LocalVariables|GetMetadataPin2) (Utilities|Array|Get(acopy) (PCG|Metadata|GetAttributes (LocalVariables|GetMetadataPin2)) (Utilities|Array|LastIndex (PCG|Metadata|GetAttributes (LocalVariables|GetMetadataPin2)))) (Variables|Hidden|GetInputAttributeName2))
      (if (Utilities|Enum|Equal(Enum) (Variables|Default|GetAttribute%20Type) "NewEnumerator2")
        (PCG|Metadata|Advanced|CopyAttribute (LocalVariables|GetMetadataPin0) (LocalVariables|GetMetadataPin3) (Utilities|Array|Get(acopy) (PCG|Metadata|GetAttributes (LocalVariables|GetMetadataPin3)) (Utilities|Array|LastIndex (PCG|Metadata|GetAttributes (LocalVariables|GetMetadataPin3)))) (Variables|Hidden|GetInputAttributeName3))))
    (:then_2
      (bind _returnvalue_9 (Game|ConstructObjectfromClass "/Script/PCG.PCGPointData"))
      (SpatialData|InitializefromData _returnvalue_9 (LocalVariables|GetSpatialDataPin0) (LocalVariables|GetMetadataPin0) false true)
      (bind _returnvalue_10 (Metadata|MutableMetadata _returnvalue_9))
      (switch Utilities|FlowControl|Switch|SwitchonPCGAttributeTypes (Variables|Default|GetAttribute%20Type)
        (:NewEnumerator0
          (PCG|Metadata|CreateFloatAttribute _returnvalue_10 _outputattributename 0.0 true true))
        (:NewEnumerator1
          (PCG|Metadata|CreateVectorAttribute _returnvalue_10 _outputattributename (0.0 0.0 0.0) true true))
        (:NewEnumerator2
          (PCG|Metadata|CreateVector4Attribute _returnvalue_10 _outputattributename (0.0 0.0 0.0 0.0) true true))))
    (:then_3
      (bind _outdata (PCG|FlowControl|PointLoop (LocalVariables|GetContext) (SpatialData|ToPointDatawithContext (LocalVariables|GetSpatialDataPin0) (LocalVariables|GetContext)) _returnvalue_9)))
    (:then_4
      (Utilities|Array|Add (LocalVariables|GetOutData) (PCG|TaggedData|MakeTaggedData _outdata)))
    (:then_5
      (return (Utilities|Struct|MakePCGDataCollection (LocalVariables|GetOutData))))))

(fn NodeColorOverride
  (return "(R=0.949020,G=0.949020,B=0.949020,A=1.000000)"))

(fn NodeTitleOverride
  (return (Utilities|String|StringToName (Utilities|String|Append "Append " (Utilities|String|EnumtoString (Variables|Default|GetAttribute%20Type))))))

(fn PointLoopBody
  (params (InContext :type PCGContext%20Structure%20%28by%20ref%29) (InData :type PCGPoint%20Data%20Object%20Reference) (InPoint :type PCGPoint%20Structure%20%28by%20ref%29) (OutMetadata :type PCGMetadata%20Object%20Reference) (Iteration :type Integer64))
  (bind (_transform _density _boundsmin _boundsmax _color _steepness _seed) (Utilities|Struct|BreakPCGPoint (LocalVariables|GetSrcPoint)))
  (bind _inputattributename3 (Variables|Hidden|GetInputAttributeName3))
  (bind _inputattributename0 (Variables|Hidden|GetInputAttributeName0))
  (bind _outputattributename (Variables|Default|GetOutputAttributeName))
  (bind _inputattributename1 (Variables|Hidden|GetInputAttributeName1))
  (bind _inputattributename2 (Variables|Hidden|GetInputAttributeName2))
  (LocalVariables|SetSrcPoint InPoint)
  (LocalVariables|SetMetadata OutMetadata)
  (LocalVariables|SetDestPoint (Utilities|Struct|MakePCGPoint _transform _density _boundsmin _boundsmax _color _steepness _seed))
  (bind _returnvalue (Metadata|MutableMetadata InData))
  (switch Utilities|FlowControl|Switch|SwitchonPCGAttributeTypes (Variables|Default|GetAttribute%20Type)
    (:NewEnumerator0
      (bind _returnvalue_2 (PCG|Metadata|GetFloatAttribute (LocalVariables|GetSrcPoint) _returnvalue (Variables|Hidden|GetInputAttributeName0)))
      (PCG|Metadata|SetFloatAttribute (LocalVariables|GetDestPoint) (LocalVariables|GetMetadata) (Variables|Default|GetOutputAttributeName) _returnvalue_2))
    (:NewEnumerator1
      (bind _returnvalue_3 (PCG|Metadata|GetFloatAttribute (LocalVariables|GetSrcPoint) _returnvalue (Variables|Hidden|GetInputAttributeName0)))
      (bind _returnvalue_4 (PCG|Metadata|GetFloatAttribute (LocalVariables|GetSrcPoint) _returnvalue (Variables|Hidden|GetInputAttributeName1)))
      (bind _returnvalue_5 (PCG|Metadata|GetFloatAttribute (LocalVariables|GetSrcPoint) _returnvalue (Variables|Hidden|GetInputAttributeName2)))
      (PCG|Metadata|SetVectorAttribute (LocalVariables|GetDestPoint) (LocalVariables|GetMetadata) (Variables|Default|GetOutputAttributeName) (Math|Vector|MakeVector _returnvalue_3 _returnvalue_4 _returnvalue_5)))
    (:NewEnumerator2
      (bind _returnvalue_6 (PCG|Metadata|GetFloatAttribute (LocalVariables|GetSrcPoint) _returnvalue _inputattributename0))
      (bind _returnvalue_7 (PCG|Metadata|GetFloatAttribute (LocalVariables|GetSrcPoint) _returnvalue _inputattributename1))
      (bind _returnvalue_8 (PCG|Metadata|GetFloatAttribute (LocalVariables|GetSrcPoint) _returnvalue _inputattributename2))
      (bind _returnvalue_9 (PCG|Metadata|GetFloatAttribute (LocalVariables|GetSrcPoint) _returnvalue _inputattributename3))
      (if (Variables|Hidden|GetDOPRINT)
        (Development|PrintString (Utilities|String|ToString(Name) _inputattributename0) true true "(R=0.000000,G=1.000000,B=0.440307,A=1.000000)")
        (Development|PrintString (Utilities|String|ToString(Name) _inputattributename1) true true "(R=0.000000,G=1.000000,B=0.351925,A=1.000000)")
        (Development|PrintString (Utilities|String|ToString(Name) _inputattributename2) true true "(R=0.000000,G=1.000000,B=0.316515,A=1.000000)")
        (Development|PrintString (Utilities|String|ToString(Name) _inputattributename3) true true "(R=0.000000,G=1.000000,B=0.460876,A=1.000000)")
        (Development|PrintString (Utilities|String|ToString(Float) _returnvalue_6) true true "(R=0.000000,G=1.000000,B=0.448472,A=1.000000)")
        (Development|PrintString (Utilities|String|ToString(Float) _returnvalue_7) true true "(R=0.000000,G=1.000000,B=0.250000,A=1.000000)")
        (Development|PrintString (Utilities|String|ToString(Float) _returnvalue_8) true true "(R=0.000000,G=1.000000,B=0.391900,A=1.000000)")
        (Development|PrintString (Utilities|String|ToString(Float) _returnvalue_9) true true "(R=0.000000,G=1.000000,B=0.403831,A=1.000000)")
        (Development|PrintString (Utilities|String|ToString(Name) _outputattributename) true true "(R=0.000000,G=1.000000,B=0.404102,A=1.000000)"))
      (PCG|Metadata|SetVector4Attribute (LocalVariables|GetDestPoint) (LocalVariables|GetMetadata) _outputattributename (Math|Vector4|MakeVector4 _returnvalue_6 _returnvalue_7 _returnvalue_8 _returnvalue_9))))
  (return (LocalVariables|GetDestPoint) true))
