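; Deprecated-folder PCG Blueprint element that applies a selectable math operation to Float, Vector or Vector4 point attributes, titled by its mode.
; ExecuteWithContext takes each input's last attribute, creates the output attribute typed from A's first, and loops to the larger point count.
; IterationLoopBody copies A's point and writes the result, falling back to point 0 past either end.
; Vector4 Truncate and Absolute read B; vector Fraction results are truncated.
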(fn ExecuteWithContext
  (params (InContext :type PCGContext%20Structure%20%28by%20ref%29) (Input :type PCGData%20Collection%20Structure%20%28by%20ref%29))
  (bind _outputattributename (Variables|Default|GetOutputAttributeName))
  (bind _outputattributename_2 (Variables|Default|GetOutputAttributeName))
  (bind _context (LocalVariables|GetContext))
  (bind _context_2 (LocalVariables|GetContext))
  (LocalVariables|SetContext InContext)
  (LocalVariables|SetInputCollection Input)
  (if (== _outputattributename "")
    (Variables|Default|SetOutputAttributeName (Utilities|String|StringToName (Utilities|String|Append (Utilities|String|ToString(Name) _outputattributename) (Utilities|String|ToString(Integer) (Math|Random|RandomInteger 65536))))))
  (switch Utilities|FlowControl|Switch|SwitchonPCGAttributeOperationMode (Variables|Default|GetOperationMode)
    (:NewEnumerator0
      _)
    (:NewEnumerator2
      (Variables|Hidden|SetOperationHas2Inputs true :!label _label))
    (:NewEnumerator3
      (goto _label))
    (:NewEnumerator4
      (goto _label))
    (:NewEnumerator5
      (goto _label))
    (:NewEnumerator6
      (goto _label))
    (:NewEnumerator7
      (goto _label))
    (:NewEnumerator8
      _)
    (:NewEnumerator9
      _)
    (:NewEnumerator10
      _)
    (:NewEnumerator11
      _)
    (:NewEnumerator12
      _)
    (:NewEnumerator13
      _))
  (Utilities|FlowControl|Sequence
    (:then_0
      (bind _result (Utilities|Set|ToArray (PCG|Input&Output|CustomInputLabels self)))
      (bind _returnvalue (Data|GetInputsbyPinLabel (LocalVariables|GetInputCollection) (Utilities|Array|Get(aref) _result)))
      (bind _aspcgspatial_data (Utilities|Casting|CastToPCGSpatialData (PCG|TaggedData|BreakTaggedData (Utilities|Array|Get(aref) _returnvalue))))
      (Variables|Hidden|SetInputAttributeNameA (Utilities|Array|Get(acopy) (PCG|Metadata|GetAttributes (Metadata|ConstMetadata _aspcgspatial_data)) (Utilities|Array|LastIndex (PCG|Metadata|GetAttributes (Metadata|ConstMetadata _aspcgspatial_data)))))
      (Variables|Hidden|SetptDataA (SpatialData|ToPointDatawithContext _aspcgspatial_data _context_2))
      (Variables|Hidden|SetptACount (Utilities|Array|Length (SpatialData|GetPoints (SpatialData|ToPointDatawithContext _aspcgspatial_data _context_2)))))
    (:then_1
      (if (Variables|Hidden|GetOperationHas2Inputs)
        (bind _returnvalue_2 (Data|GetInputsbyPinLabel (LocalVariables|GetInputCollection) (Utilities|Array|Get(aref) _result 1)))
        (bind _aspcgspatial_data_2 (Utilities|Casting|CastToPCGSpatialData (PCG|TaggedData|BreakTaggedData (Utilities|Array|Get(aref) _returnvalue_2))))
        (Variables|Hidden|SetInputAttributeNameB (Utilities|Array|Get(acopy) (PCG|Metadata|GetAttributes (Metadata|ConstMetadata _aspcgspatial_data_2)) (Utilities|Array|LastIndex (PCG|Metadata|GetAttributes (Metadata|ConstMetadata _aspcgspatial_data_2)))))
        (Variables|Hidden|SetPtDataB (SpatialData|ToPointDatawithContext _aspcgspatial_data_2 _context))
        (Variables|Hidden|SetptBCount (Utilities|Array|Length (SpatialData|GetPoints (SpatialData|ToPointDatawithContext _aspcgspatial_data_2 _context))))))
    (:then_2
      (bind _returnvalue_3 (Game|ConstructObjectfromClass "/Script/PCG.PCGPointData"))
      (Variables|/Script/PCG.PCGSpatialData|SetTargetActor (Variables|/Script/PCG.PCGSpatialData|GetTargetActor _aspcgspatial_data) _returnvalue_3)
      (bind _returnvalue_4 (Metadata|MutableMetadata _returnvalue_3))
      (switch Utilities|FlowControl|Switch|SwitchonEPCGMetadataTypes (Utilities|Array|Get(aref) (.AttributeTypes (PCG|Metadata|GetAttributes (Metadata|ConstMetadata _aspcgspatial_data))))
        (:Float
          (PCG|Metadata|CreateFloatAttribute _returnvalue_4 _outputattributename_2 0.0 true true))
        (:Vector
          (PCG|Metadata|CreateVectorAttribute _returnvalue_4 _outputattributename_2 (0.0 0.0 0.0) true true))
        (:Vector4
          (PCG|Metadata|CreateVector4Attribute _returnvalue_4 _outputattributename_2 (0.0 0.0 0.0 0.0) true true))))
    (:then_3
      (bind _outdata (PCG|FlowControl|IterationLoop (LocalVariables|GetContext) (Math|Conversions|ToInteger64(Integer) (Math|Integer|Max(Integer) (Variables|Hidden|GetptACount) (Variables|Hidden|GetptBCount))) 0 0 _returnvalue_3)))
    (:then_4
      (return (Utilities|Struct|MakePCGDataCollection (Utilities|Array|MakeArray (PCG|TaggedData|MakeTaggedData _outdata)))))))

(fn IterationLoopBody
  (params (InContext :type PCGContext%20Structure%20%28by%20ref%29) (Iteration :type Integer64) (InA :type PCGSpatial%20Data%20Object%20Reference) (InB :type PCGSpatial%20Data%20Object%20Reference) (OutMetadata :type PCGMetadata%20Object%20Reference))
  (bind (_transform _density _boundsmin _boundsmax _color _steepness _seed) (Utilities|Struct|BreakPCGPoint (LocalVariables|GetInPtA)))
  (bind _inputattributenamea (Variables|Hidden|GetInputAttributeNameA))
  (bind _inmetadataa (LocalVariables|GetInMetadataA))
  (bind (_attributenames _attributetypes) (PCG|Metadata|GetAttributes _inmetadataa))
  (bind _returnvalue (Utilities|Array|FindItem _attributenames _inputattributenamea))
  (bind (_x _y _z _w) (Math|Vector4|BreakVector4 (LocalVariables|GetInMetadataAVector4)))
  (bind (_x_2 _y_2 _z_2 _w_2) (Math|Vector4|BreakVector4 (LocalVariables|GetInMetadataBVector4)))
  (bind _output (Utilities|Array|Get(aref) _attributetypes _returnvalue))
  (LocalVariables|SetOutMeta OutMetadata)
  (LocalVariables|SetInPtA (SpatialData|GetPoint (Variables|Hidden|GetptDataA) (Math|Integer|SelectInt (Math|Conversions|ToInteger(Integer64) Iteration) 0 (< Iteration (Variables|Hidden|GetptACount)))))
  (bind _returnvalue_2 (Metadata|MutableMetadata (Variables|Hidden|GetptDataA)))
  (LocalVariables|SetInMetaDataA _returnvalue_2)
  (LocalVariables|SetAttributeType _output)
  (switch Utilities|FlowControl|Switch|SwitchonEPCGMetadataTypes _output
    (:Float
      (bind _returnvalue_3 (PCG|Metadata|GetFloatAttribute (LocalVariables|GetInPtA) (LocalVariables|GetInMetaDataA) (Variables|Hidden|GetInputAttributeNameA)))
      (LocalVariables|SetInMetadataAFloat _returnvalue_3))
    (:Vector
      (bind _returnvalue_4 (PCG|Metadata|GetVectorAttribute (LocalVariables|GetInPtA) (LocalVariables|GetInMetaDataA) (Variables|Hidden|GetInputAttributeNameA)))
      (LocalVariables|SetInMetadataAVector _returnvalue_4))
    (:Vector4
      (bind _returnvalue_5 (PCG|Metadata|GetVector4Attribute (LocalVariables|GetInPtA) (LocalVariables|GetInMetaDataA) (Variables|Hidden|GetInputAttributeNameA)))
      (LocalVariables|SetInMetadataAVector4 _returnvalue_5)))
  (if (Variables|Hidden|GetOperationHas2Inputs)
    (LocalVariables|SetInPtB (SpatialData|GetPoint (Variables|Hidden|GetPtDataB) (Math|Integer|SelectInt (Math|Conversions|ToInteger(Integer64) Iteration) 0 (< Iteration (Variables|Hidden|GetptBCount)))))
    (bind _returnvalue_6 (Metadata|MutableMetadata (Variables|Hidden|GetPtDataB)))
    (LocalVariables|SetInMetaDataB _returnvalue_6)
    (switch Utilities|FlowControl|Switch|SwitchonEPCGMetadataTypes (LocalVariables|GetAttributeType)
      (:Float
        (bind _returnvalue_7 (PCG|Metadata|GetFloatAttribute (LocalVariables|GetInPtB) (LocalVariables|GetInMetaDataB) (Variables|Hidden|GetInputAttributeNameB)))
        (LocalVariables|SetInMetadataBFloat _returnvalue_7))
      (:Vector
        (bind _returnvalue_8 (PCG|Metadata|GetVectorAttribute (LocalVariables|GetInPtB) (LocalVariables|GetInMetaDataB) (Variables|Hidden|GetInputAttributeNameB)))
        (LocalVariables|SetInMetadataBVector _returnvalue_8))
      (:Vector4
        (bind _returnvalue_9 (PCG|Metadata|GetVector4Attribute (LocalVariables|GetInPtB) (LocalVariables|GetInMetaDataB) (Variables|Hidden|GetInputAttributeNameB)))
        (LocalVariables|SetInMetadataBVector4 _returnvalue_9))))
  (LocalVariables|SetoutPt (Utilities|Struct|MakePCGPoint _transform _density _boundsmin _boundsmax _color _steepness _seed))
  (switch Utilities|FlowControl|Switch|SwitchonEPCGMetadataTypes (LocalVariables|GetAttributeType)
    (:Float
      (PCG|Metadata|SetFloatAttribute (LocalVariables|GetoutPt) (LocalVariables|GetOutMeta) (Variables|Default|GetOutputAttributeName) (Utilities|Select :Index (Variables|Default|GetOperationMode) :NewEnumerator0 (LocalVariables|GetInMetadataAFloat) :NewEnumerator2 (Math|Float|Min(Float) (LocalVariables|GetInMetadataAFloat) (LocalVariables|GetInMetadataBFloat)) :NewEnumerator3 (Math|Float|Max(Float) (LocalVariables|GetInMetadataAFloat) (LocalVariables|GetInMetadataBFloat)) :NewEnumerator4 (+ (LocalVariables|GetInMetadataAFloat) (LocalVariables|GetInMetadataBFloat)) :NewEnumerator5 (- (LocalVariables|GetInMetadataAFloat) (LocalVariables|GetInMetadataBFloat)) :NewEnumerator6 (* (LocalVariables|GetInMetadataAFloat) (LocalVariables|GetInMetadataBFloat)) :NewEnumerator7 (Math|Float|SafeDivide (LocalVariables|GetInMetadataAFloat) (LocalVariables|GetInMetadataBFloat)) :NewEnumerator8 (Math|Conversions|ToFloat(Integer) (Math|Float|Round (LocalVariables|GetInMetadataAFloat))) :NewEnumerator9 (Math|Conversions|ToFloat(Integer) (Math|Float|Truncate (LocalVariables|GetInMetadataAFloat))) :NewEnumerator10 (Math|Conversions|ToFloat(Integer) (Math|Float|Floor (LocalVariables|GetInMetadataAFloat))) :NewEnumerator11 (Math|Conversions|ToFloat(Integer) (Math|Float|Ceil (LocalVariables|GetInMetadataAFloat))) :NewEnumerator12 (Math|Float|Absolute(Float) (LocalVariables|GetInMetadataAFloat)) :NewEnumerator13 (Math|Float|Fraction (LocalVariables|GetInMetadataAFloat)))))
    (:Vector
      (PCG|Metadata|SetVectorAttribute (LocalVariables|GetoutPt) (LocalVariables|GetOutMeta) (Variables|Default|GetOutputAttributeName) (Utilities|Select :Index (Variables|Default|GetOperationMode) :NewEnumerator0 (LocalVariables|GetInMetadataAVector) :NewEnumerator2 (Math|Vector|MakeVector (Math|Float|Min(Float) (.x (LocalVariables|GetInMetadataAVector)) (.x (LocalVariables|GetInMetadataBVector))) (Math|Float|Min(Float) (.y (LocalVariables|GetInMetadataAVector)) (.y (LocalVariables|GetInMetadataBVector))) (Math|Float|Min(Float) (.z (LocalVariables|GetInMetadataAVector)) (.z (LocalVariables|GetInMetadataBVector)))) :NewEnumerator3 (Math|Vector|MakeVector (Math|Float|Max(Float) (.x (LocalVariables|GetInMetadataAVector)) (.x (LocalVariables|GetInMetadataBVector))) (Math|Float|Max(Float) (.y (LocalVariables|GetInMetadataAVector)) (.y (LocalVariables|GetInMetadataBVector))) (Math|Float|Max(Float) (.z (LocalVariables|GetInMetadataAVector)) (.z (LocalVariables|GetInMetadataBVector)))) :NewEnumerator4 (+ (LocalVariables|GetInMetadataAVector) (LocalVariables|GetInMetadataBVector)) :NewEnumerator5 (- (LocalVariables|GetInMetadataAVector) (LocalVariables|GetInMetadataBVector)) :NewEnumerator6 (* (LocalVariables|GetInMetadataAVector) (LocalVariables|GetInMetadataBVector)) :NewEnumerator7 (Math|Vector|MakeVector (Math|Float|SafeDivide (.x (LocalVariables|GetInMetadataAVector)) (.x (LocalVariables|GetInMetadataBVector))) (Math|Float|SafeDivide (.y (LocalVariables|GetInMetadataAVector)) (.y (LocalVariables|GetInMetadataBVector))) (Math|Float|SafeDivide (.z (LocalVariables|GetInMetadataAVector)) (.z (LocalVariables|GetInMetadataBVector)))) :NewEnumerator8 (Math|Vector|MakeVector (Math|Conversions|ToFloat(Integer) (Math|Float|Round (.x (LocalVariables|GetInMetadataAVector)))) (Math|Conversions|ToFloat(Integer) (Math|Float|Round (.y (LocalVariables|GetInMetadataAVector)))) (Math|Conversions|ToFloat(Integer) (Math|Float|Round (.z (LocalVariables|GetInMetadataAVector))))) :NewEnumerator9 (Math|Vector|MakeVector (Math|Conversions|ToFloat(Integer) (Math|Float|Truncate (.x (LocalVariables|GetInMetadataAVector)))) (Math|Conversions|ToFloat(Integer) (Math|Float|Truncate (.y (LocalVariables|GetInMetadataAVector)))) (Math|Conversions|ToFloat(Integer) (Math|Float|Truncate (.z (LocalVariables|GetInMetadataAVector))))) :NewEnumerator10 (Math|Vector|MakeVector (Math|Conversions|ToFloat(Integer) (Math|Float|Floor (.x (LocalVariables|GetInMetadataAVector)))) (Math|Conversions|ToFloat(Integer) (Math|Float|Floor (.y (LocalVariables|GetInMetadataAVector)))) (Math|Conversions|ToFloat(Integer) (Math|Float|Floor (.z (LocalVariables|GetInMetadataAVector))))) :NewEnumerator11 (Math|Vector|MakeVector (Math|Conversions|ToFloat(Integer) (Math|Float|Ceil (.x (LocalVariables|GetInMetadataAVector)))) (Math|Conversions|ToFloat(Integer) (Math|Float|Ceil (.y (LocalVariables|GetInMetadataAVector)))) (Math|Conversions|ToFloat(Integer) (Math|Float|Ceil (.z (LocalVariables|GetInMetadataAVector))))) :NewEnumerator12 (Math|Vector|VectorGetAbs (LocalVariables|GetInMetadataAVector)) :NewEnumerator13 (Math|Vector|MakeVector (Math|Conversions|ToFloat(Integer) (Math|Float|Truncate (Math|Float|Fraction (.x (LocalVariables|GetInMetadataAVector))))) (Math|Conversions|ToFloat(Integer) (Math|Float|Truncate (Math|Float|Fraction (.y (LocalVariables|GetInMetadataAVector))))) (Math|Conversions|ToFloat(Integer) (Math|Float|Truncate (Math|Float|Fraction (.z (LocalVariables|GetInMetadataAVector)))))))))
    (:Vector4
      (PCG|Metadata|SetVector4Attribute (LocalVariables|GetoutPt) (LocalVariables|GetOutMeta) (Variables|Default|GetOutputAttributeName) (Utilities|Select :Index (Variables|Default|GetOperationMode) :NewEnumerator0 (LocalVariables|GetInMetadataAVector4) :NewEnumerator2 (Math|Vector4|MakeVector4 (Math|Float|Min(Float) _x _x_2) (Math|Float|Min(Float) _y _y_2) (Math|Float|Min(Float) _z _z_2) (Math|Float|Min(Float) _w _w_2)) :NewEnumerator3 (Math|Vector4|MakeVector4 (Math|Float|Max(Float) _x _x_2) (Math|Float|Max(Float) _y _y_2) (Math|Float|Max(Float) _z _z_2) (Math|Float|Max(Float) _w _w_2)) :NewEnumerator4 (+ (LocalVariables|GetInMetadataAVector4) (LocalVariables|GetInMetadataBVector4)) :NewEnumerator5 (- (LocalVariables|GetInMetadataAVector4) (LocalVariables|GetInMetadataBVector4)) :NewEnumerator6 (* (LocalVariables|GetInMetadataAVector4) (LocalVariables|GetInMetadataBVector4)) :NewEnumerator7 (Math|Vector4|MakeVector4 (Math|Float|SafeDivide _x _x_2) (Math|Float|SafeDivide _y _y_2) (Math|Float|SafeDivide _z _z_2) (Math|Float|SafeDivide _w _w_2)) :NewEnumerator8 (Math|Vector4|MakeVector4 (Math|Conversions|ToFloat(Integer) (Math|Float|Round _x)) (Math|Conversions|ToFloat(Integer) (Math|Float|Round _y)) (Math|Conversions|ToFloat(Integer) (Math|Float|Round _z)) (Math|Conversions|ToFloat(Integer) (Math|Float|Round _w))) :NewEnumerator9 (Math|Vector4|MakeVector4 (Math|Conversions|ToFloat(Integer) (Math|Float|Truncate _x_2)) (Math|Conversions|ToFloat(Integer) (Math|Float|Truncate _y_2)) (Math|Conversions|ToFloat(Integer) (Math|Float|Truncate _z_2)) (Math|Conversions|ToFloat(Integer) (Math|Float|Truncate _w_2))) :NewEnumerator10 (Math|Vector4|MakeVector4 (Math|Conversions|ToFloat(Integer) (Math|Float|Floor _x)) (Math|Conversions|ToFloat(Integer) (Math|Float|Floor _y)) (Math|Conversions|ToFloat(Integer) (Math|Float|Floor _z)) (Math|Conversions|ToFloat(Integer) (Math|Float|Floor _w))) :NewEnumerator11 (Math|Vector4|MakeVector4 (Math|Conversions|ToFloat(Integer) (Math|Float|Ceil _x)) (Math|Conversions|ToFloat(Integer) (Math|Float|Ceil _y)) (Math|Conversions|ToFloat(Integer) (Math|Float|Ceil _z)) (Math|Conversions|ToFloat(Integer) (Math|Float|Ceil _w))) :NewEnumerator12 (Math|Vector4|MakeVector4 (Math|Float|Absolute(Float) _x_2) (Math|Float|Absolute(Float) _y_2) (Math|Float|Absolute(Float) _z_2) (Math|Float|Absolute(Float) _w_2)) :NewEnumerator13 (Math|Vector4|MakeVector4 (Math|Conversions|ToFloat(Integer) (Math|Float|Truncate (Math|Float|Fraction _x))) (Math|Conversions|ToFloat(Integer) (Math|Float|Truncate (Math|Float|Fraction _y))) (Math|Conversions|ToFloat(Integer) (Math|Float|Truncate (Math|Float|Fraction _z))) (Math|Conversions|ToFloat(Integer) (Math|Float|Truncate (Math|Float|Fraction _w))))))))
  (return (LocalVariables|GetoutPt) true))

(fn NodeColorOverride
  (return "(R=0.949020,G=0.949020,B=0.949020,A=1.000000)"))

(fn NodeTitleOverride
  (return (Utilities|String|StringToName (Utilities|String|EnumtoString (Variables|Default|GetOperationMode)))))
