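; Deprecated PCG Blueprint element that reduces a point attribute to one value output as param data.
; It reads the last metadata attribute on the first spatial input's points, splitting Float, Vector or Vector4 values into per-component arrays and sums.
; A selection mode picks the value at PointIndex, average, per-component min or max, or components at the lowest or highest Z.
; The result becomes a named attribute on new PCGParamData.
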(fn ExecuteWithContext
  (params (InContext :type PCGContext%20Structure%20%28by%20ref%29) (Input :type PCGData%20Collection%20Structure%20%28by%20ref%29))
  (bind _zarray (LocalVariables|GetZArray))
  (bind (_indexofmaxvalue _maxvalue) (Math|Float|MaxOfFloatArray _zarray))
  (bind _zarray_2 (LocalVariables|GetZArray))
  (bind _indexofminvalue (Math|Float|MinOfFloatArray _zarray_2))
  (bind _indexofmaxvalue_2 (Math|Float|MaxOfFloatArray _zarray_2))
  (bind (_indexofminvalue_2 _minvalue) (Math|Float|MinOfFloatArray (LocalVariables|GetWArray)))
  (bind (_indexofmaxvalue_3 _maxvalue_2) (Math|Float|MaxOfFloatArray (LocalVariables|GetWArray)))
  (bind _zarray_3 (LocalVariables|GetZArray))
  (bind _indexofminvalue_3 (Math|Float|MinOfFloatArray _zarray_3))
  (bind _indexofmaxvalue_4 (Math|Float|MaxOfFloatArray _zarray_3))
  (bind _xarray (LocalVariables|GetXArray))
  (bind (_indexofminvalue_4 _minvalue_2) (Math|Float|MinOfFloatArray _xarray))
  (bind (_indexofmaxvalue_5 _maxvalue_3) (Math|Float|MaxOfFloatArray _xarray))
  (bind _yarray (LocalVariables|GetYArray))
  (bind (_indexofminvalue_5 _minvalue_3) (Math|Float|MinOfFloatArray _yarray))
  (bind (_indexofmaxvalue_6 _maxvalue_4) (Math|Float|MaxOfFloatArray _yarray))
  (bind (_indexofminvalue_6 _minvalue_4) (Math|Float|MinOfFloatArray _zarray))
  (bind _pointindex (Variables|Default|GetPointIndex))
  (bind _output (Utilities|Array|Get(aref) _yarray _pointindex))
  (bind _xarray_2 (LocalVariables|GetXArray))
  (bind _output_2 (Utilities|Array|Get(aref) _xarray_2 _indexofmaxvalue_4))
  (bind _yarray_2 (LocalVariables|GetYArray))
  (bind _output_3 (Utilities|Array|Get(aref) _yarray_2 _indexofmaxvalue_2))
  (bind _pointindex_2 (Variables|Default|GetPointIndex))
  (bind _output_4 (Utilities|Array|Get(aref) _zarray _pointindex_2))
  (bind _yarray_3 (LocalVariables|GetYArray))
  (bind _output_5 (Utilities|Array|Get(aref) _yarray_3 _indexofminvalue))
  (bind _pointindex_3 (Variables|Default|GetPointIndex))
  (bind _output_6 (Utilities|Array|Get(aref) _xarray _pointindex_3))
  (bind _xarray_3 (LocalVariables|GetXArray))
  (bind _output_7 (Utilities|Array|Get(aref) _xarray_3 _indexofminvalue_3))
  (bind _attributeselectionmode (Variables|Default|GetAttributeSelectionMode))
  (bind _xavg (LocalVariables|GetXAvg))
  (bind _returnvalue (Utilities|Select :Index _attributeselectionmode :NewEnumerator0 _output_6 :NewEnumerator1 _xavg :NewEnumerator2 _minvalue_2 :NewEnumerator3 _maxvalue_3 :NewEnumerator4 _output_7 :NewEnumerator5 _output_2))
  (bind _attributeselectionmode_2 (Variables|Default|GetAttributeSelectionMode))
  (bind _yavg (LocalVariables|GetYAvg))
  (bind _returnvalue_2 (Utilities|Select :Index _attributeselectionmode_2 :NewEnumerator0 _output :NewEnumerator1 _yavg :NewEnumerator2 _minvalue_3 :NewEnumerator3 _maxvalue_4 :NewEnumerator4 _output_5 :NewEnumerator5 _output_3))
  (bind _attributeselectionmode_3 (Variables|Default|GetAttributeSelectionMode))
  (bind _zavg (LocalVariables|GetZAvg))
  (bind _returnvalue_3 (Utilities|Select :Index _attributeselectionmode_3 :NewEnumerator0 _output_4 :NewEnumerator1 _zavg :NewEnumerator2 _minvalue_4 :NewEnumerator3 _maxvalue :NewEnumerator4 _minvalue_4 :NewEnumerator5 _maxvalue))
  (bind _pcgcontext (LocalVariables|GetPCGContext))
  (bind _paramattributename (Variables|Default|GetParamAttributeName))
  (bind _attribvaluefloat (LocalVariables|GetAttribValueFloat))
  (bind _attribvaluevector (LocalVariables|GetAttribValueVector))
  (bind _attribvaluevector4 (LocalVariables|GetAttribValueVector4))
  (bind _paramattributename_2 (Variables|Default|GetParamAttributeName))
  (bind _paramattributename_3 (Variables|Default|GetParamAttributeName))
  (LocalVariables|SetPCGContext InContext)
  (bind _returnvalue_4 (Data|GetAllSpatialInputs Input))
  (bind _aspcgspatial_data (Utilities|Casting|CastToPCGSpatialData (PCG|TaggedData|BreakTaggedData (Utilities|Array|Get(aref) _returnvalue_4))))
  (LocalVariables|SetSourceMetadata (Metadata|ConstMetadata _aspcgspatial_data))
  (bind _output_get (LocalVariables|SetAttribType (Utilities|Array|Get(aref) (.AttributeTypes (PCG|Metadata|GetAttributes (Metadata|ConstMetadata _aspcgspatial_data))) (Utilities|Array|LastIndex (PCG|Metadata|GetAttributes (Metadata|ConstMetadata _aspcgspatial_data))))))
  (LocalVariables|SetInputAttributeName (Utilities|Array|Get(aref) (PCG|Metadata|GetAttributes (Metadata|ConstMetadata _aspcgspatial_data)) (Utilities|Array|LastIndex (PCG|Metadata|GetAttributes (Metadata|ConstMetadata _aspcgspatial_data)))))
  (for _array_element (SpatialData|GetPoints (SpatialData|ToPointDatawithContext _aspcgspatial_data _pcgcontext))
    (switch Utilities|FlowControl|Switch|SwitchonEPCGMetadataTypes _output_get
      (:Float
        (bind _returnvalue_5 (PCG|Metadata|GetFloatAttribute _array_element (LocalVariables|GetSourceMetadata) (LocalVariables|GetInputAttributeName)))
        (Utilities|Array|Add (LocalVariables|GetXArray) _returnvalue_5)
        (LocalVariables|SetXAvg (+ (LocalVariables|GetXAvg) _returnvalue_5)))
      (:Vector
        (bind _returnvalue_6 (PCG|Metadata|GetVectorAttribute _array_element (LocalVariables|GetSourceMetadata) (LocalVariables|GetInputAttributeName)))
        (Utilities|Array|Add (LocalVariables|GetXArray) (.x _returnvalue_6))
        (LocalVariables|SetXAvg (+ (LocalVariables|GetXAvg) (.x _returnvalue_6)))
        (Utilities|Array|Add (LocalVariables|GetYArray) (.y _returnvalue_6))
        (LocalVariables|SetYAvg (+ (LocalVariables|GetYAvg) (.y _returnvalue_6)))
        (Utilities|Array|Add (LocalVariables|GetZArray) (.z _returnvalue_6))
        (LocalVariables|SetZAvg (+ (LocalVariables|GetZAvg) (.z _returnvalue_6))))
      (:Vector4
        (bind _returnvalue_7 (PCG|Metadata|GetVector4Attribute _array_element (LocalVariables|GetSourceMetadata) (LocalVariables|GetInputAttributeName)))
        (Utilities|Array|Add (LocalVariables|GetXArray) (Math|Vector4|BreakVector4 _returnvalue_7))
        (LocalVariables|SetXAvg (+ (LocalVariables|GetXAvg) (Math|Vector4|BreakVector4 _returnvalue_7)))
        (Utilities|Array|Add (LocalVariables|GetYArray) (.Y (Math|Vector4|BreakVector4 _returnvalue_7)))
        (LocalVariables|SetYAvg (+ (LocalVariables|GetYAvg) (.Y (Math|Vector4|BreakVector4 _returnvalue_7))))
        (Utilities|Array|Add (LocalVariables|GetZArray) (.Z (Math|Vector4|BreakVector4 _returnvalue_7)))
        (LocalVariables|SetZAvg (+ (LocalVariables|GetZAvg) (.Z (Math|Vector4|BreakVector4 _returnvalue_7))))
        (Utilities|Array|Add (LocalVariables|GetWArray) (.W (Math|Vector4|BreakVector4 _returnvalue_7)))
        (LocalVariables|SetWAvg (+ (LocalVariables|GetWAvg) (.W (Math|Vector4|BreakVector4 _returnvalue_7)))))))
  (if (and (Utilities|Enum|Equal(Enum) (Variables|Default|GetAttributeSelectionMode) "NewEnumerator0") (>= (Variables|Default|GetPointIndex) (Utilities|Array|Length (LocalVariables|GetXArray))))
    (LocalVariables|SetInvalidIndex true)
    (else
      (Utilities|FlowControl|Sequence
        (:then_0
          (LocalVariables|SetXAvg (/ (LocalVariables|GetXAvg) (Utilities|Array|Length (SpatialData|GetPoints (SpatialData|ToPointDatawithContext _aspcgspatial_data _pcgcontext)))))
          (LocalVariables|SetYAvg (/ (LocalVariables|GetYAvg) (Utilities|Array|Length (SpatialData|GetPoints (SpatialData|ToPointDatawithContext _aspcgspatial_data _pcgcontext)))))
          (LocalVariables|SetZAvg (/ (LocalVariables|GetZAvg) (Utilities|Array|Length (SpatialData|GetPoints (SpatialData|ToPointDatawithContext _aspcgspatial_data _pcgcontext)))))
          (LocalVariables|SetWAvg (/ (LocalVariables|GetWAvg) (Utilities|Array|Length (SpatialData|GetPoints (SpatialData|ToPointDatawithContext _aspcgspatial_data _pcgcontext)))))
          (switch Utilities|FlowControl|Switch|SwitchonEPCGMetadataTypes (LocalVariables|GetAttribType)
            (:Float
              (LocalVariables|SetAttribValueFloat _returnvalue))
            (:Vector
              (LocalVariables|SetAttribValueVector (Math|Vector|MakeVector _returnvalue _returnvalue_2 _returnvalue_3)))
            (:Vector4
              (LocalVariables|SetAttribValueVector4 (Math|Vector4|MakeVector4 _returnvalue _returnvalue_2 _returnvalue_3 (Utilities|Select :Index (Variables|Default|GetAttributeSelectionMode) :NewEnumerator0 (Utilities|Array|Get(aref) (LocalVariables|GetWArray) (Variables|Default|GetPointIndex)) :NewEnumerator1 (LocalVariables|GetWAvg) :NewEnumerator2 _minvalue :NewEnumerator3 _maxvalue_2 :NewEnumerator4 (Utilities|Array|Get(aref) (LocalVariables|GetWArray) (Math|Float|MinOfFloatArray (LocalVariables|GetZArray))) :NewEnumerator5 (Utilities|Array|Get(aref) (LocalVariables|GetWArray) (Math|Float|MaxOfFloatArray (LocalVariables|GetZArray)))))))))
        (:then_1
          _))))
  (bind _returnvalue_8 (Game|ConstructObjectfromClass "/Script/PCG.PCGParamData"))
  (bind _returnvalue_9 (Metadata|MutableMetadata _returnvalue_8))
  (bind _returnvalue_10 (PCG|Metadata|AddEntry _returnvalue_9 -1))
  (switch Utilities|FlowControl|Switch|SwitchonEPCGMetadataTypes (LocalVariables|GetAttribType)
    (:Float
      (PCG|Metadata|CreateFloatAttribute _returnvalue_9 _paramattributename_2 _attribvaluefloat false true)
      (PCG|Metadata|SetFloatAttributebyMetadataKey _returnvalue_10 _returnvalue_9 _paramattributename_2 _attribvaluefloat))
    (:Vector
      (PCG|Metadata|CreateVectorAttribute _returnvalue_9 _paramattributename _attribvaluevector false true)
      (PCG|Metadata|SetVectorAttributebyMetadataKey _returnvalue_10 _returnvalue_9 _paramattributename _attribvaluevector))
    (:Vector4
      (PCG|Metadata|CreateVector4Attribute _returnvalue_9 _paramattributename_3 _attribvaluevector4 false true)
      (PCG|Metadata|SetVector4AttributebyMetadataKey _returnvalue_10 _returnvalue_9 _paramattributename_3 _attribvaluevector4)))
  (return (Utilities|Struct|MakePCGDataCollection (Utilities|Array|MakeArray (PCG|TaggedData|MakeTaggedData _returnvalue_8)))))

(fn NodeColorOverride
  (return "(R=1.000000,G=0.597202,B=0.000000,A=1.000000)"))
