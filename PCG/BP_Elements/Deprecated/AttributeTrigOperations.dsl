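; PCG Blueprint element, in a Deprecated folder, that writes a trig or degree/radian conversion of float attributes to a named output attribute.
; ExecuteWithContext caches points and each input's last attribute, then runs an iteration loop over the larger point count.
; The loop body selects the operation by mode, falling back to point 0 when an input runs short.
; Only the two-input Atan2 case has execution wired; other switch cases are empty.
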(fn ExecuteWithContext
  (params (InContext :type PCGContext%20Structure%20%28by%20ref%29) (Input :type PCGData%20Collection%20Structure%20%28by%20ref%29))
  (bind _context (LocalVariables|GetContext))
  (bind _context_2 (LocalVariables|GetContext))
  (LocalVariables|SetContext InContext)
  (LocalVariables|SetInputCollection Input)
  (switch Utilities|FlowControl|Switch|SwitchonPCGAttributeTrigOperationMode (Variables|Default|GetOperationMode)
    (:NewEnumerator0
      _)
    (:NewEnumerator1
      _)
    (:NewEnumerator2
      _)
    (:NewEnumerator6
      (Variables|Hidden|SetOperationHas2Inputs true)
      (Utilities|FlowControl|Sequence
        (:then_0
          (bind _result (Utilities|Set|ToArray (PCG|Input&Output|CustomInputLabels self)))
          (bind _returnvalue (Data|GetInputsbyPinLabel (LocalVariables|GetInputCollection) (Utilities|Array|Get(aref) _result)))
          (bind _aspcgspatial_data (Utilities|Casting|CastToPCGSpatialData (PCG|TaggedData|BreakTaggedData (Utilities|Array|Get(aref) _returnvalue))))
          (Variables|Hidden|SetInputAttributeNameA (Utilities|Array|Get(acopy) (PCG|Metadata|GetAttributes (Metadata|ConstMetadata _aspcgspatial_data)) (Utilities|Array|LastIndex (PCG|Metadata|GetAttributes (Metadata|ConstMetadata _aspcgspatial_data)))))
          (Variables|Hidden|SetptDataA (SpatialData|ToPointDatawithContext _aspcgspatial_data _context_2))
          (Variables|Hidden|SetptACount (Utilities|Array|Length (SpatialData|GetPoints (SpatialData|ToPointDatawithContext _aspcgspatial_data _context_2)))))
        (:then_1
          (if (Variables|Hidden|GetOperationHas2Inputs)
            (bind _returnvalue_2 (Data|GetInputsbyPinLabel (LocalVariables|GetInputCollection) (Utilities|Array|Get(aref) _result 1)))
            (bind _aspcgspatial_data_2 (Utilities|Casting|CastToPCGSpatialData (PCG|TaggedData|BreakTaggedData (Utilities|Array|Get(aref) _returnvalue_2))))
            (Variables|Hidden|SetInputAttributeNameB (Utilities|Array|Get(acopy) (PCG|Metadata|GetAttributes (Metadata|ConstMetadata _aspcgspatial_data_2)) (Utilities|Array|LastIndex (PCG|Metadata|GetAttributes (Metadata|ConstMetadata _aspcgspatial_data_2)))))
            (Variables|Hidden|SetPtDataB (SpatialData|ToPointDatawithContext _aspcgspatial_data_2 _context))
            (Variables|Hidden|SetptBCount (Utilities|Array|Length (SpatialData|GetPoints (SpatialData|ToPointDatawithContext _aspcgspatial_data_2 _context))))))
        (:then_2
          (bind _returnvalue_3 (Game|ConstructObjectfromClass "/Script/PCG.PCGPointData"))
          (Variables|/Script/PCG.PCGSpatialData|SetTargetActor (Variables|/Script/PCG.PCGSpatialData|GetTargetActor _aspcgspatial_data) _returnvalue_3)
          (bind _returnvalue_4 (Metadata|MutableMetadata _returnvalue_3))
          (PCG|Metadata|CreateFloatAttribute _returnvalue_4 (Variables|Default|GetOutputAttributeName) 0.0 true true))
        (:then_3
          (bind _outdata (PCG|FlowControl|IterationLoop (LocalVariables|GetContext) (Math|Conversions|ToInteger64(Integer) (Math|Integer|Max(Integer) (Variables|Hidden|GetptACount) (Variables|Hidden|GetptBCount))) 0 0 _returnvalue_3)))
        (:then_4
          (return (Utilities|Struct|MakePCGDataCollection (Utilities|Array|MakeArray (PCG|TaggedData|MakeTaggedData _outdata)))))))
    (:NewEnumerator3
      _)
    (:NewEnumerator4
      _)
    (:NewEnumerator5
      _)
    (:NewEnumerator7
      _)
    (:NewEnumerator8
      _)))

(fn IterationLoopBody
  (params (InContext :type PCGContext%20Structure%20%28by%20ref%29) (Iteration :type Integer64) (InA :type PCGSpatial%20Data%20Object%20Reference) (InB :type PCGSpatial%20Data%20Object%20Reference) (OutMetadata :type PCGMetadata%20Object%20Reference))
  (bind (_transform _density _boundsmin _boundsmax _color _steepness _seed) (Utilities|Struct|BreakPCGPoint (LocalVariables|GetInPtA)))
  (LocalVariables|SetOutMeta OutMetadata)
  (LocalVariables|SetInPtA (SpatialData|GetPoint (Variables|Hidden|GetptDataA) (Math|Integer|SelectInt (Math|Conversions|ToInteger(Integer64) Iteration) 0 (< Iteration (Variables|Hidden|GetptACount)))))
  (bind _returnvalue (Metadata|MutableMetadata (Variables|Hidden|GetptDataA)))
  (LocalVariables|SetInMetaDataA _returnvalue)
  (bind _returnvalue_2 (PCG|Metadata|GetFloatAttribute (LocalVariables|GetInPtA) (LocalVariables|GetInMetaDataA) (Variables|Hidden|GetInputAttributeNameA)))
  (LocalVariables|SetInMetadataAFloat _returnvalue_2)
  (if (Variables|Hidden|GetOperationHas2Inputs)
    (LocalVariables|SetInPtB (SpatialData|GetPoint (Variables|Hidden|GetPtDataB) (Math|Integer|SelectInt (Math|Conversions|ToInteger(Integer64) Iteration) 0 (< Iteration (Variables|Hidden|GetptBCount)))))
    (bind _returnvalue_3 (Metadata|MutableMetadata (Variables|Hidden|GetPtDataB)))
    (LocalVariables|SetInMetaDataB _returnvalue_3)
    (bind _returnvalue_4 (PCG|Metadata|GetFloatAttribute (LocalVariables|GetInPtB) (LocalVariables|GetInMetaDataB) (Variables|Hidden|GetInputAttributeNameB)))
    (LocalVariables|SetInMetadataBFloat _returnvalue_4))
  (LocalVariables|SetoutPt (Utilities|Struct|MakePCGPoint _transform _density _boundsmin _boundsmax _color _steepness _seed))
  (PCG|Metadata|SetFloatAttribute (LocalVariables|GetoutPt) (LocalVariables|GetOutMeta) (Variables|Default|GetOutputAttributeName) (Utilities|Select :Index (Variables|Default|GetOperationMode) :NewEnumerator0 (Math|Trig|Acos(Radians) (LocalVariables|GetInMetadataAFloat)) :NewEnumerator1 (Math|Trig|Asin(Radians) (LocalVariables|GetInMetadataAFloat)) :NewEnumerator2 (Math|Trig|Atan(Radians) (LocalVariables|GetInMetadataAFloat)) :NewEnumerator6 (Math|Trig|Atan2(Radians) (LocalVariables|GetInMetadataAFloat) (LocalVariables|GetInMetadataBFloat)) :NewEnumerator3 (Math|Trig|Cos(Radians) (LocalVariables|GetInMetadataAFloat)) :NewEnumerator4 (Math|Trig|Sin(Radians) (LocalVariables|GetInMetadataAFloat)) :NewEnumerator5 (Math|Trig|Tan(Radians) (LocalVariables|GetInMetadataAFloat)) :NewEnumerator7 (Math|Trig|DegreesToRadians (LocalVariables|GetInMetadataAFloat)) :NewEnumerator8 (Math|Trig|RadiansToDegrees (LocalVariables|GetInMetadataAFloat))))
  (return (LocalVariables|GetoutPt) true))

(fn NodeColorOverride
  (return "(R=0.949020,G=0.949020,B=0.949020,A=1.000000)"))

(fn NodeTitleOverride
  (return (Utilities|String|StringToName (Utilities|String|EnumtoString (Variables|Default|GetOperationMode)))))
