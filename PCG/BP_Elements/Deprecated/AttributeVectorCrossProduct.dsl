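; Deprecated PCG Blueprint element that writes the cross product of vector attributes from two inputs into a named output attribute.
; ExecuteWithContext picks the last metadata attribute on each of the first two custom input pins, converts both to point data, and loops to the larger point count.
; IterationLoopBody builds each output point from input A's properties and sets the product.
; The shorter input falls back to its first point once exhausted.
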(fn ExecuteWithContext
  (params (InContext :type PCGContext%20Structure%20%28by%20ref%29) (Input :type PCGData%20Collection%20Structure%20%28by%20ref%29))
  (bind _context (LocalVariables|GetContext))
  (LocalVariables|SetContext InContext)
  (Utilities|FlowControl|Sequence
    (:then_0
      (bind _result (Utilities|Set|ToArray (PCG|Input&Output|CustomInputLabels self)))
      (bind _returnvalue (Data|GetInputsbyPinLabel Input (Utilities|Array|Get(aref) _result)))
      (bind _returnvalue_2 (Data|GetInputsbyPinLabel Input (Utilities|Array|Get(aref) _result 1)))
      (bind _aspcgspatial_data (Utilities|Casting|CastToPCGSpatialData (PCG|TaggedData|BreakTaggedData (Utilities|Array|Get(aref) _returnvalue))))
      (bind _aspcgspatial_data_2 (Utilities|Casting|CastToPCGSpatialData (PCG|TaggedData|BreakTaggedData (Utilities|Array|Get(aref) _returnvalue_2))))
      (Variables|Hidden|SetInputAttributeNameA (Utilities|Array|Get(acopy) (PCG|Metadata|GetAttributes (Metadata|ConstMetadata _aspcgspatial_data)) (Utilities|Array|LastIndex (PCG|Metadata|GetAttributes (Metadata|ConstMetadata _aspcgspatial_data)))))
      (Variables|Hidden|SetInputAttributeNameB (Utilities|Array|Get(acopy) (PCG|Metadata|GetAttributes (Metadata|ConstMetadata _aspcgspatial_data_2)) (Utilities|Array|LastIndex (PCG|Metadata|GetAttributes (Metadata|ConstMetadata _aspcgspatial_data_2))))))
    (:then_1
      (bind _returnvalue_3 (Game|ConstructObjectfromClass "/Script/PCG.PCGPointData"))
      (Variables|/Script/PCG.PCGSpatialData|SetTargetActor (Variables|/Script/PCG.PCGSpatialData|GetTargetActor _aspcgspatial_data) _returnvalue_3)
      (bind _returnvalue_4 (Metadata|MutableMetadata _returnvalue_3))
      (PCG|Metadata|CreateVectorAttribute _returnvalue_4 (Variables|Default|GetOutputAttributeName) (0.0 0.0 0.0) true true))
    (:then_2
      (Variables|Hidden|SetptDataA (SpatialData|ToPointDatawithContext _aspcgspatial_data _context))
      (Variables|Hidden|SetPtDataB (SpatialData|ToPointDatawithContext _aspcgspatial_data_2 _context))
      (Variables|Hidden|SetptACount (Utilities|Array|Length (SpatialData|GetPoints (SpatialData|ToPointDatawithContext _aspcgspatial_data _context))))
      (Variables|Hidden|SetptBCount (Utilities|Array|Length (SpatialData|GetPoints (SpatialData|ToPointDatawithContext _aspcgspatial_data_2 _context))))
      (bind _outdata (PCG|FlowControl|IterationLoop (LocalVariables|GetContext) (Math|Conversions|ToInteger64(Integer) (Math|Integer|Max(Integer) (Utilities|Array|Length (SpatialData|GetPoints (SpatialData|ToPointDatawithContext _aspcgspatial_data _context))) (Utilities|Array|Length (SpatialData|GetPoints (SpatialData|ToPointDatawithContext _aspcgspatial_data_2 _context))))) 0 0 _returnvalue_3)))
    (:then_3
      (Utilities|Array|Add (LocalVariables|GetOutData) (PCG|TaggedData|MakeTaggedData _outdata)))
    (:then_4
      (return (Utilities|Struct|MakePCGDataCollection (LocalVariables|GetOutData))))))

(fn IterationLoopBody
  (params (InContext :type PCGContext%20Structure%20%28by%20ref%29) (Iteration :type Integer64) (InA :type PCGSpatial%20Data%20Object%20Reference) (InB :type PCGSpatial%20Data%20Object%20Reference) (OutMetadata :type PCGMetadata%20Object%20Reference))
  (bind (_transform _density _boundsmin _boundsmax _color _steepness _seed) (Utilities|Struct|BreakPCGPoint (LocalVariables|GetInPtA)))
  (LocalVariables|SetOutMeta OutMetadata)
  (LocalVariables|SetInPtA (SpatialData|GetPoint (Variables|Hidden|GetptDataA) (Math|Integer|SelectInt (Math|Conversions|ToInteger(Integer64) Iteration) 0 (< Iteration (Variables|Hidden|GetptACount)))))
  (LocalVariables|SetInPtB (SpatialData|GetPoint (Variables|Hidden|GetPtDataB) (Math|Integer|SelectInt (Math|Conversions|ToInteger(Integer64) Iteration) 0 (< Iteration (Variables|Hidden|GetptBCount)))))
  (bind _returnvalue (Metadata|MutableMetadata (Variables|Hidden|GetptDataA)))
  (LocalVariables|SetInMetaDataA _returnvalue)
  (bind _returnvalue_2 (Metadata|MutableMetadata (Variables|Hidden|GetPtDataB)))
  (LocalVariables|SetInMetaDataB _returnvalue_2)
  (LocalVariables|SetoutPt (Utilities|Struct|MakePCGPoint _transform _density _boundsmin _boundsmax _color _steepness _seed))
  (bind _returnvalue_3 (PCG|Metadata|GetVectorAttribute (LocalVariables|GetInPtA) (LocalVariables|GetInMetaDataA) (Variables|Hidden|GetInputAttributeNameA)))
  (bind _returnvalue_4 (PCG|Metadata|GetVectorAttribute (LocalVariables|GetInPtB) (LocalVariables|GetInMetaDataB) (Variables|Hidden|GetInputAttributeNameB)))
  (PCG|Metadata|SetVectorAttribute (LocalVariables|GetoutPt) (LocalVariables|GetOutMeta) (Variables|Default|GetOutputAttributeName) (Math|Vector|CrossProduct _returnvalue_3 _returnvalue_4))
  (return (LocalVariables|GetoutPt) true))

(fn NodeColorOverride
  (return "(R=0.949020,G=0.949020,B=0.949020,A=1.000000)"))
