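; Deprecated PCG Blueprint element, titled "Vector Length", that writes the length of a vector attribute to a float attribute on each point.
; ExecuteWithContext takes the first spatial input, assumes its last metadata attribute is the vector, and creates new point data holding the named output attribute.
; PointLoopBody rebuilds each point from its basic properties and stores the vector's length, and the resulting data is returned as the output collection.
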(fn ExecuteWithContext
  (params (InContext :type PCGContext%20Structure%20%28by%20ref%29) (Input :type PCGData%20Collection%20Structure%20%28by%20ref%29))
  (LocalVariables|SetContext InContext)
  (bind _returnvalue (Data|GetAllSpatialInputs Input))
  (Utilities|FlowControl|Sequence
    (:then_0
      (bind _aspcgspatial_data (Utilities|Casting|CastToPCGSpatialData (PCG|TaggedData|BreakTaggedData (Utilities|Array|Get(aref) _returnvalue))))
      (Variables|Hidden|SetInputAttributeName (Utilities|Array|Get(aref) (PCG|Metadata|GetAttributes (Metadata|ConstMetadata _aspcgspatial_data)) (Utilities|Array|LastIndex (PCG|Metadata|GetAttributes (Metadata|ConstMetadata _aspcgspatial_data)))))
      (bind _returnvalue_2 (Game|ConstructObjectfromClass "/Script/PCG.PCGPointData"))
      (bind _returnvalue_3 (Metadata|MutableMetadata _returnvalue_2))
      (PCG|Metadata|CreateFloatAttribute _returnvalue_3 (Variables|Default|GetOutputAttributeName) 0.0 true true))
    (:then_1
      (bind _outdata (PCG|FlowControl|PointLoop (LocalVariables|GetContext) (SpatialData|ToPointDatawithContext _aspcgspatial_data (LocalVariables|GetContext)) _returnvalue_2)))
    (:then_2
      (Utilities|Array|Add (LocalVariables|GetOutData) (PCG|TaggedData|MakeTaggedData _outdata)))
    (:then_3
      (return (Utilities|Struct|MakePCGDataCollection (LocalVariables|GetOutData))))))

(fn NodeColorOverride
  (return "(R=0.949020,G=0.949020,B=0.949020,A=1.000000)"))

(fn NodeTitleOverride
  (return "Vector Length"))

(fn PointLoopBody
  (params (InContext :type PCGContext%20Structure%20%28by%20ref%29) (InData :type PCGPoint%20Data%20Object%20Reference) (InPoint :type PCGPoint%20Structure%20%28by%20ref%29) (OutMetadata :type PCGMetadata%20Object%20Reference) (Iteration :type Integer64))
  (bind (_transform _density _boundsmin _boundsmax _color _steepness _seed) (Utilities|Struct|BreakPCGPoint (LocalVariables|GetSrcPoint)))
  (LocalVariables|SetSrcPoint InPoint)
  (LocalVariables|SetMetadata OutMetadata)
  (bind _output_get (LocalVariables|SetDestPoint (Utilities|Struct|MakePCGPoint _transform _density _boundsmin _boundsmax _color _steepness _seed)))
  (bind _returnvalue (Metadata|MutableMetadata InData))
  (bind _returnvalue_2 (PCG|Metadata|GetVectorAttribute (LocalVariables|GetSrcPoint) _returnvalue (Variables|Hidden|GetInputAttributeName)))
  (PCG|Metadata|SetFloatAttribute (LocalVariables|GetDestPoint) (LocalVariables|GetMetadata) (Variables|Default|GetOutputAttributeName) (Math|Vector|VectorLength _returnvalue_2))
  (return (LocalVariables|GetDestPoint) true))
