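; Custom PCG Blueprint node titled "Break Vector", in a Deprecated folder, that splits a point attribute into float attributes.
; It takes the last metadata attribute of the first spatial input and creates four named float attributes.
; A point loop fills them from Float, Vector or Vector4 values, with optional debug printing for Vector.
; The result is copied four times, deleting three component attributes from each copy, to feed four output pins.
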(fn ExecuteWithContext
  (params (InContext :type PCGContext%20Structure%20%28by%20ref%29) (Input :type PCGData%20Collection%20Structure%20%28by%20ref%29))
  (bind _tmpoutpointdata (LocalVariables|GetTmpOutPointData))
  (LocalVariables|SetContext InContext)
  (bind _returnvalue (Data|GetAllSpatialInputs Input))
  (Utilities|FlowControl|Sequence
    (:then_0
      (bind _aspcgspatial_data (Utilities|Casting|CastToPCGSpatialData (PCG|TaggedData|BreakTaggedData (Utilities|Array|Get(aref) _returnvalue))))
      (Variables|Hidden|SetInputAttributeName (Utilities|Array|Get(aref) (PCG|Metadata|GetAttributes (Metadata|ConstMetadata _aspcgspatial_data)) (Utilities|Array|LastIndex (PCG|Metadata|GetAttributes (Metadata|ConstMetadata _aspcgspatial_data)))))
      (bind _returnvalue_2 (Game|ConstructObjectfromClass "/Script/PCG.PCGPointData"))
      (SpatialData|InitializefromData _returnvalue_2 _aspcgspatial_data 0 false true)
      (bind _returnvalue_3 (Metadata|MutableMetadata _returnvalue_2))
      (bind _result (Utilities|Set|ToArray (Utilities|Set|MakeSet (Variables|Default|GetOutputAttributeName0) (Variables|Default|GetOutputAttributeName1) (Variables|Default|GetOutputAttributeName2) (Variables|Default|GetOutputAttributeName3))))
      (for _array_element _result
        (PCG|Metadata|CreateFloatAttribute _returnvalue_3 _array_element 0.0 true true)))
    (:then_1
      (bind _outdata (PCG|FlowControl|PointLoop (LocalVariables|GetContext) (SpatialData|ToPointDatawithContext _aspcgspatial_data (LocalVariables|GetContext)) _returnvalue_2))
      (LocalVariables|SetTmpOutPointData _outdata))
    (:then_2
      (LocalVariables|SetTmpCombinedMetadata (Metadata|ConstMetadata _tmpoutpointdata))
      (LocalVariables|SetTmpOutPoints (SpatialData|GetPoints _tmpoutpointdata))
      ;[1 1 1] Split OutPointData with one metadata attribute per pin. (Copy 4 times, delete extra attributes)
      (Utilities|FlowControl|Sequence
        (:then_0
          (bind _returnvalue_4 (Game|ConstructObjectfromClass "/Script/PCG.PCGPointData"))
          (SpatialData|InitializefromData _returnvalue_4 (LocalVariables|GetTmpOutPointData) 0 false true)
          (SpatialData|SetPoints _returnvalue_4 (LocalVariables|GetTmpOutPoints))
          (bind _returnvalue_5 (Metadata|MutableMetadata _returnvalue_4))
          (PCG|Metadata|InitializeAsCopy :self _returnvalue_5 :InMetadataToCopy (LocalVariables|GetTmpCombinedMetadata))
          (PCG|Metadata|DeleteAttribute _returnvalue_5 (Variables|Default|GetOutputAttributeName1))
          (PCG|Metadata|DeleteAttribute _returnvalue_5 (Variables|Default|GetOutputAttributeName2))
          (PCG|Metadata|DeleteAttribute _returnvalue_5 (Variables|Default|GetOutputAttributeName3))
          (bind _result_2 (Utilities|Set|ToArray (PCG|Input&Output|CustomOutputLabels self)))
          (Utilities|Array|Add (LocalVariables|GetOutDatas) (PCG|TaggedData|MakeTaggedData :Data _returnvalue_4 :Pin (Utilities|Array|Get(aref) _result_2))))
        (:then_1
          (bind _returnvalue_6 (Game|ConstructObjectfromClass "/Script/PCG.PCGPointData"))
          (SpatialData|InitializefromData _returnvalue_6 (LocalVariables|GetTmpOutPointData) 0 false true)
          (SpatialData|SetPoints _returnvalue_6 (LocalVariables|GetTmpOutPoints))
          (bind _returnvalue_7 (Metadata|MutableMetadata _returnvalue_6))
          (PCG|Metadata|InitializeAsCopy :self _returnvalue_7 :InMetadataToCopy (LocalVariables|GetTmpCombinedMetadata))
          (PCG|Metadata|DeleteAttribute _returnvalue_7 (Variables|Default|GetOutputAttributeName0))
          (PCG|Metadata|DeleteAttribute _returnvalue_7 (Variables|Default|GetOutputAttributeName2))
          (PCG|Metadata|DeleteAttribute _returnvalue_7 (Variables|Default|GetOutputAttributeName3))
          (bind _result_3 (Utilities|Set|ToArray (PCG|Input&Output|CustomOutputLabels self)))
          (Utilities|Array|Add (LocalVariables|GetOutDatas) (PCG|TaggedData|MakeTaggedData :Data _returnvalue_6 :Pin (Utilities|Array|Get(aref) _result_3 1))))
        (:then_2
          (bind _returnvalue_8 (Game|ConstructObjectfromClass "/Script/PCG.PCGPointData"))
          (SpatialData|InitializefromData _returnvalue_8 (LocalVariables|GetTmpOutPointData) 0 false true)
          (SpatialData|SetPoints _returnvalue_8 (LocalVariables|GetTmpOutPoints))
          (bind _returnvalue_9 (Metadata|MutableMetadata _returnvalue_8))
          (PCG|Metadata|InitializeAsCopy :self _returnvalue_9 :InMetadataToCopy (LocalVariables|GetTmpCombinedMetadata))
          (PCG|Metadata|DeleteAttribute _returnvalue_9 (Variables|Default|GetOutputAttributeName0))
          (PCG|Metadata|DeleteAttribute _returnvalue_9 (Variables|Default|GetOutputAttributeName1))
          (PCG|Metadata|DeleteAttribute _returnvalue_9 (Variables|Default|GetOutputAttributeName3))
          (bind _result_4 (Utilities|Set|ToArray (PCG|Input&Output|CustomOutputLabels self)))
          (Utilities|Array|Add (LocalVariables|GetOutDatas) (PCG|TaggedData|MakeTaggedData :Data _returnvalue_8 :Pin (Utilities|Array|Get(aref) _result_4 2))))
        (:then_3
          (bind _returnvalue_10 (Game|ConstructObjectfromClass "/Script/PCG.PCGPointData"))
          (SpatialData|InitializefromData _returnvalue_10 (LocalVariables|GetTmpOutPointData) 0 false true)
          (SpatialData|SetPoints _returnvalue_10 (LocalVariables|GetTmpOutPoints))
          (bind _returnvalue_11 (Metadata|MutableMetadata _returnvalue_10))
          (PCG|Metadata|InitializeAsCopy :self _returnvalue_11 :InMetadataToCopy (LocalVariables|GetTmpCombinedMetadata))
          (PCG|Metadata|DeleteAttribute _returnvalue_11 (Variables|Default|GetOutputAttributeName0))
          (PCG|Metadata|DeleteAttribute _returnvalue_11 (Variables|Default|GetOutputAttributeName1))
          (PCG|Metadata|DeleteAttribute _returnvalue_11 (Variables|Default|GetOutputAttributeName2))
          (bind _result_5 (Utilities|Set|ToArray (PCG|Input&Output|CustomOutputLabels self)))
          (Utilities|Array|Add (LocalVariables|GetOutDatas) (PCG|TaggedData|MakeTaggedData :Data _returnvalue_10 :Pin (Utilities|Array|Get(aref) _result_5 3))))
        (:then_4
          (return (Utilities|Struct|MakePCGDataCollection (LocalVariables|GetOutDatas))))))))

(fn NodeColorOverride
  (return "(R=0.949020,G=0.949020,B=0.949020,A=1.000000)"))

(fn NodeTitleOverride
  (return "Break Vector"))

(fn PointLoopBody
  (params (InContext :type PCGContext%20Structure%20%28by%20ref%29) (InData :type PCGPoint%20Data%20Object%20Reference) (InPoint :type PCGPoint%20Structure%20%28by%20ref%29) (OutMetadata :type PCGMetadata%20Object%20Reference) (Iteration :type Integer64))
  (bind _returnvalue (Metadata|ConstMetadata InData))
  (bind (_attributenames _attributetypes) (PCG|Metadata|GetAttributes _returnvalue))
  (bind _outputattributename1 (Variables|Default|GetOutputAttributeName1))
  (bind _metadata (LocalVariables|GetMetadata))
  (bind _destpoint (LocalVariables|GetDestPoint))
  (bind _destpoint_2 (LocalVariables|GetDestPoint))
  (bind _metadata_2 (LocalVariables|GetMetadata))
  (bind _outputattributename0 (Variables|Default|GetOutputAttributeName0))
  (bind _outputattributename2 (Variables|Default|GetOutputAttributeName2))
  (bind _output_get (LocalVariables|SetSrcPoint InPoint))
  (LocalVariables|SetMetadata OutMetadata)
  (bind _output_get_2 (LocalVariables|SetDestPoint (Utilities|Struct|MakePCGPoint (Utilities|Struct|BreakPCGPoint _output_get) (.Density (Utilities|Struct|BreakPCGPoint _output_get)) (.BoundsMin (Utilities|Struct|BreakPCGPoint _output_get)) (.BoundsMax (Utilities|Struct|BreakPCGPoint _output_get)) (.Color (Utilities|Struct|BreakPCGPoint _output_get)) (.Steepness (Utilities|Struct|BreakPCGPoint _output_get)) (.Seed (Utilities|Struct|BreakPCGPoint _output_get)))))
  (PCG|Metadata|InitializeMetadata _output_get_2 OutMetadata (LocalVariables|GetSrcPoint))
  (switch Utilities|FlowControl|Switch|SwitchonEPCGMetadataTypes (Utilities|Array|Get(aref) _attributetypes (Utilities|Array|FindItem _attributenames (Variables|Hidden|GetInputAttributeName)))
    (:Float
      (bind _returnvalue_2 (PCG|Metadata|GetFloatAttribute (LocalVariables|GetSrcPoint) _returnvalue (Variables|Hidden|GetInputAttributeName)))
      (PCG|Metadata|SetFloatAttribute (LocalVariables|GetDestPoint) (LocalVariables|GetMetadata) (Variables|Default|GetOutputAttributeName0) _returnvalue_2))
    (:Vector
      (bind _returnvalue_3 (PCG|Metadata|GetVectorAttribute (LocalVariables|GetSrcPoint) _returnvalue (Variables|Hidden|GetInputAttributeName)))
      (PCG|Metadata|SetFloatAttribute _destpoint _metadata _outputattributename0 (.x _returnvalue_3))
      (PCG|Metadata|SetFloatAttribute _destpoint _metadata _outputattributename1 (.y _returnvalue_3))
      (PCG|Metadata|SetFloatAttribute _destpoint _metadata _outputattributename2 (.z _returnvalue_3))
      (if (Variables|Default|GetDOPRINT)
        (Development|PrintString (Utilities|String|ToString(Name) _outputattributename0) true true "(R=1.000000,G=0.938499,B=0.000000,A=1.000000)")
        (Development|PrintString (Utilities|String|ToString(Name) _outputattributename1) true true "(R=1.000000,G=0.998558,B=0.000000,A=1.000000)")
        (Development|PrintString (Utilities|String|ToString(Name) _outputattributename2) true true "(R=0.919080,G=1.000000,B=0.000000,A=1.000000)")
        (Development|PrintString (Utilities|String|ToString(Float) (.x _returnvalue_3)) true true "(R=1.000000,G=0.938499,B=0.000000,A=1.000000)")
        (Development|PrintString (Utilities|String|ToString(Float) (.y _returnvalue_3)) true true "(R=1.000000,G=0.998558,B=0.000000,A=1.000000)")
        (Development|PrintString (Utilities|String|ToString(Float) (.z _returnvalue_3)) true true "(R=0.919080,G=1.000000,B=0.000000,A=1.000000)")))
    (:Vector4
      (bind _returnvalue_4 (PCG|Metadata|GetVector4Attribute (LocalVariables|GetSrcPoint) _returnvalue (Variables|Hidden|GetInputAttributeName)))
      (PCG|Metadata|SetFloatAttribute _destpoint_2 _metadata_2 _outputattributename0 (Math|Vector4|BreakVector4 _returnvalue_4))
      (PCG|Metadata|SetFloatAttribute _destpoint_2 _metadata_2 _outputattributename1 (.Y (Math|Vector4|BreakVector4 _returnvalue_4)))
      (PCG|Metadata|SetFloatAttribute _destpoint_2 _metadata_2 _outputattributename2 (.Z (Math|Vector4|BreakVector4 _returnvalue_4)))
      (PCG|Metadata|SetFloatAttribute _destpoint_2 _metadata_2 (Variables|Default|GetOutputAttributeName3) (.W (Math|Vector4|BreakVector4 _returnvalue_4)))))
  (return (LocalVariables|GetDestPoint) true))
