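; Deprecated PCG Blueprint element that splits a Float, Vector or Vector4 attribute from the first input param data into separate float attribute sets.
; It takes the type from the last metadata attribute and reads the value at entry key 0.
; Sequenced type switches create param data and entries, write floats under configurable X/Y/Z/W names, then tag each output to custom pins by index.
; The node is colored orange.
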(fn Execute
  (params (Input :type PCGData%20Collection%20Structure%20%28by%20ref%29))
  (bind _inmetadata (LocalVariables|GetInMetadata))
  (bind (_attributenames _attributetypes) (PCG|Metadata|GetAttributes _inmetadata))
  (bind _returnvalue (Utilities|Array|LastIndex _attributenames))
  (bind _output (Utilities|Array|Get(aref) _attributetypes _returnvalue))
  (bind _outmetadataw (LocalVariables|GetOutMetadataW))
  (bind _outmetadataz (LocalVariables|GetOutMetadataZ))
  (bind _outmetadatay (LocalVariables|GetOutMetadataY))
  (bind _attributenameoutx (Variables|Default|GetAttributeNameOutX))
  (bind _outmetadatax (LocalVariables|GetOutMetadataX))
  (bind _attributenameoutx_2 (Variables|Default|GetAttributeNameOutX))
  (bind _attributenameouty (Variables|Default|GetAttributeNameOutY))
  (bind _attributenameoutz (Variables|Default|GetAttributeNameOutZ))
  (bind _attributenameoutw (Variables|Default|GetAttributeNameOutW))
  (bind _outmetadatay_2 (LocalVariables|GetOutMetadataY))
  (bind _outmetadataz_2 (LocalVariables|GetOutMetadataZ))
  (bind _outmetadatax_2 (LocalVariables|GetOutMetadataX))
  (bind _attributenameoutx_3 (Variables|Default|GetAttributeNameOutX))
  (bind _attributenameouty_2 (Variables|Default|GetAttributeNameOutY))
  (bind _attributenameoutz_2 (Variables|Default|GetAttributeNameOutZ))
  (bind _outmetadatax_3 (LocalVariables|GetOutMetadataX))
  (bind _returnvalue_2 (Data|GetParams Input))
  (bind _aspcgattribute_set (Utilities|Casting|CastToPCGParamData (PCG|TaggedData|BreakTaggedData (Utilities|Array|Get(aref) _returnvalue_2))))
  (bind _returnvalue_3 (Metadata|MutableMetadata _aspcgattribute_set))
  (LocalVariables|SetInMetadata _returnvalue_3)
  (Utilities|FlowControl|Sequence
    (:then_0
      (switch Utilities|FlowControl|Switch|SwitchonEPCGMetadataTypes _output
        (:Float
          _)
        (:Vector
          (bind _returnvalue_4 (Game|ConstructObjectfromClass "/Script/PCG.PCGParamData"))
          (bind _output_get (LocalVariables|SetOutParamDataZ _returnvalue_4))
          (bind _returnvalue_5 (Metadata|MutableMetadata _output_get))
          (LocalVariables|SetOutMetadataZ _returnvalue_5)
          (bind _returnvalue_6 (Game|ConstructObjectfromClass "/Script/PCG.PCGParamData"))
          (bind _output_get_2 (LocalVariables|SetOutParamDataY _returnvalue_6))
          (bind _returnvalue_7 (Metadata|MutableMetadata _output_get_2))
          (LocalVariables|SetOutMetadataY _returnvalue_7))
        (:Vector4
          (bind _returnvalue_8 (Game|ConstructObjectfromClass "/Script/PCG.PCGParamData"))
          (bind _output_get_3 (LocalVariables|SetOutParamDataW _returnvalue_8))
          (bind _returnvalue_9 (Metadata|MutableMetadata _output_get_3))
          (LocalVariables|SetOutMetadataW _returnvalue_9)))
      (bind _returnvalue_10 (Game|ConstructObjectfromClass "/Script/PCG.PCGParamData"))
      (bind _output_get_4 (LocalVariables|SetOutParamDataX _returnvalue_10))
      (bind _returnvalue_11 (Metadata|MutableMetadata _output_get_4))
      (LocalVariables|SetOutMetadataX _returnvalue_11))
    (:then_1
      (switch Utilities|FlowControl|Switch|SwitchonEPCGMetadataTypes _output
        (:Float
          _)
        (:Vector
          (bind _returnvalue_12 (PCG|Metadata|AddEntry (LocalVariables|GetOutMetadataZ) -1))
          (LocalVariables|SetMetadataEntryKeyZ _returnvalue_12)
          (bind _returnvalue_13 (PCG|Metadata|AddEntry (LocalVariables|GetOutMetadataY) -1))
          (LocalVariables|SetMetadataEntryKeyY _returnvalue_13))
        (:Vector4
          (bind _returnvalue_14 (PCG|Metadata|AddEntry (LocalVariables|GetOutMetadataW) -1))
          (LocalVariables|SetMetadataEntryKeyW _returnvalue_14)))
      (bind _returnvalue_15 (PCG|Metadata|AddEntry (LocalVariables|GetOutMetadataX) -1))
      (LocalVariables|SetMetadataEntryKeyX _returnvalue_15))
    (:then_2
      (switch Utilities|FlowControl|Switch|SwitchonEPCGMetadataTypes _output
        (:Float
          (bind _returnvalue_16 (PCG|Metadata|GetFloatAttributebyMetadataKey 0 (LocalVariables|GetInMetadata)))
          (PCG|Metadata|CreateFloatAttribute _outmetadatax _attributenameoutx_2 _returnvalue_16 false true)
          (PCG|Metadata|SetFloatAttributebyMetadataKey (LocalVariables|GetMetadataEntryKeyX) _outmetadatax _attributenameoutx_2 _returnvalue_16))
        (:Vector
          (bind _returnvalue_17 (PCG|Metadata|GetVectorAttributebyMetadataKey 0 (LocalVariables|GetInMetadata)))
          (PCG|Metadata|CreateFloatAttribute _outmetadatax_3 _attributenameoutx (.x _returnvalue_17) false true)
          (PCG|Metadata|SetFloatAttributebyMetadataKey (LocalVariables|GetMetadataEntryKeyX) _outmetadatax_3 _attributenameoutx (.x _returnvalue_17))
          (PCG|Metadata|CreateFloatAttribute _outmetadatay _attributenameouty (.y _returnvalue_17) false true)
          (PCG|Metadata|SetFloatAttributebyMetadataKey (LocalVariables|GetMetadataEntryKeyY) _outmetadatay _attributenameouty (.y _returnvalue_17))
          (PCG|Metadata|CreateFloatAttribute _outmetadataz _attributenameoutz (.z _returnvalue_17) false true)
          (PCG|Metadata|SetFloatAttributebyMetadataKey (LocalVariables|GetMetadataEntryKeyZ) _outmetadataz _attributenameoutz (.z _returnvalue_17)))
        (:Vector4
          (bind _returnvalue_18 (PCG|Metadata|GetVector4AttributebyMetadataKey 0 (LocalVariables|GetInMetadata)))
          (PCG|Metadata|CreateFloatAttribute _outmetadatax_2 _attributenameoutx_3 (Math|Vector4|BreakVector4 _returnvalue_18) false true)
          (PCG|Metadata|SetFloatAttributebyMetadataKey (LocalVariables|GetMetadataEntryKeyX) _outmetadatax_2 _attributenameoutx_3 (Math|Vector4|BreakVector4 _returnvalue_18))
          (PCG|Metadata|CreateFloatAttribute _outmetadatay_2 _attributenameouty_2 (.Y (Math|Vector4|BreakVector4 _returnvalue_18)) false true)
          (PCG|Metadata|SetFloatAttributebyMetadataKey (LocalVariables|GetMetadataEntryKeyY) _outmetadatay_2 _attributenameouty_2 (.Y (Math|Vector4|BreakVector4 _returnvalue_18)))
          (PCG|Metadata|CreateFloatAttribute _outmetadataz_2 _attributenameoutz_2 (.Z (Math|Vector4|BreakVector4 _returnvalue_18)) false true)
          (PCG|Metadata|SetFloatAttributebyMetadataKey (LocalVariables|GetMetadataEntryKeyZ) _outmetadataz_2 _attributenameoutz_2 (.Z (Math|Vector4|BreakVector4 _returnvalue_18)))
          (PCG|Metadata|CreateFloatAttribute _outmetadataw _attributenameoutw (.W (Math|Vector4|BreakVector4 _returnvalue_18)) false true)
          (PCG|Metadata|SetFloatAttributebyMetadataKey (LocalVariables|GetMetadataEntryKeyW) _outmetadataw _attributenameoutw (.W (Math|Vector4|BreakVector4 _returnvalue_18))))))
    (:then_3
      (bind _result (Utilities|Set|ToArray (PCG|Input&Output|CustomOutputLabels self)))
      (switch Utilities|FlowControl|Switch|SwitchonEPCGMetadataTypes _output
        (:Float
          _)
        (:Vector
          (Utilities|Array|Add (LocalVariables|GetOutDatas) (PCG|TaggedData|MakeTaggedData :Data (LocalVariables|GetOutParamDataZ) :Pin (Utilities|Array|Get(aref) _result 2)))
          (Utilities|Array|Add (LocalVariables|GetOutDatas) (PCG|TaggedData|MakeTaggedData :Data (LocalVariables|GetOutParamDataY) :Pin (Utilities|Array|Get(aref) _result 1))))
        (:Vector4
          (Utilities|Array|Add (LocalVariables|GetOutDatas) (PCG|TaggedData|MakeTaggedData :Data (LocalVariables|GetOutParamDataW) :Pin (Utilities|Array|Get(aref) _result 3)))))
      (Utilities|Array|Add (LocalVariables|GetOutDatas) (PCG|TaggedData|MakeTaggedData :Data (LocalVariables|GetOutParamDataX) :Pin (Utilities|Array|Get(aref) _result))))
    (:then_4
      (return (Utilities|Struct|MakePCGDataCollection (LocalVariables|GetOutDatas))))))

(fn NodeColorOverride
  (return "(R=1.000000,G=0.600000,B=0.000000,A=1.000000)"))
