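; A deprecated PCG Blueprint element that merges the spatial inputs and param data of its input collection into one output collection.
; Execute fetches all spatial inputs and all params from the input, appends both arrays to a local tagged-data array, spatial first, then returns a new PCG data collection built from it.
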(fn Execute
  (params (Input :type PCGData%20Collection%20Structure%20%28by%20ref%29))
  (bind _outputtaggeddata (LocalVariables|GetOutputTaggedData))
  (bind _returnvalue (Data|GetAllSpatialInputs Input))
  (bind _returnvalue_2 (Data|GetParams Input))
  (Utilities|Array|AppendArray _outputtaggeddata _returnvalue)
  (Utilities|Array|AppendArray _outputtaggeddata _returnvalue_2)
  (return (Utilities|Struct|MakePCGDataCollection (LocalVariables|GetOutputTaggedData))))
