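; Deprecated custom PCG element that copies source points onto every target point.
; ExecuteWithContext converts both labeled inputs' first entries to point data, runs a nested loop, then returns the result with the target actor and input settings.
; NestedLoopBody composes source and target transforms, using inverse target scale when target scaling is off, and reseeds from position.
; An attribute inheritance setting selects whether metadata comes from source or target.
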(fn ExecuteWithContext
  (params (InContext :type PCGContext%20Structure%20%28by%20ref%29) (Input :type PCGData%20Collection%20Structure%20%28by%20ref%29))
  (bind _inputcollection (LocalVariables|GetInputCollection))
  (bind _outpointdata (LocalVariables|GetOutPointData))
  (LocalVariables|SetContext InContext)
  (LocalVariables|SetInputCollection Input)
  (Utilities|FlowControl|Sequence
    (:then_0
      ;[1 1 1] Prepare Inputs
      (bind _result (Utilities|Set|ToArray (PCG|Input&Output|CustomInputLabels self)))
      (bind _returnvalue (Data|GetInputsbyPinLabel _inputcollection (Utilities|Array|Get(aref) _result)))
      (bind _returnvalue_2 (Data|GetInputsbyPinLabel _inputcollection (Utilities|Array|Get(aref) _result 1)))
      (LocalVariables|SetTargetPointsToCopyTo (PCG|TaggedData|BreakTaggedData (Utilities|Array|Get(aref) _returnvalue_2)))
      (LocalVariables|SetSourcePointsToCopy (PCG|TaggedData|BreakTaggedData (Utilities|Array|Get(aref) _returnvalue))))
    (:then_1
      (bind _aspcgspatial_data (Utilities|Casting|CastToPCGSpatialData (LocalVariables|GetTargetPointsToCopyTo)))
      (LocalVariables|SetTargetActor (Variables|/Script/PCG.PCGSpatialData|GetTargetActor _aspcgspatial_data))
      (LocalVariables|SetTargetPointData (SpatialData|ToPointDatawithContext _aspcgspatial_data (LocalVariables|GetContext)))
      (bind _aspcgspatial_data_2 (Utilities|Casting|CastToPCGSpatialData (LocalVariables|GetSourcePointsToCopy)))
      (LocalVariables|SetSourcePointData (SpatialData|ToPointDatawithContext _aspcgspatial_data_2 (LocalVariables|GetContext)))
      (bind _returnvalue_3 (Game|ConstructObjectfromClass "/Script/PCG.PCGPointData"))
      (bind _output_get (LocalVariables|SetOutData _returnvalue_3))
      (SpatialData|InitializefromData _returnvalue_3 (Utilities|Select :Index (Variables|Default|GetAttributeInheritance) :NewEnumerator0 (LocalVariables|GetSourcePointData) :NewEnumerator1 (LocalVariables|GetTargetPointData)) 0 true true)
      (bind _outdata (PCG|FLowControl|NestedLoop (LocalVariables|GetContext) (LocalVariables|GetTargetPointData) (LocalVariables|GetSourcePointData) (LocalVariables|GetOutData)))
      (LocalVariables|SetOutPointData _outdata))
    (:then_2
      (Variables|/Script/PCG.PCGSpatialData|SetTargetActor (LocalVariables|GetTargetActor) _outpointdata)
      (LocalVariables|SetOutputTaggedData (Utilities|Array|MakeArray (PCG|TaggedData|MakeTaggedData :Data _outpointdata)))
      (bind _returnvalue_4 (Data|GetAllSettings Input))
      (Utilities|Array|AppendArray (LocalVariables|GetOutputTaggedData) _returnvalue_4)
      (return (Utilities|Struct|MakePCGDataCollection (LocalVariables|GetOutputTaggedData))))))

(fn NestedLoopBody
  (params (InContext :type PCGContext%20Structure%20%28by%20ref%29) (InOuterData :type PCGPoint%20Data%20Object%20Reference) (InInnerData :type PCGPoint%20Data%20Object%20Reference) (InOuterPoint :type PCGPoint%20Structure%20%28by%20ref%29) (InInnerPoint :type PCGPoint%20Structure%20%28by%20ref%29) (OutMetadata :type PCGMetadata%20Object%20Reference) (OuterIteration :type Integer64) (InnerIteration :type Integer64))
  (LocalVariables|SetOutMeta OutMetadata)
  (LocalVariables|SetIn_A InOuterData)
  (LocalVariables|SetIn_B InInnerData)
  (LocalVariables|SetInPtA InOuterPoint)
  (LocalVariables|SetInPtB InInnerPoint)
  (bind _structout (Utilities|Struct|SetmembersinPCGPoint (LocalVariables|GetInPtB) (Math|Transform|ComposeTransforms (Math|Transform|MakeTransform (.location (Utilities|Struct|BreakPCGPoint (LocalVariables|GetInPtB))) (.rotation (Utilities|Struct|BreakPCGPoint (LocalVariables|GetInPtB))) (Math|Vector|SelectVector (.scale (Utilities|Struct|BreakPCGPoint (LocalVariables|GetInPtB))) (/ "1.000000,1.000000,1.000000" (.scale (Utilities|Struct|BreakPCGPoint (LocalVariables|GetInPtA)))) (Variables|Default|GetApplyScaleFromTargetPoints))) (Utilities|Struct|BreakPCGPoint (LocalVariables|GetInPtA)))))
  (PCG|Metadata|InitializeMetadata _structout (LocalVariables|GetOutMeta) (Utilities|Select :Index (Variables|Default|GetAttributeInheritance) :NewEnumerator0 (LocalVariables|GetInPtB) :NewEnumerator1 (LocalVariables|GetInPtA)))
  (PCG|Helpers|SetSeedfromPosition _structout)
  (return _structout true))
