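; Custom PCG Blueprint element, filed under Deprecated, that adds a named Float, Vector or Vector4 attribute with a default value to point data.
; It initializes new point data from the first spatial input only, creates the attribute on its metadata, then runs a point loop over that input converted to points.
; The loop copies each point, inheriting metadata when KeepExistingAttributes is set.
; The node title displays the default value.
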(fn ExecuteWithContext
  (params (InContext :type PCGContext%20Structure%20%28by%20ref%29) (Input :type PCGData%20Collection%20Structure%20%28by%20ref%29))
  (bind _attribute_name (Variables|Default|GetAttribute%20Name))
  (LocalVariables|SetContext InContext)
  (bind _returnvalue (Data|GetAllSpatialInputs Input))
  (Utilities|FlowControl|Sequence
    (:then_0
      (bind _aspcgspatial_data (Utilities|Casting|CastToPCGSpatialData (PCG|TaggedData|BreakTaggedData (Utilities|Array|Get(aref) _returnvalue))))
      (bind _returnvalue_2 (Game|ConstructObjectfromClass "/Script/PCG.PCGPointData"))
      (SpatialData|InitializefromData _returnvalue_2 _aspcgspatial_data 0 (Variables|Default|GetKeepExistingAttributes) true)
      (bind _returnvalue_3 (Metadata|MutableMetadata _returnvalue_2))
      (switch Utilities|FlowControl|Switch|SwitchonPCGAttributeTypes (Variables|Default|GetAttribute%20Type)
        (:NewEnumerator0
          (PCG|Metadata|CreateFloatAttribute _returnvalue_3 _attribute_name (Variables|Default|GetDefaultFloat) true true))
        (:NewEnumerator1
          (PCG|Metadata|CreateVectorAttribute _returnvalue_3 _attribute_name (Variables|Default|GetDefaultVector) true true))
        (:NewEnumerator2
          (PCG|Metadata|CreateVector4Attribute _returnvalue_3 _attribute_name (Variables|Default|GetDefaultVector4) true true))))
    (:then_1
      (bind _outdata (PCG|FlowControl|PointLoop (LocalVariables|GetContext) (SpatialData|ToPointDatawithContext _aspcgspatial_data (LocalVariables|GetContext)) _returnvalue_2))
      (return (Utilities|Struct|MakePCGDataCollection (Utilities|Array|MakeArray (PCG|TaggedData|MakeTaggedData _outdata)))))))

(fn NodeColorOverride
  (return "(R=0.949020,G=0.949020,B=0.949020,A=1.000000)"))

(fn NodeTitleOverride
  (bind (_x _y _z _w) (Math|Vector4|BreakVector4 (Variables|Default|GetDefaultVector4)))
  (return (Utilities|String|StringToName (Utilities|Select :Index (Variables|Default|GetAttribute%20Type) :NewEnumerator0 (Utilities|String|ToString(Float) (Variables|Default|GetDefaultFloat)) :NewEnumerator1 (Utilities|String|Append (Utilities|String|ToString(Float) (.x (Variables|Default|GetDefaultVector))) "," (Utilities|String|ToString(Float) (.y (Variables|Default|GetDefaultVector))) "," (Utilities|String|ToString(Float) (.z (Variables|Default|GetDefaultVector)))) :NewEnumerator2 (Utilities|String|Append (Utilities|String|ToString(Float) _x) "," (Utilities|String|ToString(Float) _y) "," (Utilities|String|ToString(Float) _z) "," (Utilities|String|ToString(Float) _w))))))

(fn PointLoopBody
  (params (InContext :type PCGContext%20Structure%20%28by%20ref%29) (InData :type PCGPoint%20Data%20Object%20Reference) (InPoint :type PCGPoint%20Structure%20%28by%20ref%29) (OutMetadata :type PCGMetadata%20Object%20Reference) (Iteration :type Integer64))
  (bind (_transform _density _boundsmin _boundsmax _color _steepness _seed) (Utilities|Struct|BreakPCGPoint InPoint))
  (bind _pcgpoint (Utilities|Struct|MakePCGPoint _transform _density _boundsmin _boundsmax _color _steepness _seed))
  (if (Variables|Default|GetKeepExistingAttributes)
    (PCG|Metadata|InitializeMetadata _pcgpoint OutMetadata InPoint (Metadata|ConstMetadata InData)))
  (return _pcgpoint true))
