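; Deprecated PCG Blueprint element that outputs a param data object holding one named attribute with a constant value.
; Execute constructs a PCGParamData, adds one metadata entry, then switches on AttributeType to create and set a float, vector, Vector4 or int64 attribute.
; All types derive from a single Vector4 Value, with float and integer reading only X.
; Title and color overrides show the value on an orange node.
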(fn Execute
  (params (Input :type PCGData%20Collection%20Structure%20%28by%20ref%29))
  (bind _value (Variables|Default|GetValue))
  (bind _x (Math|Vector4|BreakVector4 _value))
  (bind _value_2 (Variables|Default|GetValue))
  (bind _returnvalue (Math|Conversions|ToVector(Vector4) _value_2))
  (bind _value_3 (Variables|Default|GetValue))
  (bind _x_2 (Math|Vector4|BreakVector4 _value_3))
  (bind _returnvalue_2 (Math|Conversions|ToInteger64(Float) _x_2))
  (bind _attributename (Variables|Default|GetAttributeName))
  (bind _metadata (LocalVariables|GetMetadata))
  (bind _metadata_2 (LocalVariables|GetMetadata))
  (bind _value_4 (Variables|Default|GetValue))
  (bind _attributename_2 (Variables|Default|GetAttributeName))
  (bind _attributename_3 (Variables|Default|GetAttributeName))
  (bind _attributename_4 (Variables|Default|GetAttributeName))
  (bind _metadata_3 (LocalVariables|GetMetadata))
  (bind _metadata_4 (LocalVariables|GetMetadata))
  (bind _returnvalue_3 (Game|ConstructObjectfromClass "/Script/PCG.PCGParamData"))
  (bind _returnvalue_4 (Metadata|MutableMetadata _returnvalue_3))
  (bind _output_get (LocalVariables|SetMetadata _returnvalue_4))
  (bind _returnvalue_5 (PCG|Metadata|AddEntry _output_get -1))
  (LocalVariables|SetMetadataEntryKey _returnvalue_5)
  (switch Utilities|FlowControl|Switch|SwitchonPCGAttributeTypes (Variables|Default|GetAttributeType)
    (:NewEnumerator0
      (PCG|Metadata|CreateFloatAttribute _metadata_3 _attributename _x false true)
      (PCG|Metadata|SetFloatAttributebyMetadataKey (LocalVariables|GetMetadataEntryKey) _metadata_3 _attributename _x))
    (:NewEnumerator1
      (PCG|Metadata|CreateVectorAttribute _metadata_2 _attributename_3 _returnvalue false true)
      (PCG|Metadata|SetVectorAttributebyMetadataKey (LocalVariables|GetMetadataEntryKey) _metadata_2 _attributename_3 _returnvalue))
    (:NewEnumerator2
      (PCG|Metadata|CreateVector4Attribute _metadata_4 _attributename_2 _value_4 false true)
      (PCG|Metadata|SetVector4AttributebyMetadataKey (LocalVariables|GetMetadataEntryKey) _metadata_4 _attributename_2 _value_4))
    (:NewEnumerator3
      (PCG|Metadata|CreateInteger64Attribute _metadata _attributename_4 _returnvalue_2 false true)
      (PCG|Metadata|SetInteger64AttributebyMetadataKey (LocalVariables|GetMetadataEntryKey) _metadata _attributename_4 _returnvalue_2)))
  (return (Utilities|Struct|MakePCGDataCollection (Utilities|Array|MakeArray (PCG|TaggedData|MakeTaggedData _returnvalue_3)))))

(fn NodeColorOverride
  (return "(R=1.000000,G=0.600000,B=0.000000,A=1.000000)"))

(fn NodeTitleOverride
  (bind (_x _y _z _w) (Math|Vector4|BreakVector4 (Variables|Default|GetValue)))
  (return (Utilities|String|StringToName (Utilities|String|Append "Param " (Utilities|Select :Index (Variables|Default|GetAttributeType) :NewEnumerator0 (Utilities|String|ToString(Float) _x) :NewEnumerator1 (Utilities|String|Append (Utilities|String|ToString(Float) _x) "," (Utilities|String|ToString(Float) _y) "," (Utilities|String|ToString(Float) _z)) :NewEnumerator2 (Utilities|String|Append (Utilities|String|Append (Utilities|String|ToString(Float) _x) "," (Utilities|String|ToString(Float) _y) "," (Utilities|String|ToString(Float) _z)) "," (Utilities|String|ToString(Float) _w)) :NewEnumerator3 (Utilities|String|ToString(Integer64) (Math|Conversions|ToInteger64(Float) _x)))))))
