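; Deprecated PCG Blueprint element that creates point data, titled as replaceable by the native version and colored red.
; ExecuteWithContext builds points from a point or position array, or a single point or position repeated Count times, then runs an iteration loop.
; The loop writes Index and MaxIndex attributes and optionally culls points by actor-data density.
; Local mode applies owner location and rotation, adjusts seeds, and disables caching.
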(fn ExecuteWithContext
  (params (InContext :type PCGContext%20Structure%20%28by%20ref%29) (Input :type PCGData%20Collection%20Structure%20%28by%20ref%29))
  (bind _pointstoloopon (LocalVariables|GetPointsToLoopOn))
  (bind _returnvalue (Utilities|Array|Length _pointstoloopon))
  (bind _positionstocreatepoints (Variables|Default|GetPositionsToCreatePoints))
  (bind _returnvalue_2 (PCG|Temporary|GetActorData InContext))
  (bind _aspcgspatial_data (Utilities|Casting|CastToPCGSpatialData _returnvalue_2))
  (LocalVariables|SetTarget%20Actor (Variables|/Script/PCG.PCGSpatialData|GetTargetActor _aspcgspatial_data))
  (if (Utilities|Select :Index (Variables|Default|GetPointCreationMethod) :NewEnumerator0 false :NewEnumerator1 true :NewEnumerator2 false :NewEnumerator3 true)
    (Utilities|Array|Clear (LocalVariables|GetPointsToLoopOn))
    (if (not (Utilities|Enum|Equal(Enum) (Variables|Default|GetPointCreationMethod) "NewEnumerator1"))
      (Utilities|Array|Clear _positionstocreatepoints)
      (for _ (range (Variables|Default|GetCount))
        (Utilities|Array|Add _positionstocreatepoints (Variables|Default|GetSinglePosition))))
    (for (_ _array_index) (Variables|Default|GetPositionsToCreatePoints)
      (Utilities|Array|Add (LocalVariables|GetPointsToLoopOn) (Utilities|Struct|MakePCGPoint (Math|Transform|MakeTransform (Utilities|Array|Get(aref) (Variables|Default|GetPositionsToCreatePoints) _array_index)))))
    (elif (Utilities|Enum|Equal(Enum) (Variables|Default|GetPointCreationMethod) "NewEnumerator0")
      (LocalVariables|SetPointsToLoopOn (Variables|Default|GetPointsToCreate))
      (else
        (Utilities|Array|Clear (LocalVariables|GetPointsToLoopOn))
        (for _ (range (Variables|Default|GetCount))
          (Utilities|Array|Add (LocalVariables|GetPointsToLoopOn) (Variables|Default|GetSinglePoint))))))
  (bind _returnvalue_3 (Game|ConstructObjectfromClass "/Script/PCG.PCGPointData"))
  (bind _output_get (Variables|Hidden|SetPtData _returnvalue_3))
  (SpatialData|InitializefromData _output_get _aspcgspatial_data 0 true true)
  (bind _returnvalue_4 (Metadata|MutableMetadata (Variables|Hidden|GetPtData)))
  (PCG|Metadata|CreateFloatAttribute _returnvalue_4 "Index" 0.0 false true)
  (PCG|Metadata|CreateFloatAttribute _returnvalue_4 "MaxIndex" 0.0 false true)
  (SpatialData|SetPoints (Variables|Hidden|GetPtData) _pointstoloopon)
  (Variables|Hidden|SetPtCount _returnvalue)
  (bind _outdata (PCG|FlowControl|IterationLoop InContext (Math|Conversions|ToInteger64(Integer) _returnvalue) 0 0 (Variables|Hidden|GetPtData)))
  (return (Utilities|Struct|MakePCGDataCollection (Utilities|Array|MakeArray (PCG|TaggedData|MakeTaggedData _outdata)))))

(fn IsCacheableOverride
  (return (not (Variables|Default|GetLocal))))

(fn IterationLoopBody
  (params (InContext :type PCGContext%20Structure%20%28by%20ref%29) (Iteration :type Integer64) (InA :type PCGSpatial%20Data%20Object%20Reference) (InB :type PCGSpatial%20Data%20Object%20Reference) (OutMetadata :type PCGMetadata%20Object%20Reference))
  (bind _inpoint (LocalVariables|GetinPoint))
  (bind (_transform _boundsmin _boundsmax _seed) (Utilities|Struct|BreakPCGPoint _inpoint))
  (bind _index (LocalVariables|Getindex))
  (bind _returnvalue (Math|Conversions|ToInteger(Integer64) _index))
  (bind _ptdata (Variables|Hidden|GetPtData))
  (bind _returnvalue_2 (SpatialData|GetPoint _ptdata _returnvalue))
  (LocalVariables|Setcontext InContext)
  (LocalVariables|Setindex Iteration)
  (LocalVariables|SetmaxIndex (Math|Conversions|ToInteger64(Integer) (Variables|Hidden|GetPtCount)))
  (LocalVariables|SetoutMeta OutMetadata)
  (LocalVariables|SetinPoint _returnvalue_2)
  (bind _returnvalue_3 (PCG|Temporary|GetActorData (LocalVariables|Getcontext)))
  (bind _aspcgspatial_data (Utilities|Casting|CastToPCGSpatialData _returnvalue_3))
  (LocalVariables|SetInLoopInContextSpatialData _aspcgspatial_data)
  (if (Variables|Default|GetLocal)
    (bind _returnvalue_4 (PCG|Temporary|GetOriginalComponent (LocalVariables|Getcontext)))
    (bind _returnvalue_5 (PCG|Helpers|ComputeSeedfromPosition (.location (Math|Transform|ComposeTransforms _transform (Math|Transform|MakeTransform (.location (Transformation|GetActorTransform (Components|GetOwner _returnvalue_4))) (.rotation (Transformation|GetActorTransform (Components|GetOwner _returnvalue_4))))))))
    (bind _structout (Utilities|Struct|SetmembersinPCGPoint _returnvalue_2 (Math|Transform|ComposeTransforms _transform (Math|Transform|MakeTransform (.location (Transformation|GetActorTransform (Components|GetOwner _returnvalue_4))) (.rotation (Transformation|GetActorTransform (Components|GetOwner _returnvalue_4))))) (Math|Integer|SelectInt (Math|Conversions|ToInteger(Integer64) Iteration) (Math|Integer|SelectInt _returnvalue_5 _seed (Variables|Default|GetComputeSeedFromPosition)) (and (or (Utilities|Enum|Equal(Enum) (Variables|Default|GetPointCreationMethod) "NewEnumerator2") (Utilities|Enum|Equal(Enum) (Variables|Default|GetPointCreationMethod) "NewEnumerator3")) (> (Variables|Default|GetCount) 0)))))
    (PCG|Metadata|SetFloatAttribute _structout (LocalVariables|GetoutMeta) "Index" (Math|Conversions|ToFloat(Integer64) (LocalVariables|Getindex)))
    (PCG|Metadata|SetFloatAttribute _structout (LocalVariables|GetoutMeta) "MaxIndex" (Math|Conversions|ToFloat(Integer64) (LocalVariables|GetmaxIndex)))
    (return _structout (select (Variables|Default|GetCullPointsOutsideVolume) (> (Distribution|GetDensityatPosition (LocalVariables|GetInLoopInContextSpatialData) (.location (Math|Transform|ComposeTransforms _transform (Math|Transform|MakeTransform (.location (Transformation|GetActorTransform (Components|GetOwner _returnvalue_4))) (.rotation (Transformation|GetActorTransform (Components|GetOwner _returnvalue_4))))))) 0) true))
    (else
      (PCG|Metadata|SetFloatAttribute (LocalVariables|GetinPoint) (LocalVariables|GetoutMeta) "Index" (Math|Conversions|ToFloat(Integer64) (LocalVariables|Getindex)))
      (PCG|Metadata|SetFloatAttribute (LocalVariables|GetinPoint) (LocalVariables|GetoutMeta) "MaxIndex" (Math|Conversions|ToFloat(Integer64) (LocalVariables|GetmaxIndex)))
      (return (LocalVariables|GetinPoint) (select (Variables|Default|GetCullPointsOutsideVolume) (> (Distribution|GetDensityatPosition (LocalVariables|GetInLoopInContextSpatialData) (.location (Utilities|Struct|BreakPCGPoint (LocalVariables|GetinPoint)))) 0) true)))))

(fn NodeColorOverride
  (return "(R=1.000000,G=0.000000,B=0.000000,A=1.000000)"))

(fn NodeTitleOverride
  (return "Deprecated Create Points BP - Replace With Native Version"))
