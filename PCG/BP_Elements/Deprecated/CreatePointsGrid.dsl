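; Deprecated PCG Blueprint element that generates a grid of points, shown in red with a title directing replacement by the native version.
; Execute derives cell counts from extents and cell size, then runs an iteration loop over a 3D volume or selected coordinate plane.
; Each iteration computes a centered position, optionally applies owner location and rotation, culls by spatial density, and seeds by position.
; Results are cacheable only outside local mode.
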(fn ExecuteWithContext
  (params (InContext :type PCGContext%20Structure%20%28by%20ref%29) (Input :type PCGData%20Collection%20Structure%20%28by%20ref%29))
  (bind _cell_size (Variables|Default|GetCell%20Size))
  (bind _gridextents (Variables|Default|GetGridExtents))
  (bind _returnvalue (* _gridextents 2.0))
  (bind _returnvalue_2 (/ _returnvalue _cell_size))
  (bind _returnvalue_3 (Math|Float|Truncate(Vector) _returnvalue_2))
  (bind (_x _y _z) (Utilities|Struct|BreakIntVector _returnvalue_3))
  (bind _output_get (LocalVariables|SetContext InContext))
  (bind _returnvalue_4 (PCG|Temporary|GetActorData _output_get))
  (bind _aspcgspatial_data (Utilities|Casting|CastToPCGSpatialData _returnvalue_4))
  (Variables|Private|SetContext%20PCGSpatial%20Data _aspcgspatial_data)
  (LocalVariables|SetTarget%20Actor (Variables|/Script/PCG.PCGSpatialData|GetTargetActor _aspcgspatial_data))
  (Utilities|Array|Clear (LocalVariables|GetOutPoints))
  (Variables|Private|SetIncrementSize (Math|Vector|VectorGetAbs _cell_size))
  (Variables|Private|SetIncrements _returnvalue_3)
  (bind _outdata (PCG|FlowControl|IterationLoop (LocalVariables|GetContext) (Math|Conversions|ToInteger64(Integer) (Math|Integer|SelectInt (Utilities|Operators|Multiply _x _y _z) (Utilities|Select :Index (Variables|Default|GetCoordinate%20Plane%20Axes) :NewEnumerator0 (* _x _y) :NewEnumerator2 (* _x _z) :NewEnumerator3 (* _y _z)) (Variables|Default|GetVolume))) 0 0 0))
  (Variables|/Script/PCG.PCGSpatialData|SetTargetActor (LocalVariables|GetTarget%20Actor) _outdata)
  (return (Utilities|Struct|MakePCGDataCollection (Utilities|Array|MakeArray (PCG|TaggedData|MakeTaggedData _outdata)))))

(fn IsCacheableOverride
  (return (not (Variables|Default|GetLocal))))

(fn IterationLoopBody
  (params (InContext :type PCGContext%20Structure%20%28by%20ref%29) (Iteration :type Integer64) (InA :type PCGSpatial%20Data%20Object%20Reference) (InB :type PCGSpatial%20Data%20Object%20Reference) (OutMetadata :type PCGMetadata%20Object%20Reference))
  (bind _increments (Variables|Private|GetIncrements))
  (bind (_x _y _z) (Utilities|Struct|BreakIntVector _increments))
  (bind _volume (Variables|Default|GetVolume))
  (bind _increments_2 (Variables|Private|GetIncrements))
  (bind _coordinate_plane_axes (Variables|Default|GetCoordinate%20Plane%20Axes))
  (bind _intvector (Utilities|Struct|MakeIntVector _y _z))
  (bind (_x_2 _y_2 _z_2) (Utilities|Struct|BreakIntVector _increments))
  (bind _intvector_2 (Utilities|Struct|MakeIntVector _x_2 _z_2))
  (bind _returnvalue (Utilities|Select :Index _coordinate_plane_axes :NewEnumerator0 _increments :NewEnumerator2 _intvector_2 :NewEnumerator3 _intvector))
  (bind _returnvalue_2 (select _volume _increments_2 _returnvalue))
  (bind (_x_3 _y_3) (Utilities|Struct|BreakIntVector _returnvalue_2))
  (bind _returnvalue_3 (Math|Conversions|ToInteger(Integer64) Iteration))
  (bind _gridcenterposition (Variables|Default|GetGridCenterPosition))
  (bind _volume_2 (Variables|Default|GetVolume))
  (bind _coordinate_plane_axes_2 (Variables|Default|GetCoordinate%20Plane%20Axes))
  (bind _returnvalue_4 (Utilities|Select :Index _coordinate_plane_axes_2 :NewEnumerator0 "1.000000,1.000000, 0" :NewEnumerator2 "1.000000,0.000000,1.000000" :NewEnumerator3 "0.000000,1.000000,1.000000"))
  (bind _cell_size (Variables|Default|GetCell%20Size))
  (bind _increments_3 (Variables|Private|GetIncrements))
  (bind _returnvalue_5 (* _cell_size _increments_3))
  (bind _returnvalue_6 (- _returnvalue_5 _cell_size))
  (bind _returnvalue_7 (/ _returnvalue_6 2.0))
  (bind _returnvalue_8 (* _x_3 _y_3))
  (bind _returnvalue_9 (/ _returnvalue_3 _returnvalue_8))
  (bind _incrementsize (Variables|Private|GetIncrementSize))
  (bind _returnvalue_10 (* _incrementsize "0.000000,0.000000,1.000000"))
  (bind _returnvalue_11 (* _returnvalue_10 _returnvalue_9))
  (bind _returnvalue_12 (/ _returnvalue_3 _x_3))
  (bind _returnvalue_13 (Math|Integer|%25%28Integer%29 _returnvalue_12 _y_3))
  (bind _volume_3 (Variables|Default|GetVolume))
  (bind _coordinate_plane_axes_3 (Variables|Default|GetCoordinate%20Plane%20Axes))
  (bind _returnvalue_14 (Utilities|Select :Index _coordinate_plane_axes_3 :NewEnumerator0 "0,1.000000, 0" :NewEnumerator2 "0, 0,1.000000" :NewEnumerator3 "0, 0,1.000000"))
  (bind _returnvalue_15 (Math|Vector|SelectVector "0,1.000000, 0" _returnvalue_14 _volume_3))
  (bind _returnvalue_16 (* _incrementsize _returnvalue_15))
  (bind _returnvalue_17 (* _returnvalue_16 _returnvalue_13))
  (bind _returnvalue_18 (Math|Integer|%25%28Integer%29 _returnvalue_3 _x_3))
  (bind _returnvalue_19 (Utilities|Select :Index _coordinate_plane_axes_3 :NewEnumerator0 "1.000000, 0, 0" :NewEnumerator2 "1.000000,0.000000, 0" :NewEnumerator3 "0,1.000000, 0"))
  (bind _returnvalue_20 (Math|Vector|SelectVector "1.000000, 0, 0" _returnvalue_19 _volume_3))
  (bind _returnvalue_21 (* _incrementsize _returnvalue_20))
  (bind _returnvalue_22 (* _returnvalue_21 _returnvalue_18))
  (bind _returnvalue_23 (+ _returnvalue_22 _returnvalue_17))
  (bind _returnvalue_24 (+ _returnvalue_23 _returnvalue_11))
  (bind _returnvalue_25 (- _returnvalue_24 _returnvalue_7))
  (bind _returnvalue_26 (* _returnvalue_25 _returnvalue_4))
  (bind _returnvalue_27 (Math|Vector|SelectVector _returnvalue_25 _returnvalue_26 _volume_2))
  (bind _returnvalue_28 (+ _returnvalue_27 _gridcenterposition))
  (bind _returnvalue_29 (Math|Transform|MakeTransform _returnvalue_28))
  (if (Variables|Default|GetLocal)
    (bind _returnvalue_30 (PCG|Temporary|GetOriginalComponent InContext))
    (if (select (Variables|Default|GetCullPointsOutsideVolume) (> (Distribution|GetDensityatPosition (Variables|Private|GetContext%20PCGSpatial%20Data) (.location (Math|Transform|ComposeTransforms _returnvalue_29 (Math|Transform|MakeTransform (.location (Transformation|GetActorTransform (Components|GetOwner _returnvalue_30))) (.rotation (Transformation|GetActorTransform (Components|GetOwner _returnvalue_30))))))) 0) true)
      (if (Variables|Default|GetLocal%20Inverse%20Transform)
        (bind _returnvalue_31 (PCG|Helpers|ComputeSeedfromPosition (.location _returnvalue_29)))
        (return (Utilities|Struct|MakePCGPoint _returnvalue_29 "-1.000000,-1.000000,-1.000000" "1.000000,1.000000,1.000000" _returnvalue_31) true)
        (else
          (bind _returnvalue_32 (PCG|Helpers|ComputeSeedfromPosition (.location (Math|Transform|ComposeTransforms _returnvalue_29 (Math|Transform|MakeTransform (.location (Transformation|GetActorTransform (Components|GetOwner _returnvalue_30))) (.rotation (Transformation|GetActorTransform (Components|GetOwner _returnvalue_30))))))))
          (return (Utilities|Struct|MakePCGPoint (Math|Transform|ComposeTransforms _returnvalue_29 (Math|Transform|MakeTransform (.location (Transformation|GetActorTransform (Components|GetOwner _returnvalue_30))) (.rotation (Transformation|GetActorTransform (Components|GetOwner _returnvalue_30))))) "-1.000000,-1.000000,-1.000000" "1.000000,1.000000,1.000000" _returnvalue_32) true))))
    (else
      (bind _returnvalue_33 (PCG|Helpers|ComputeSeedfromPosition _returnvalue_28))
      (return (Utilities|Struct|MakePCGPoint _returnvalue_29 "-1.000000,-1.000000,-1.000000" "1.000000,1.000000,1.000000" _returnvalue_33) (select (Variables|Default|GetCullPointsOutsideVolume) (> (Distribution|GetDensityatPosition (Variables|Private|GetContext%20PCGSpatial%20Data) _returnvalue_28) 0) true)))))

(fn NodeColorOverride
  (return "(R=1.000000,G=0.000000,B=0.000000,A=1.000000)"))

(fn NodeTitleOverride
  (return "Deprecated Create Points Grid BP - Replace With Native Version"))
