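; Deprecated PCG Blueprint element, titled "Debug Print Attributes", that prints metadata attribute values on screen and to the log for debugging.
; It takes metadata from the first param input, then from the first point data input, whose point count it prints and whose PointIndex point supplies the entry key.
; A type switch per attribute prints Float, Integer64, Vector and Vector4 with prefix, name and type, and String as the bare value.
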(fn Execute
  (params (Input :type PCGData%20Collection%20Structure%20%28by%20ref%29))
  (bind _inmetadata (LocalVariables|GetInMetadata))
  (bind (_attributenames _attributetypes) (PCG|Metadata|GetAttributes _inmetadata))
  (bind _text_color (Variables|Default|GetText%20Color))
  (bind _inmetadata_2 (LocalVariables|GetInMetadata))
  (bind _prefix (Variables|Default|GetPrefix))
  (bind _duration (Variables|Default|GetDuration))
  (bind _metadataentrykey (LocalVariables|GetMetadataEntryKey))
  (Utilities|FlowControl|Sequence
    (:then_0
      (bind _returnvalue (Data|GetParams Input))
      (bind _aspcgattribute_set (Utilities|Casting|CastToPCGParamData (PCG|TaggedData|BreakTaggedData (Utilities|Array|Get(aref) _returnvalue))))
      (LocalVariables|SetInMetadata (Metadata|ConstMetadata _aspcgattribute_set)))
    (:then_1
      (bind _returnvalue_2 (Data|GetAllSpatialInputs Input))
      (bind _aspcgpoint_data (Utilities|Casting|CastToPCGPointData (PCG|TaggedData|BreakTaggedData (Utilities|Array|Get(aref) _returnvalue_2))))
      (LocalVariables|SetInMetadata (Metadata|ConstMetadata _aspcgpoint_data))
      (LocalVariables|SetMetadataEntryKey (.MetadataEntry (Utilities|Struct|BreakPCGPoint (SpatialData|GetPoint _aspcgpoint_data (Variables|Default|GetPointIndex)))))
      (Development|PrintString (Utilities|String|Append (Utilities|String|ToString(Integer) (Utilities|Array|Length (SpatialData|GetPoints _aspcgpoint_data))) " pts") true true (Variables|Default|GetText%20Color))))
  (for _index (range (Utilities|Array|Length _attributenames))
    (switch Utilities|FlowControl|Switch|SwitchonEPCGMetadataTypes (Utilities|Array|Get(aref) _attributetypes _index)
      (:Float
        (bind _returnvalue_3 (PCG|Metadata|GetFloatAttributebyMetadataKey _metadataentrykey _inmetadata_2 (Utilities|Array|Get(aref) _attributenames _index)))
        (Development|PrintString (Utilities|String|Append _prefix " " (Utilities|String|ToString(Name) (Utilities|Array|Get(aref) _attributenames _index)) " " (Utilities|String|EnumtoString (Utilities|Array|Get(aref) _attributetypes _index)) " " (Utilities|String|ToString(Float) _returnvalue_3)) true true _text_color _duration))
      (:Integer64
        (bind _returnvalue_4 (PCG|Metadata|GetInteger64AttributebyMetadataKey _metadataentrykey _inmetadata_2 (Utilities|Array|Get(aref) _attributenames _index)))
        (Development|PrintString (Utilities|String|Append _prefix " " (Utilities|String|ToString(Name) (Utilities|Array|Get(aref) _attributenames _index)) " " (Utilities|String|EnumtoString (Utilities|Array|Get(aref) _attributetypes _index)) " " (Utilities|String|ToString(Integer64) _returnvalue_4)) true true _text_color _duration))
      (:Vector
        (bind _returnvalue_5 (PCG|Metadata|GetVectorAttributebyMetadataKey _metadataentrykey _inmetadata_2 (Utilities|Array|Get(aref) _attributenames _index)))
        (Development|PrintString (Utilities|String|Append _prefix " " (Utilities|String|ToString(Name) (Utilities|Array|Get(aref) _attributenames _index)) " " (Utilities|String|EnumtoString (Utilities|Array|Get(aref) _attributetypes _index)) " " (Utilities|String|ToString(Vector) _returnvalue_5)) true true _text_color _duration))
      (:Vector4
        (bind _returnvalue_6 (PCG|Metadata|GetVector4AttributebyMetadataKey _metadataentrykey _inmetadata_2 (Utilities|Array|Get(aref) _attributenames _index)))
        (Development|PrintString (Utilities|String|Append _prefix " " (Utilities|String|ToString(Name) (Utilities|Array|Get(aref) _attributenames _index)) " " (Utilities|String|EnumtoString (Utilities|Array|Get(aref) _attributetypes _index)) " " (Utilities|String|ToString(Float) (Math|Vector4|BreakVector4 _returnvalue_6)) "," (Utilities|String|ToString(Float) (.Y (Math|Vector4|BreakVector4 _returnvalue_6))) "," (Utilities|String|ToString(Float) (.Z (Math|Vector4|BreakVector4 _returnvalue_6))) "," (Utilities|String|ToString(Float) (.W (Math|Vector4|BreakVector4 _returnvalue_6)))) true true _text_color _duration))
      (:String
        (bind _returnvalue_7 (PCG|Metadata|GetStringAttributebyMetadataKey _metadataentrykey _inmetadata_2 (Utilities|Array|Get(aref) _attributenames _index)))
        (Development|PrintString _returnvalue_7))))
  (return 0))

(fn NodeColorOverride
  (return 0))

(fn NodeTitleOverride
  (return "Debug Print Attributes"))
