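; Deprecated PCG Blueprint element that applies random noise to the density of points from the first spatial input.
; PointLoop seeds a random stream from each point's position, rewriting its seed and drawing noise between DensityNoiseMin and DensityNoiseMax.
; ConserveMinMaxDensityFromInput optionally scales noise to zero at density extremes.
; DensityMode selects how noise sets or combines with density before clamping; the output keeps the input's target actor.
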(fn ExecuteWithContext
  (params (InContext :type PCGContext%20Structure%20%28by%20ref%29) (Input :type PCGData%20Collection%20Structure%20%28by%20ref%29))
  (bind _returnvalue (Data|GetAllSpatialInputs Input))
  (bind _aspcgspatial_data (Utilities|Casting|CastToPCGSpatialData (PCG|TaggedData|BreakTaggedData (Utilities|Array|Get(aref) _returnvalue))))
  (bind _outdata (PCG|FlowControl|PointLoop InContext (SpatialData|ToPointDatawithContext _aspcgspatial_data InContext) 0))
  (Variables|/Script/PCG.PCGSpatialData|SetTargetActor (Variables|/Script/PCG.PCGSpatialData|GetTargetActor _aspcgspatial_data) _outdata)
  (return (Utilities|Struct|MakePCGDataCollection (Utilities|Array|MakeArray (PCG|TaggedData|MakeTaggedData _outdata)))))

(fn PointLoopBody
  (params (InContext :type PCGContext%20Structure%20%28by%20ref%29) (InData :type PCGPoint%20Data%20Object%20Reference) (InPoint :type PCGPoint%20Structure%20%28by%20ref%29) (OutMetadata :type PCGMetadata%20Object%20Reference) (Iteration :type Integer64))
  (bind (_transform _density) (Utilities|Struct|BreakPCGPoint InPoint))
  (bind _returnvalue (PCG|Helpers|ComputeSeedfromPosition (.location _transform)))
  (bind _structout (Utilities|Struct|SetmembersinPCGPoint InPoint (Math|Float|Clamp(Float) (Utilities|Select :Index (Variables|Default|GetDensityMode) :NewEnumerator0 (Math|Float|SelectFloat (* (- 1.0 (Math|Float|Absolute(Float) (- (* _density 2.0) 1.0))) (Math|Random|RandomFloatinRangefromStream (Math|Random|MakeRandomStream _returnvalue) (Variables|Default|GetDensityNoiseMin) (Variables|Default|GetDensityNoiseMax))) (Math|Random|RandomFloatinRangefromStream (Math|Random|MakeRandomStream _returnvalue) (Variables|Default|GetDensityNoiseMin) (Variables|Default|GetDensityNoiseMax)) (Variables|Default|GetConserveMinMaxDensityFromInput)) :NewEnumerator2 (Math|Float|Min(Float) _density (Math|Float|SelectFloat (* (- 1.0 (Math|Float|Absolute(Float) (- (* _density 2.0) 1.0))) (Math|Random|RandomFloatinRangefromStream (Math|Random|MakeRandomStream _returnvalue) (Variables|Default|GetDensityNoiseMin) (Variables|Default|GetDensityNoiseMax))) (Math|Random|RandomFloatinRangefromStream (Math|Random|MakeRandomStream _returnvalue) (Variables|Default|GetDensityNoiseMin) (Variables|Default|GetDensityNoiseMax)) (Variables|Default|GetConserveMinMaxDensityFromInput))) :NewEnumerator3 (Math|Float|Max(Float) _density (Math|Float|SelectFloat (* (- 1.0 (Math|Float|Absolute(Float) (- (* _density 2.0) 1.0))) (Math|Random|RandomFloatinRangefromStream (Math|Random|MakeRandomStream _returnvalue) (Variables|Default|GetDensityNoiseMin) (Variables|Default|GetDensityNoiseMax))) (Math|Random|RandomFloatinRangefromStream (Math|Random|MakeRandomStream _returnvalue) (Variables|Default|GetDensityNoiseMin) (Variables|Default|GetDensityNoiseMax)) (Variables|Default|GetConserveMinMaxDensityFromInput))) :NewEnumerator4 (+ _density (Math|Float|SelectFloat (* (- 1.0 (Math|Float|Absolute(Float) (- (* _density 2.0) 1.0))) (Math|Random|RandomFloatinRangefromStream (Math|Random|MakeRandomStream _returnvalue) (Variables|Default|GetDensityNoiseMin) (Variables|Default|GetDensityNoiseMax))) (Math|Random|RandomFloatinRangefromStream (Math|Random|MakeRandomStream _returnvalue) (Variables|Default|GetDensityNoiseMin) (Variables|Default|GetDensityNoiseMax)) (Variables|Default|GetConserveMinMaxDensityFromInput))) :NewEnumerator5 (- _density (Math|Float|SelectFloat (* (- 1.0 (Math|Float|Absolute(Float) (- (* _density 2.0) 1.0))) (Math|Random|RandomFloatinRangefromStream (Math|Random|MakeRandomStream _returnvalue) (Variables|Default|GetDensityNoiseMin) (Variables|Default|GetDensityNoiseMax))) (Math|Random|RandomFloatinRangefromStream (Math|Random|MakeRandomStream _returnvalue) (Variables|Default|GetDensityNoiseMin) (Variables|Default|GetDensityNoiseMax)) (Variables|Default|GetConserveMinMaxDensityFromInput))) :NewEnumerator6 (* _density (Math|Float|SelectFloat (* (- 1.0 (Math|Float|Absolute(Float) (- (* _density 2.0) 1.0))) (Math|Random|RandomFloatinRangefromStream (Math|Random|MakeRandomStream _returnvalue) (Variables|Default|GetDensityNoiseMin) (Variables|Default|GetDensityNoiseMax))) (Math|Random|RandomFloatinRangefromStream (Math|Random|MakeRandomStream _returnvalue) (Variables|Default|GetDensityNoiseMin) (Variables|Default|GetDensityNoiseMax)) (Variables|Default|GetConserveMinMaxDensityFromInput))) :NewEnumerator7 (/ _density (Math|Float|SelectFloat (* (- 1.0 (Math|Float|Absolute(Float) (- (* _density 2.0) 1.0))) (Math|Random|RandomFloatinRangefromStream (Math|Random|MakeRandomStream _returnvalue) (Variables|Default|GetDensityNoiseMin) (Variables|Default|GetDensityNoiseMax))) (Math|Random|RandomFloatinRangefromStream (Math|Random|MakeRandomStream _returnvalue) (Variables|Default|GetDensityNoiseMin) (Variables|Default|GetDensityNoiseMax)) (Variables|Default|GetConserveMinMaxDensityFromInput))) :NewEnumerator8 (Math|Conversions|ToFloat(Integer) (Math|Float|Round (Math|Float|SelectFloat (* (- 1.0 (Math|Float|Absolute(Float) (- (* _density 2.0) 1.0))) (Math|Random|RandomFloatinRangefromStream (Math|Random|MakeRandomStream _returnvalue) (Variables|Default|GetDensityNoiseMin) (Variables|Default|GetDensityNoiseMax))) (Math|Random|RandomFloatinRangefromStream (Math|Random|MakeRandomStream _returnvalue) (Variables|Default|GetDensityNoiseMin) (Variables|Default|GetDensityNoiseMax)) (Variables|Default|GetConserveMinMaxDensityFromInput)))) :NewEnumerator9 (Math|Conversions|ToFloat(Integer) (Math|Float|Truncate (Math|Float|SelectFloat (* (- 1.0 (Math|Float|Absolute(Float) (- (* _density 2.0) 1.0))) (Math|Random|RandomFloatinRangefromStream (Math|Random|MakeRandomStream _returnvalue) (Variables|Default|GetDensityNoiseMin) (Variables|Default|GetDensityNoiseMax))) (Math|Random|RandomFloatinRangefromStream (Math|Random|MakeRandomStream _returnvalue) (Variables|Default|GetDensityNoiseMin) (Variables|Default|GetDensityNoiseMax)) (Variables|Default|GetConserveMinMaxDensityFromInput)))) :NewEnumerator10 0.0 :NewEnumerator11 0.0 :NewEnumerator12 0.0 :NewEnumerator13 0.0)) _returnvalue))
  (return _structout true))
