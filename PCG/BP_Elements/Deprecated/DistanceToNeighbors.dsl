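; Deprecated PCG Blueprint element, titled to redirect to Point Neighborhood, that sets point density from normalized distance to nearby target data.
; The first input supplies the target, and an optional second supplies the points to process, defaulting to the target.
; A point loop projects each point within a search box, optionally copying color, writing offset vectors and culling outside the actor volume.
; Optional height filtering uses a second, actor-height projection to zero or drop points.
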(fn ExecuteWithContext
  (params (InContext :type PCGContext%20Structure%20%28by%20ref%29) (Input :type PCGData%20Collection%20Structure%20%28by%20ref%29))
  (bind _output_get (LocalVariables|SetContext InContext))
  (bind _returnvalue (PCG|Temporary|GetActorData _output_get))
  (bind _aspcgspatial_data (Utilities|Casting|CastToPCGSpatialData _returnvalue))
  (Variables|Default|SetActorSpatialData _aspcgspatial_data)
  (bind _result (Utilities|Set|ToArray (PCG|Input&Output|CustomInputLabels self)))
  (bind _returnvalue_2 (Data|GetInputsbyPinLabel Input (Utilities|Array|Get(aref) _result)))
  (if (> (Utilities|Array|Length _returnvalue_2) 0)
    (Utilities|FlowControl|Sequence
      (:then_0
        (bind _aspcgspatial_data_2 (Utilities|Casting|CastToPCGSpatialData (PCG|TaggedData|BreakTaggedData (Utilities|Array|Get(aref) _returnvalue_2))))
        (bind _returnvalue_3 (PCG|Temporary|GetComponent (LocalVariables|GetContext)))
        (bind _returnvalue_4 (PCG|Helpers|GetActorBoundsPCG (Components|GetOwner _returnvalue_3)))
        (Variables|Hidden|SetTargetData (SpatialData|ToPointDatawithContext _aspcgspatial_data_2 (LocalVariables|GetContext)))
        (Variables|Hidden|SetSearchBounds-ZUnbound (Math|Box|MakeBox (Math|Vector|MakeVector (* (Variables|Default|GetSearchDistance) -1.0) (* (Variables|Default|GetSearchDistance) -1.0) (.z (Utilities|Struct|BreakBox _returnvalue_4))) (Math|Vector|MakeVector (Variables|Default|GetSearchDistance) (Variables|Default|GetSearchDistance) (.z (.Max (Utilities|Struct|BreakBox _returnvalue_4))))))
        (Variables|Hidden|SetSearchBounds (Math|Box|MakeBox (Math|Vector|MakeVector (* (Variables|Default|GetSearchDistance) -1.0) (* (Variables|Default|GetSearchDistance) -1.0) (* (Variables|Default|GetSearchDistance) -1.0)) (Math|Vector|MakeVector (Variables|Default|GetSearchDistance) (Variables|Default|GetSearchDistance) (Variables|Default|GetSearchDistance))))
        (Variables|Hidden|SetIn%20Data (Variables|Hidden|GetTargetData)))
      (:then_1
        (bind _returnvalue_5 (Data|GetInputsbyPinLabel Input (Utilities|Array|Get(aref) _result 1)))
        (if (> (Utilities|Array|Length _returnvalue_5) 0)
          (bind _aspcgspatial_data_3 (Utilities|Casting|CastToPCGSpatialData (PCG|TaggedData|BreakTaggedData (Utilities|Array|Get(aref) _returnvalue_5))))
          (Variables|Hidden|SetIn%20Data (SpatialData|ToPointDatawithContext _aspcgspatial_data_3 (LocalVariables|GetContext)))))
      (:then_2
        (bind _returnvalue_6 (Game|ConstructObjectfromClass "/Script/PCG.PCGPointData"))
        (LocalVariables|SetOut%20Data _returnvalue_6)
        (SpatialData|InitializefromData (LocalVariables|GetOut%20Data) (Variables|Hidden|GetIn%20Data) 0 true true)
        (if (Variables|Default|GetWriteDistanceToAttribute)
          (bind _returnvalue_7 (Metadata|MutableMetadata (LocalVariables|GetOut%20Data)))
          (PCG|Metadata|CreateVectorAttribute _returnvalue_7 (Variables|Default|GetAttribute%20Name) "0, 0, 0" false true))
        (bind _outdata (PCG|FlowControl|PointLoop (LocalVariables|GetContext) (Variables|Hidden|GetIn%20Data) (LocalVariables|GetOut%20Data)))
        (return (Utilities|Struct|MakePCGDataCollection (Utilities|Array|MakeArray (PCG|TaggedData|MakeTaggedData _outdata))))))))

(fn NodeColorOverride
  (return "(R=1.000000,G=0.000000,B=0.000000,A=1.000000)"))

(fn NodeTitleOverride
  (return "DEPRECATED, Use Point Neighborhood"))

(fn PointLoopBody
  (params (InContext :type PCGContext%20Structure%20%28by%20ref%29) (InData :type PCGPoint%20Data%20Object%20Reference) (InPoint :type PCGPoint%20Structure%20%28by%20ref%29) (OutMetadata :type PCGMetadata%20Object%20Reference) (Iteration :type Integer64))
  (bind _color_blend_mode (Variables|Default|GetColor%20Blend%20Mode))
  (bind _pcgprojectionparams (Utilities|Struct|MakePCGProjectionParams true false false _color_blend_mode))
  (bind _searchbounds (Variables|Hidden|GetSearchBounds))
  (bind _inptxform (LocalVariables|GetinPTXform))
  (bind _targetdata (Variables|Hidden|GetTargetData))
  (bind _outpoint (SpatialData|ProjectPoint _targetdata _inptxform _searchbounds _pcgprojectionparams 0))
  (bind (_transform _density _boundsmin _boundsmax _color) (Utilities|Struct|BreakPCGPoint _outpoint))
  (bind _removelowerpoints (Variables|VerticalFiltering|GetRemoveLowerPoints))
  (bind _projptxform_zunbound (LocalVariables|GetprojPtXForm-ZUnbound))
  (bind _location (.location _projptxform_zunbound))
  (bind _inptxform_2 (LocalVariables|GetinPTXform))
  (bind _location_2 (.location _inptxform_2))
  (bind _returnvalue (- _location_2 _location))
  (bind _z (.z _returnvalue))
  (bind _returnvalue_2 (Math|Float|Sign(Float) _z))
  (bind _returnvalue_3 (Math|Float|Clamp(Float) _returnvalue_2))
  (bind _returnvalue_4 (- 1.0 _returnvalue_3))
  (bind _returnvalue_5 0.0)
  (bind _returnvalue_6 (+ _returnvalue_5 _returnvalue_4))
  (bind _returnvalue_7 (Math|Float|SelectFloat _returnvalue_5 _returnvalue_6 _removelowerpoints))
  (bind _returnvalue_8 (Math|Float|Absolute(Float) _z))
  (bind _minheightthreshold (Variables|VerticalFiltering|GetMinHeightThreshold))
  (bind _returnvalue_9 (> _returnvalue_8 _minheightthreshold))
  (bind _returnvalue_10 (Math|Float|SelectFloat 1.0 0.0 _returnvalue_9))
  (bind _projptxform (LocalVariables|GetprojPtXForm))
  (bind _location_3 (.location _projptxform))
  (bind _inptxform_3 (LocalVariables|GetinPTXform))
  (bind _location_4 (.location _inptxform_3))
  (bind _normalized3ddistance (LocalVariables|GetNormalized3DDistance))
  (bind _returnvalue_11 (Math|Float|Clamp(Float) _normalized3ddistance))
  (bind _removehigherpoints (Variables|VerticalFiltering|GetRemoveHigherPoints))
  (bind _returnvalue_12 (+ _returnvalue_7 _returnvalue_3))
  (bind _returnvalue_13 (Math|Float|SelectFloat _returnvalue_7 _returnvalue_12 _removehigherpoints))
  (bind _normalized3ddistance_2 (LocalVariables|GetNormalized3DDistance))
  (bind _returnvalue_14 (* _returnvalue_13 _normalized3ddistance_2))
  (bind _returnvalue_15 (* _returnvalue_14 _returnvalue_10))
  (bind _inpt (LocalVariables|GetinPt))
  (bind _inpt_2 (LocalVariables|GetinPt))
  (LocalVariables|SetMetaData OutMetadata)
  (LocalVariables|SetinPt InPoint)
  (LocalVariables|SetinPTXform (Math|Transform|MakeTransform (.location (Utilities|Struct|BreakPCGPoint InPoint)) "0, 0, 0" (.scale (Utilities|Struct|BreakPCGPoint InPoint))))
  (LocalVariables|SetprojPtXForm _transform)
  (LocalVariables|SetProjectedColor _color)
  (LocalVariables|SetWorldDistance (- _location_4 _location_3))
  (LocalVariables|SetNormalized3DDistance (/ (Math|Vector|Distance(Vector) _location_4 _location_3) (Variables|Default|GetSearchDistance)))
  (if (Variables|VerticalFiltering|GetFilterByHeight)
    (LocalVariables|SetprojPtXForm-ZUnbound (Utilities|Struct|BreakPCGPoint (SpatialData|ProjectPoint (Variables|Hidden|GetTargetData) (LocalVariables|GetinPTXform) (Variables|Hidden|GetSearchBounds-ZUnbound) (Utilities|Struct|MakePCGProjectionParams true false) 0)))
    (LocalVariables|SetWorldDistance _returnvalue)
    (if (Variables|Default|GetProject%20Colors)
      (bind _structout (Utilities|Struct|SetmembersinPCGPoint _inpt_2 _returnvalue_15 (LocalVariables|GetProjectedColor)))
      (LocalVariables|SetoutPt _structout)
      (else
        (bind _structout_2 (Utilities|Struct|SetmembersinPCGPoint _inpt_2 _returnvalue_15))
        (LocalVariables|SetoutPt _structout_2)))
    (if (Variables|Default|GetWriteDistanceToAttribute)
      (PCG|Metadata|SetVectorAttribute (LocalVariables|GetoutPt) (LocalVariables|GetMetaData) (Variables|Default|GetAttribute%20Name) (LocalVariables|GetWorldDistance)))
    (return (LocalVariables|GetoutPt) (and (select (Variables|Default|GetCullPointsOutsideVolume) (> (Distribution|GetDensityatPosition (Variables|Default|GetActorSpatialData) (.location (Utilities|Struct|BreakPCGPoint (LocalVariables|GetoutPt)))) 0) true) (or (> _returnvalue_15 0) (Variables|VerticalFiltering|GetKeep%200-Density%20Points))))
    (elif (Variables|Default|GetProject%20Colors)
      (bind _structout_3 (Utilities|Struct|SetmembersinPCGPoint _inpt _returnvalue_11 (LocalVariables|GetProjectedColor)))
      (LocalVariables|SetoutPt _structout_3)
      (if (Variables|Default|GetWriteDistanceToAttribute)
        (PCG|Metadata|SetVectorAttribute (LocalVariables|GetoutPt) (LocalVariables|GetMetaData) (Variables|Default|GetAttribute%20Name) (LocalVariables|GetWorldDistance)))
      (return (LocalVariables|GetoutPt) (select (Variables|Default|GetCullPointsOutsideVolume) (> (Distribution|GetDensityatPosition (Variables|Default|GetActorSpatialData) (.location (Utilities|Struct|BreakPCGPoint (LocalVariables|GetoutPt)))) 0) true))
      (else
        (bind _structout_4 (Utilities|Struct|SetmembersinPCGPoint _inpt _returnvalue_11))
        (LocalVariables|SetoutPt _structout_4)))))
