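; Deprecated PCG Blueprint element that duplicates each input point by a per-point count read from a configurable attribute.
; ExecuteWithContext converts the first spatial input to points, creates output data with Index and LastIndex float attributes, and runs a variable loop.
; The loop body truncates the Float or Double count and emits count + 1 copies.
; Each copy keeps its metadata, offsets the seed by its index, and records Index and LastIndex.
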(fn ExecuteWithContext
  (params (InContext :type PCGContext%20Structure%20%28by%20ref%29) (Input :type PCGData%20Collection%20Structure%20%28by%20ref%29))
  (bind _returnvalue (Data|GetAllSpatialInputs Input))
  (bind _aspcgspatial_data (Utilities|Casting|CastToPCGSpatialData (PCG|TaggedData|BreakTaggedData (Utilities|Array|Get(aref) _returnvalue))))
  (bind _returnvalue_2 (Game|ConstructObjectfromClass "/Script/PCG.PCGPointData"))
  (SpatialData|InitializefromData _returnvalue_2 _aspcgspatial_data 0 true true)
  (bind _returnvalue_3 (Metadata|MutableMetadata _returnvalue_2))
  (PCG|Metadata|CreateFloatAttribute _returnvalue_3 "Index" 0.0 false true)
  (PCG|Metadata|CreateFloatAttribute _returnvalue_3 "LastIndex" 0.0 false true)
  (bind _outdata (PCG|FlowControl|VariableLoop InContext (SpatialData|ToPointDatawithContext _aspcgspatial_data InContext) _returnvalue_2))
  (Utilities|Array|Add (LocalVariables|GetTaggedDataColl) (PCG|TaggedData|MakeTaggedData _outdata))
  (return (Utilities|Struct|MakePCGDataCollection (LocalVariables|GetTaggedDataColl))))

(fn VariableLoopBody
  (params (InContext :type PCGContext%20Structure%20%28by%20ref%29) (InData :type PCGPoint%20Data%20Object%20Reference) (InPoint :type PCGPoint%20Structure%20%28by%20ref%29) (OutMetadata :type PCGMetadata%20Object%20Reference) (Iteration :type Integer64))
  (bind (_transform _density _boundsmin _boundsmax _color _steepness _seed) (Utilities|Struct|BreakPCGPoint (LocalVariables|GetInPt)))
  (bind _returnvalue (Metadata|ConstMetadata InData))
  (bind (_attributenames _attributetypes) (PCG|Metadata|GetAttributes _returnvalue))
  (bind _duplicatescountattribute (Variables|Default|GetDuplicatesCountAttribute))
  (bind _inpt (LocalVariables|GetInPt))
  (LocalVariables|SetInPt InPoint)
  (LocalVariables|SetConstMetaData _returnvalue)
  (switch Utilities|FlowControl|Switch|SwitchonEPCGMetadataTypes (Utilities|Array|Get(aref) _attributetypes (Utilities|Array|FindItem _attributenames (Variables|Default|GetDuplicatesCountAttribute)))
    (:Float
      (bind _returnvalue_2 (PCG|Metadata|GetFloatAttribute _inpt _returnvalue _duplicatescountattribute))
      (LocalVariables|SetCount (Math|Float|Truncate _returnvalue_2)))
    (:Double
      (bind _returnvalue_3 (PCG|Metadata|GetDoubleAttribute _inpt _returnvalue _duplicatescountattribute))
      (LocalVariables|SetCount (Math|Float|Truncate _returnvalue_3))))
  (for _index (range (+ (LocalVariables|GetCount) 1))
    (bind _outpoint (PCG|CopyPoint (LocalVariables|GetInPt) true _returnvalue OutMetadata))
    (bind _structout (Utilities|Struct|SetmembersinPCGPoint _outpoint (+ _seed _index)))
    (PCG|Metadata|SetFloatAttribute _outpoint OutMetadata "Index" (Math|Conversions|ToFloat(Integer) _index))
    (PCG|Metadata|SetFloatAttribute _outpoint OutMetadata "LastIndex" (Math|Conversions|ToFloat(Integer) (LocalVariables|GetCount)))
    (Utilities|Array|Add (LocalVariables|GetPoints) _structout))
  (return (LocalVariables|GetPoints)))
