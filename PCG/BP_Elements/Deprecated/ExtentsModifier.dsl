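; Blueprint PCG element, filed under Deprecated, that modifies the extents of points in a procedural generation graph.
; It converts the first data on its first custom input pin to point data, runs a point loop and carries the target actor onto the output.
; Each point is copied with its extents either replaced or multiplied by Extents Scale, as Use Absolute Scale selects.
; Any further inputs are ignored.
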(fn ExecuteWithContext
  (params (InContext :type PCGContext%20Structure%20%28by%20ref%29) (Input :type PCGData%20Collection%20Structure%20%28by%20ref%29))
  (bind _result (Utilities|Set|ToArray (PCG|Input&Output|CustomInputLabels self)))
  (bind _returnvalue (Data|GetInputsbyPinLabel Input (Utilities|Array|Get(aref) _result)))
  (if (> (Utilities|Array|Length _returnvalue) 0)
    (bind _aspcgspatial_data (Utilities|Casting|CastToPCGSpatialData (PCG|TaggedData|BreakTaggedData (Utilities|Array|Get(aref) _returnvalue))))
    (bind _outdata (PCG|FlowControl|PointLoop InContext (SpatialData|ToPointDatawithContext _aspcgspatial_data InContext) 0))
    (Variables|/Script/PCG.PCGSpatialData|SetTargetActor (Variables|/Script/PCG.PCGSpatialData|GetTargetActor _aspcgspatial_data) _outdata)
    (return (Utilities|Struct|MakePCGDataCollection (Utilities|Array|MakeArray (PCG|TaggedData|MakeTaggedData _outdata))))
    (else
      (return 0))))

(fn PointLoopBody
  (params (InContext :type PCGContext%20Structure%20%28by%20ref%29) (InData :type PCGPoint%20Data%20Object%20Reference) (InPoint :type PCGPoint%20Structure%20%28by%20ref%29) (OutMetadata :type PCGMetadata%20Object%20Reference) (Iteration :type Integer64))
  (bind _returnvalue (PCG|Points|GetExtents InPoint))
  (bind _outpoint (PCG|CopyPoint InPoint true (Metadata|ConstMetadata InData) OutMetadata))
  (PCG|Points|SetExtents _outpoint (Math|Vector|SelectVector (Variables|Default|GetExtents%20Scale) (* _returnvalue (Variables|Default|GetExtents%20Scale)) (Variables|Default|GetUse%20Absolute%20Scale)))
  (return _outpoint true))
