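; Custom PCG Blueprint element, in a Deprecated folder, that strips every metadata attribute from point data except the one named in AttributeToKeep.
; It casts the first spatial input to point data, then copies its points and metadata into a new point data object.
; A reverse loop deletes the other attributes before output, with optional debug printing of input and output attribute names.
; The node title shows the kept attribute.
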(fn ExecuteWithContext
  (params (InContext :type PCGContext%20Structure%20%28by%20ref%29) (Input :type PCGData%20Collection%20Structure%20%28by%20ref%29))
  (LocalVariables|SetContext InContext)
  (bind _returnvalue (Data|GetAllSpatialInputs Input))
  (Utilities|FlowControl|Sequence
    (:then_0
      (bind _aspcgpoint_data (Utilities|Casting|CastToPCGPointData (PCG|TaggedData|BreakTaggedData (Utilities|Array|Get(aref) _returnvalue))))
      (LocalVariables|SetInPointData _aspcgpoint_data))
    (:then_1
      (bind _returnvalue_2 (Game|ConstructObjectfromClass "/Script/PCG.PCGPointData"))
      (bind _output_get (LocalVariables|SetOutPointData _returnvalue_2))
      (SpatialData|InitializefromData _output_get (LocalVariables|GetInPointData) 0 false true)
      (SpatialData|SetPoints (LocalVariables|GetOutPointData) (SpatialData|GetPoints (LocalVariables|GetInPointData)))
      (bind _returnvalue_3 (Metadata|MutableMetadata (LocalVariables|GetOutPointData)))
      (LocalVariables|SetOutMetadata _returnvalue_3)
      (LocalVariables|SetInMetadata (Metadata|ConstMetadata (LocalVariables|GetInPointData)))
      (PCG|Metadata|InitializeAsCopy :self (LocalVariables|GetOutMetadata) :InMetadataToCopy (LocalVariables|GetInMetadata))
      (bind (_arrayindex _arrayelement) (Utilities|Array|ReverseforEachLoop (PCG|Metadata|GetAttributes (LocalVariables|GetOutMetadata)))
        (:LoopBody
          (if (!= _arrayelement (Variables|Default|GetAttributeToKeep))
            (PCG|Metadata|DeleteAttribute (LocalVariables|GetOutMetadata) _arrayelement)))))
    (:then_2
      (Utilities|Array|Add (LocalVariables|GetOutDatas) (PCG|TaggedData|MakeTaggedData (LocalVariables|GetOutPointData)))
      (if (Variables|Hidden|GetDebugPrint)
        (for _array_element (PCG|Metadata|GetAttributes (LocalVariables|GetInMetadata))
          (Development|PrintString (Utilities|String|Append "IN:" (Utilities|String|ToString(Name) _array_element)) true true "(R=0.000000,G=1.000000,B=0.151323,A=1.000000)"))
        (for _array_element_2 (PCG|Metadata|GetAttributes (LocalVariables|GetOutMetadata))
          (Development|PrintString (Utilities|String|Append "OUT:" (Utilities|String|ToString(Name) _array_element_2)) true true "(R=1.000000,G=0.114046,B=0.000000,A=1.000000)"))
        (else
          _))
      (return (Utilities|Struct|MakePCGDataCollection (LocalVariables|GetOutDatas))))))

(fn NodeColorOverride
  (return "(R=0.979167,G=0.979167,B=0.979167,A=1.000000)"))

(fn NodeTitleOverride
  (return (Utilities|String|StringToName (Utilities|String|Append "FilterAttributes: " (Utilities|String|ToString(Name) (Variables|Default|GetAttributeToKeep))))))
