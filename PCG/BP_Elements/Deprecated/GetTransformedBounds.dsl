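; Custom PCG Blueprint element, in a Deprecated folder, that stores each point's transformed bounding box as min and max vector attributes.
; It converts the first input on its first custom pin to point data and creates output point data with both attributes, named by variables.
; A point loop copies each point and writes its box corners to those attributes.
; The output keeps the input's target actor.
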(fn ExecuteWithContext
  (params (InContext :type PCGContext%20Structure%20%28by%20ref%29) (Input :type PCGData%20Collection%20Structure%20%28by%20ref%29))
  (bind _result (Utilities|Set|ToArray (PCG|Input&Output|CustomInputLabels self)))
  (bind _returnvalue (Data|GetInputsbyPinLabel Input (Utilities|Array|Get(aref) _result)))
  (if (> (Utilities|Array|Length _returnvalue) 0)
    (bind _aspcgspatial_data (Utilities|Casting|CastToPCGSpatialData (PCG|TaggedData|BreakTaggedData (Utilities|Array|Get(aref) _returnvalue))))
    (Utilities|FlowControl|Sequence
      (:then_0
        (bind _returnvalue_2 (Game|ConstructObjectfromClass "/Script/PCG.PCGPointData"))
        (SpatialData|InitializefromData _returnvalue_2 _aspcgspatial_data 0 true true)
        (bind _returnvalue_3 (Metadata|MutableMetadata _returnvalue_2))
        (PCG|Metadata|CreateVectorAttribute _returnvalue_3 (Variables|Default|GetBoundsMinAttribute) "-1.000000,-1.000000,-1.000000" true true)
        (PCG|Metadata|CreateVectorAttribute _returnvalue_3 (Variables|Default|GetBoundsMaxAttribute) "1.000000,1.000000,1.000000" true true))
      (:then_1
        (bind _outdata (PCG|FlowControl|PointLoop InContext (SpatialData|ToPointDatawithContext _aspcgspatial_data InContext) _returnvalue_2))
        (Variables|/Script/PCG.PCGSpatialData|SetTargetActor (Variables|/Script/PCG.PCGSpatialData|GetTargetActor _aspcgspatial_data) _outdata)
        (return (Utilities|Struct|MakePCGDataCollection (Utilities|Array|MakeArray (PCG|TaggedData|MakeTaggedData _outdata))))))
    (else
      (return 0))))

(fn PointLoopBody
  (params (InContext :type PCGContext%20Structure%20%28by%20ref%29) (InData :type PCGPoint%20Data%20Object%20Reference) (InPoint :type PCGPoint%20Structure%20%28by%20ref%29) (OutMetadata :type PCGMetadata%20Object%20Reference) (Iteration :type Integer64))
  (bind _outpoint (PCG|CopyPoint InPoint true 0 OutMetadata))
  (bind _returnvalue (PCG|Points|GetTransformedBounds _outpoint))
  (PCG|Metadata|SetVectorAttribute _outpoint OutMetadata (Variables|Default|GetBoundsMinAttribute) (Utilities|Struct|BreakBox _returnvalue))
  (PCG|Metadata|SetVectorAttribute _outpoint OutMetadata (Variables|Default|GetBoundsMaxAttribute) (.Max (Utilities|Struct|BreakBox _returnvalue)))
  (return _outpoint true))
