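; Deprecated PCG Blueprint element that sets point density from world height.
; It converts the first spatial data on its first custom input to points, runs a point loop and carries over the target actor.
; Each point's Z, minus GradientOffset and divided by GradientScale, forms a clamped gradient.
; DensityMode selects that gradient, its rounding or an arithmetic blend with existing density, then the result is clamped and the seed recomputed from position.
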(fn ExecuteWithContext
  (params (InContext :type PCGContext%20Structure%20%28by%20ref%29) (Input :type PCGData%20Collection%20Structure%20%28by%20ref%29))
  (bind _result (Utilities|Set|ToArray (PCG|Input&Output|CustomInputLabels self)))
  (bind _returnvalue (Data|GetInputsbyPinLabel Input (Utilities|Array|Get(aref) _result)))
  (if (> (Utilities|Array|Length _returnvalue) 0)
    (bind _aspcgspatial_data (Utilities|Casting|CastToPCGSpatialData (PCG|TaggedData|BreakTaggedData (Utilities|Array|Get(aref) _returnvalue))))
    (bind _outdata (PCG|FlowControl|PointLoop InContext (SpatialData|ToPointDatawithContext _aspcgspatial_data InContext) 0))
    (Variables|/Script/PCG.PCGSpatialData|SetTargetActor (Variables|/Script/PCG.PCGSpatialData|GetTargetActor _aspcgspatial_data) _outdata)
    (return (Utilities|Struct|MakePCGDataCollection (Utilities|Array|MakeArray (PCG|TaggedData|MakeTaggedData _outdata))))
    (else
      (return 0))))

(fn PointLoopBody
  (params (InContext :type PCGContext%20Structure%20%28by%20ref%29) (InData :type PCGPoint%20Data%20Object%20Reference) (InPoint :type PCGPoint%20Structure%20%28by%20ref%29) (OutMetadata :type PCGMetadata%20Object%20Reference) (Iteration :type Integer64))
  (bind (_transform _density) (Utilities|Struct|BreakPCGPoint InPoint))
  (bind _location (.location _transform))
  (bind _returnvalue (PCG|Helpers|ComputeSeedfromPosition _location))
  (bind _structout (Utilities|Struct|SetmembersinPCGPoint InPoint (Math|Float|Clamp(Float) (Utilities|Select :Index (Variables|Default|GetDensityMode) :NewEnumerator0 (Math|Float|Clamp(Float) (Math|Float|SafeDivide (- (.z _location) (Variables|Default|GetGradientOffset)) (Variables|Default|GetGradientScale))) :NewEnumerator2 (Math|Float|Min(Float) _density (Math|Float|Clamp(Float) (Math|Float|SafeDivide (- (.z _location) (Variables|Default|GetGradientOffset)) (Variables|Default|GetGradientScale)))) :NewEnumerator3 (Math|Float|Max(Float) _density (Math|Float|Clamp(Float) (Math|Float|SafeDivide (- (.z _location) (Variables|Default|GetGradientOffset)) (Variables|Default|GetGradientScale)))) :NewEnumerator4 (+ _density (Math|Float|Clamp(Float) (Math|Float|SafeDivide (- (.z _location) (Variables|Default|GetGradientOffset)) (Variables|Default|GetGradientScale)))) :NewEnumerator5 (- _density (Math|Float|Clamp(Float) (Math|Float|SafeDivide (- (.z _location) (Variables|Default|GetGradientOffset)) (Variables|Default|GetGradientScale)))) :NewEnumerator6 (* _density (Math|Float|Clamp(Float) (Math|Float|SafeDivide (- (.z _location) (Variables|Default|GetGradientOffset)) (Variables|Default|GetGradientScale)))) :NewEnumerator7 (Math|Float|SafeDivide _density (Math|Float|Clamp(Float) (Math|Float|SafeDivide (- (.z _location) (Variables|Default|GetGradientOffset)) (Variables|Default|GetGradientScale)))) :NewEnumerator8 (Math|Conversions|ToFloat(Integer) (Math|Float|Round (Math|Float|Clamp(Float) (Math|Float|SafeDivide (- (.z _location) (Variables|Default|GetGradientOffset)) (Variables|Default|GetGradientScale))))) :NewEnumerator9 (Math|Conversions|ToFloat(Integer) (Math|Float|Truncate (Math|Float|Clamp(Float) (Math|Float|SafeDivide (- (.z _location) (Variables|Default|GetGradientOffset)) (Variables|Default|GetGradientScale))))) :NewEnumerator10 (Math|Float|Clamp(Float) (Math|Float|SafeDivide (- (.z _location) (Variables|Default|GetGradientOffset)) (Variables|Default|GetGradientScale))) :NewEnumerator11 (Math|Float|Clamp(Float) (Math|Float|SafeDivide (- (.z _location) (Variables|Default|GetGradientOffset)) (Variables|Default|GetGradientScale))) :NewEnumerator12 (Math|Float|Clamp(Float) (Math|Float|SafeDivide (- (.z _location) (Variables|Default|GetGradientOffset)) (Variables|Default|GetGradientScale))) :NewEnumerator13 (Math|Float|Clamp(Float) (Math|Float|SafeDivide (- (.z _location) (Variables|Default|GetGradientOffset)) (Variables|Default|GetGradientScale))))) _returnvalue))
  (return _structout true))
