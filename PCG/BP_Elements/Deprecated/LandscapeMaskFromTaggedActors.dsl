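; Deprecated PCG Blueprint element that masks input spatial data using a color-keyed render of tagged actors.
; It resolves the original PCG volume, even from partition actors, and renders tagged actors orthographically in base color over slightly padded bounds.
; A color-key mask, optionally inverted, is blurred and wrapped as red-channel render target data.
; The output intersects this with the first spatial input and passes settings through.
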(fn Execute
  (params (Input :type PCGData%20Collection%20Structure%20%28by%20ref%29))
  (bind (_origin _boxextent) (Collision|GetComponentBounds (Actor|GetComponentbyClass (LocalVariables|GetPCGVolumeActor) "/Script/Engine.BrushComponent")))
  (bind _returnvalue (Data|GetAllSpatialInputs Input))
  (bind _aspcgspatial_data (Utilities|Casting|CastToPCGSpatialData (PCG|TaggedData|BreakTaggedData (Utilities|Array|Get(aref) _returnvalue))))
  (bind _output_get (LocalVariables|SetInputPCGSpatialData _aspcgspatial_data))
  (LocalVariables|SetTargetActor (Variables|/Script/PCG.PCGSpatialData|GetTargetActor _output_get))
  (bind _aspcgpartition_actor (Utilities|Casting|CastToPCGPartitionActor (Variables|/Script/PCG.PCGSpatialData|GetTargetActor _output_get))
    (:then
      (LocalVariables|SetPCGVolumeActor (Components|GetOwner (PCG|PartitionActor|GetOriginalComponent _aspcgpartition_actor (Actor|GetComponentbyClass _aspcgpartition_actor "/Script/PCG.PCGComponent")))))
    (:CastFailed
      (LocalVariables|SetPCGVolumeActor (Variables|/Script/PCG.PCGSpatialData|GetTargetActor _output_get))))
  (LocalVariables|SetExtents (* _boxextent (Math|Vector|MakeVector 1.1 1.1 1.1)))
  ;[1 1 1] Setup RenderTargets
  (Utilities|FlowControl|Sequence
    (:then_0
      (Utilities|Array|AppendArray (LocalVariables|GettagsArray) (Variables|Default|GetActor%20Tags))
      (bind (_texturetarget _rt_resolution) (Default|RTRenderWorldOrtho "SCS_BaseColor" (LocalVariables|GetPCGVolumeActor) (LocalVariables|GetExtents) false (LocalVariables|GettagsArray) (Variables|Default|GetTexels%20Per%20Meter)))
      (bind _rendertarget2d (Default|RTGenerateColorMask _texturetarget (Utilities|Map|MakeMap "InvertMask" (Variables|Default|GetInvertMask)) (Utilities|Map|MakeMap "ColorKey" (Variables|Default|GetColorKey)) _rt_resolution))
      (bind _rendertarget2d_2 (Class|RenderTargetFunctions|RTBlur _rendertarget2d (Utilities|Map|MakeMap "BlurDistance" (Variables|Default|GetBlurDistance)) (Utilities|Map|MakeMap) (Math|Float|Max(Float) (.x (LocalVariables|GetExtents)) (.y (LocalVariables|GetExtents))) _rt_resolution)))
    (:then_1
      (if false
        (Class|RenderTargetFunctions|RTVisualize (LocalVariables|GetPCGVolumeActor) _texturetarget)
        (Class|RenderTargetFunctions|RTVisualize (LocalVariables|GetPCGVolumeActor) _rendertarget2d 1)
        (Class|RenderTargetFunctions|RTVisualize (LocalVariables|GetPCGVolumeActor) _rendertarget2d_2 2)))
    (:then_2
      (bind _returnvalue_2 (Game|ConstructObjectfromClass "/Script/PCG.PCGRenderTargetData"))
      (Variables|/Script/PCG.PCGSpatialData|SetTargetActor (LocalVariables|GetTargetActor) _returnvalue_2)
      (Variables|/Script/PCG.PCGTexture2DSingleBaseData|SetColorChannel "Red" _returnvalue_2)
      (RenderTarget|Initialize _returnvalue_2 _rendertarget2d_2 (Math|Transform|MakeTransform (.location (Transformation|GetActorTransform (LocalVariables|GetPCGVolumeActor))) (.rotation (Transformation|GetActorTransform (LocalVariables|GetPCGVolumeActor))) (Math|Vector|MakeVector (Math|Float|Max(Float) (.x (LocalVariables|GetExtents)) (.y (LocalVariables|GetExtents))) (Math|Float|Max(Float) (.x (LocalVariables|GetExtents)) (.y (LocalVariables|GetExtents))) (.z (LocalVariables|GetExtents)))) false false))
    (:then_4
      (LocalVariables|SetOutputTaggedData (Utilities|Array|MakeArray (PCG|TaggedData|MakeTaggedData (SpatialData|IntersectWith (LocalVariables|GetInputPCGSpatialData) _returnvalue_2))))
      (bind _returnvalue_3 (Data|GetAllSettings Input))
      (Utilities|Array|AppendArray (LocalVariables|GetOutputTaggedData) _returnvalue_3)
      (return (Utilities|Struct|MakePCGDataCollection (LocalVariables|GetOutputTaggedData))))))
