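; Deprecated PCG element that turns variation in captured world normals, presumably landscape, into density for the first input spatial data.
; It orthographically renders normals over slightly padded PCG volume bounds, then derives a variance mask adjusted by levels and blur on render targets.
; The mask becomes render target data with red channel as density, placed on the volume with square XY extents.
; It returns that data intersected with the input, plus passed-through settings.
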(fn ExecuteWithContext
  (params (InContext :type PCGContext%20Structure%20%28by%20ref%29) (Input :type PCGData%20Collection%20Structure%20%28by%20ref%29))
  (bind _output_get (LocalVariables|SetInputCollection Input))
  (bind _result (Utilities|Set|ToArray (PCG|Input&Output|CustomInputLabels self)))
  (bind _returnvalue (Data|GetInputsbyPinLabel _output_get (Utilities|Array|Get(aref) _result)))
  (if (> (Utilities|Array|Length _returnvalue) 0)
    (bind _aspcgspatial_data (Utilities|Casting|CastToPCGSpatialData (PCG|TaggedData|BreakTaggedData (Utilities|Array|Get(aref) _returnvalue))))
    (bind _output_get_2 (LocalVariables|SetInputPCGSpatialData _aspcgspatial_data))
    (LocalVariables|SetTargetActor (Variables|/Script/PCG.PCGSpatialData|GetTargetActor _output_get_2))
    (bind _returnvalue_2 (PCG|Temporary|GetOriginalComponent InContext))
    (bind _output_get_3 (LocalVariables|SetPCGVolumeActor (Components|GetOwner _returnvalue_2)))
    (LocalVariables|SetExtents (* (.BoxExtent (Collision|GetComponentBounds (Actor|GetComponentbyClass _output_get_3 "/Script/Engine.BrushComponent"))) (Math|Vector|MakeVector 1.1 1.1 1.1)))
    (Utilities|FlowControl|Sequence
      (:then_0
        ;[1 1 1] Setup RenderTargets
        (bind (_texturetarget _rt_resolution) (Default|RTRenderWorldOrtho "SCS_Normal" (LocalVariables|GetPCGVolumeActor) (LocalVariables|GetExtents) true (Utilities|Array|MakeArray) (Variables|Default|GetTexels%20Per%20Meter)))
        (bind _rendertarget2d (Default|RTGenerateNormalVarianceMask _texturetarget (Utilities|Map|MakeMap) (Utilities|Map|MakeMap) _rt_resolution))
        (bind _rendertarget2d_2 (Class|RenderTargetFunctions|RTAdjustLevels _rendertarget2d (Utilities|Map|MakeMap "Power" (Variables|Default|GetPower) "Min" (Variables|Default|GetMin) "Max" (Variables|Default|GetMax)) (Utilities|Map|MakeMap) _rt_resolution))
        (bind _rendertarget2d_3 (Class|RenderTargetFunctions|RTBlur _rendertarget2d_2 (Utilities|Map|MakeMap "BlurDistance" (Variables|Default|GetBlurDistance)) (Utilities|Map|MakeMap) (Math|Float|Max(Float) (.x (LocalVariables|GetExtents)) (.y (LocalVariables|GetExtents))) _rt_resolution)))
      (:then_2
        (bind _returnvalue_3 (Game|ConstructObjectfromClass "/Script/PCG.PCGRenderTargetData"))
        (Variables|/Script/PCG.PCGSpatialData|SetTargetActor (LocalVariables|GetTargetActor) _returnvalue_3)
        (Variables|/Script/PCG.PCGTexture2DSingleBaseData|SetColorChannel "Red" _returnvalue_3)
        (Variables|/Script/PCG.PCGTexture2DSingleBaseData|SetbUseDensitySourceChannel true _returnvalue_3)
        (RenderTarget|Initialize _returnvalue_3 _rendertarget2d_3 (Math|Transform|MakeTransform (.location (Transformation|GetActorTransform (LocalVariables|GetPCGVolumeActor))) (.rotation (Transformation|GetActorTransform (LocalVariables|GetPCGVolumeActor))) (Math|Vector|MakeVector (Math|Float|Max(Float) (.x (LocalVariables|GetExtents)) (.y (LocalVariables|GetExtents))) (Math|Float|Max(Float) (.x (LocalVariables|GetExtents)) (.y (LocalVariables|GetExtents))) (.z (LocalVariables|GetExtents)))) false false))
      (:then_4
        (LocalVariables|SetOutputTaggedData (Utilities|Array|MakeArray (PCG|TaggedData|MakeTaggedData (SpatialData|IntersectWith (LocalVariables|GetInputPCGSpatialData) _returnvalue_3))))
        (bind _returnvalue_4 (Data|GetAllSettings (LocalVariables|GetInputCollection)))
        (Utilities|Array|AppendArray (LocalVariables|GetOutputTaggedData) _returnvalue_4)
        (return (Utilities|Struct|MakePCGDataCollection (LocalVariables|GetOutputTaggedData)))))
    (else
      (return 0))))
