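; Deprecated PCG Blueprint element that sets point density from how closely each point's up vector matches a Normal.
; ExecuteWithContext converts the first input to point data, runs a point loop and copies the target actor.
; Each point's dot product plus Offset is clamped, raised to 1/Strength, then combined with existing density per DensityMode and clamped.
; Round and truncate modes use only existing density, and Reseed points optionally derives seeds from position.
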(fn ExecuteWithContext
  (params (InContext :type PCGContext%20Structure%20%28by%20ref%29) (Input :type PCGData%20Collection%20Structure%20%28by%20ref%29))
  (bind _result (Utilities|Set|ToArray (PCG|Input&Output|CustomInputLabels self)))
  (bind _returnvalue (Data|GetInputsbyPinLabel Input (Utilities|Array|Get(aref) _result)))
  (if (> (Utilities|Array|Length _returnvalue) 0)
    (bind _aspcgspatial_data (Utilities|Casting|CastToPCGSpatialData (PCG|TaggedData|BreakTaggedData (Utilities|Array|Get(aref) _returnvalue))))
    (bind _outdata (PCG|FlowControl|PointLoop InContext (SpatialData|ToPointDatawithContext _aspcgspatial_data InContext) 0))
    (Variables|/Script/PCG.PCGSpatialData|SetTargetActor (Variables|/Script/PCG.PCGSpatialData|GetTargetActor _aspcgspatial_data) _outdata)
    (return (Utilities|Struct|MakePCGDataCollection (Utilities|Array|MakeArray (PCG|TaggedData|MakeTaggedData _outdata))))
    (else
      (return 0))))

(fn PointLoopBody
  (params (InContext :type PCGContext%20Structure%20%28by%20ref%29) (InData :type PCGPoint%20Data%20Object%20Reference) (InPoint :type PCGPoint%20Structure%20%28by%20ref%29) (OutMetadata :type PCGMetadata%20Object%20Reference) (Iteration :type Integer64))
  (bind (_transform _density _boundsmin _boundsmax _seed) (Utilities|Struct|BreakPCGPoint InPoint))
  (LocalVariables|Setseed _seed)
  (if (Variables|Default|GetReseed%20points)
    (bind _returnvalue (PCG|Helpers|ComputeSeedfromPosition (.location _transform)))
    (LocalVariables|Setseed _returnvalue))
  (bind _structout (Utilities|Struct|SetmembersinPCGPoint InPoint (Math|Float|Clamp(Float) (Utilities|Select :Index (Variables|Default|GetDensityMode) :NewEnumerator0 (Math|Float|Power (Math|Float|Clamp(Float) (+ (Math|Vector|DotProduct (Math|Vector|GetUpVector (.rotation _transform)) (Variables|Default|GetNormal)) (Variables|Default|GetOffset))) (/ 1.0 (Variables|Default|GetStrength))) :NewEnumerator2 (Math|Float|Min(Float) _density (Math|Float|Power (Math|Float|Clamp(Float) (+ (Math|Vector|DotProduct (Math|Vector|GetUpVector (.rotation _transform)) (Variables|Default|GetNormal)) (Variables|Default|GetOffset))) (/ 1.0 (Variables|Default|GetStrength)))) :NewEnumerator3 (Math|Float|Max(Float) _density (Math|Float|Power (Math|Float|Clamp(Float) (+ (Math|Vector|DotProduct (Math|Vector|GetUpVector (.rotation _transform)) (Variables|Default|GetNormal)) (Variables|Default|GetOffset))) (/ 1.0 (Variables|Default|GetStrength)))) :NewEnumerator4 (+ _density (Math|Float|Power (Math|Float|Clamp(Float) (+ (Math|Vector|DotProduct (Math|Vector|GetUpVector (.rotation _transform)) (Variables|Default|GetNormal)) (Variables|Default|GetOffset))) (/ 1.0 (Variables|Default|GetStrength)))) :NewEnumerator5 (- _density (Math|Float|Power (Math|Float|Clamp(Float) (+ (Math|Vector|DotProduct (Math|Vector|GetUpVector (.rotation _transform)) (Variables|Default|GetNormal)) (Variables|Default|GetOffset))) (/ 1.0 (Variables|Default|GetStrength)))) :NewEnumerator6 (* _density (Math|Float|Power (Math|Float|Clamp(Float) (+ (Math|Vector|DotProduct (Math|Vector|GetUpVector (.rotation _transform)) (Variables|Default|GetNormal)) (Variables|Default|GetOffset))) (/ 1.0 (Variables|Default|GetStrength)))) :NewEnumerator7 (Math|Float|SafeDivide _density (Math|Float|Power (Math|Float|Clamp(Float) (+ (Math|Vector|DotProduct (Math|Vector|GetUpVector (.rotation _transform)) (Variables|Default|GetNormal)) (Variables|Default|GetOffset))) (/ 1.0 (Variables|Default|GetStrength)))) :NewEnumerator8 (Math|Conversions|ToFloat(Integer) (Math|Float|Round _density)) :NewEnumerator9 (Math|Conversions|ToFloat(Integer) (Math|Float|Truncate _density)) :NewEnumerator10 0.0 :NewEnumerator11 0.0 :NewEnumerator12 0.0 :NewEnumerator13 0.0)) (LocalVariables|Getseed)))
  (return _structout true))
