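; Custom PCG Blueprint element, titled "CrossProduct Param" and filed under Deprecated, that computes the cross product of two vector attributes.
; It takes the first attribute set from each of its first two custom input pins and reads the last attribute at entry key 0 as a vector.
; The result goes into a newly constructed param data object, under the AttributeNameOut name, as the only output.
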(fn ExecuteWithContext
  (params (InContext :type PCGContext%20Structure%20%28by%20ref%29) (Input :type PCGData%20Collection%20Structure%20%28by%20ref%29))
  (bind (_attributenames _attributetypes) (PCG|Metadata|GetAttributes (LocalVariables|GetInMetadataA)))
  (bind (_attributenames_2 _attributetypes_2) (PCG|Metadata|GetAttributes (LocalVariables|GetInMetaDataB)))
  (bind _inmetadatabvector (LocalVariables|GetInMetadataBVector))
  (bind _inmetadataavector (LocalVariables|GetInMetadataAVector))
  (bind _returnvalue (Math|Vector|CrossProduct _inmetadataavector _inmetadatabvector))
  (bind _attributenameout (Variables|Default|GetAttributeNameOut))
  (bind _inputcollection (LocalVariables|GetInputCollection))
  (LocalVariables|SetContext InContext)
  (LocalVariables|SetInputCollection Input)
  (bind _result (Utilities|Set|ToArray (PCG|Input&Output|CustomInputLabels self)))
  (bind _returnvalue_2 (Data|GetParamsbyPinLabel _inputcollection (Utilities|Array|Get(aref) _result)))
  (bind _returnvalue_3 (Data|GetParamsbyPinLabel _inputcollection (Utilities|Array|Get(aref) _result 1)))
  (bind _aspcgattribute_set (Utilities|Casting|CastToPCGParamData (PCG|TaggedData|BreakTaggedData (Utilities|Array|Get(aref) _returnvalue_2))))
  (bind _returnvalue_4 (Metadata|MutableMetadata _aspcgattribute_set))
  (LocalVariables|SetInMetadataA _returnvalue_4)
  (bind _returnvalue_5 (PCG|Metadata|GetVectorAttributebyMetadataKey 0 (LocalVariables|GetInMetadataA) (Utilities|Array|Get(acopy) _attributenames (Utilities|Array|LastIndex _attributetypes))))
  (LocalVariables|SetInMetadataAVector _returnvalue_5)
  (bind _aspcgattribute_set_2 (Utilities|Casting|CastToPCGParamData (PCG|TaggedData|BreakTaggedData (Utilities|Array|Get(aref) _returnvalue_3))))
  (bind _returnvalue_6 (Metadata|MutableMetadata _aspcgattribute_set_2))
  (LocalVariables|SetInMetaDataB _returnvalue_6)
  (bind _returnvalue_7 (PCG|Metadata|GetVectorAttributebyMetadataKey 0 (LocalVariables|GetInMetaDataB) (Utilities|Array|Get(aref) _attributenames_2 (Utilities|Array|LastIndex _attributetypes_2))))
  (LocalVariables|SetInMetadataBVector _returnvalue_7)
  (bind _returnvalue_8 (Game|ConstructObjectfromClass "/Script/PCG.PCGParamData"))
  (bind _returnvalue_9 (Metadata|MutableMetadata _returnvalue_8))
  (bind _output_get (LocalVariables|SetOutMetadata _returnvalue_9))
  (bind _returnvalue_10 (PCG|Metadata|AddEntry _output_get -1))
  (LocalVariables|SetOutMetaDataEntry _returnvalue_10)
  (PCG|Metadata|CreateVectorAttribute (LocalVariables|GetOutMetadata) _attributenameout _returnvalue false true)
  (PCG|Metadata|SetVectorAttributebyMetadataKey (LocalVariables|GetOutMetaDataEntry) (LocalVariables|GetOutMetadata) _attributenameout _returnvalue)
  (return (Utilities|Struct|MakePCGDataCollection (Utilities|Array|MakeArray (PCG|TaggedData|MakeTaggedData _returnvalue_8)))))

(fn NodeColorOverride
  (return "(R=1.000000,G=0.600000,B=0.000000,A=1.000000)"))

(fn NodeTitleOverride
  (return "CrossProduct Param"))
