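; Deprecated PCG Blueprint node that packs float attributes from separate param inputs into one Float, Vector or Vector4 attribute, chosen by OutputAttributeType.
; It takes the first param data on each custom input pin, mapping pins by index to X, Y, Z, W.
; Each component reads only the last attribute at entry key 0.
; The result is written under AttributeNameOut to a single entry in newly constructed PCGParamData.
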(fn ExecuteWithContext
  (params (InContext :type PCGContext%20Structure%20%28by%20ref%29) (Input :type PCGData%20Collection%20Structure%20%28by%20ref%29))
  (bind (_attributenames _attributetypes) (PCG|Metadata|GetAttributes (LocalVariables|GetInMetaDataY)))
  (bind (_attributenames_2 _attributetypes_2) (PCG|Metadata|GetAttributes (LocalVariables|GetInMetaDataZ)))
  (bind (_attributenames_3 _attributetypes_3) (PCG|Metadata|GetAttributes (LocalVariables|GetInMetaDataW)))
  (bind (_attributenames_4 _attributetypes_4) (PCG|Metadata|GetAttributes (LocalVariables|GetInMetaDataZ)))
  (bind _inattributezfloat (LocalVariables|GetInAttributeZFloat))
  (bind _inattributeyfloat (LocalVariables|GetInAttributeYFloat))
  (bind _inattributexfloat (LocalVariables|GetInAttributeXFloat))
  (bind _returnvalue (Math|Vector|MakeVector _inattributexfloat _inattributeyfloat _inattributezfloat))
  (bind _inattributewfloat (LocalVariables|GetInAttributeWFloat))
  (bind _inattributezfloat_2 (LocalVariables|GetInAttributeZFloat))
  (bind _inattributeyfloat_2 (LocalVariables|GetInAttributeYFloat))
  (bind _inattributexfloat_2 (LocalVariables|GetInAttributeXFloat))
  (bind _returnvalue_2 (Math|Vector4|MakeVector4 _inattributexfloat_2 _inattributeyfloat_2 _inattributezfloat_2 _inattributewfloat))
  (bind _attributenameout (Variables|Default|GetAttributeNameOut))
  (bind _attributenameout_2 (Variables|Default|GetAttributeNameOut))
  (bind _attributenameout_3 (Variables|Default|GetAttributeNameOut))
  (bind _inattributexfloat_3 (LocalVariables|GetInAttributeXFloat))
  (LocalVariables|SetContext InContext)
  (bind _result (Utilities|Set|ToArray (PCG|Input&Output|CustomInputLabels self)))
  (bind _returnvalue_3 (Data|GetParamsbyPinLabel Input (Utilities|Array|Get(aref) _result)))
  (bind _returnvalue_4 (Data|GetParamsbyPinLabel Input (Utilities|Array|Get(aref) _result 1)))
  (bind _returnvalue_5 (Data|GetParamsbyPinLabel Input (Utilities|Array|Get(aref) _result 2)))
  (bind _returnvalue_6 (Data|GetParamsbyPinLabel Input (Utilities|Array|Get(aref) _result 3)))
  (switch Utilities|FlowControl|Switch|SwitchonPCGAttributeTypes (Variables|Default|GetOutputAttributeType)
    (:NewEnumerator0
      (bind _aspcgattribute_set (Utilities|Casting|CastToPCGParamData (PCG|TaggedData|BreakTaggedData (Utilities|Array|Get(aref) _returnvalue_3))))
      (bind _returnvalue_7 (Metadata|MutableMetadata _aspcgattribute_set))
      (LocalVariables|SetInMetadataX _returnvalue_7))
    (:NewEnumerator1
      (bind _aspcgattribute_set_2 (Utilities|Casting|CastToPCGParamData (PCG|TaggedData|BreakTaggedData (Utilities|Array|Get(aref) _returnvalue_3))))
      (bind _returnvalue_8 (Metadata|MutableMetadata _aspcgattribute_set_2))
      (LocalVariables|SetInMetadataX _returnvalue_8)
      (bind _aspcgattribute_set_3 (Utilities|Casting|CastToPCGParamData (PCG|TaggedData|BreakTaggedData (Utilities|Array|Get(aref) _returnvalue_4))))
      (bind _returnvalue_9 (Metadata|MutableMetadata _aspcgattribute_set_3))
      (LocalVariables|SetInMetaDataY _returnvalue_9)
      (bind _aspcgattribute_set_4 (Utilities|Casting|CastToPCGParamData (PCG|TaggedData|BreakTaggedData (Utilities|Array|Get(aref) _returnvalue_5))))
      (bind _returnvalue_10 (Metadata|MutableMetadata _aspcgattribute_set_4))
      (LocalVariables|SetInMetaDataZ _returnvalue_10))
    (:NewEnumerator2
      (bind _aspcgattribute_set_5 (Utilities|Casting|CastToPCGParamData (PCG|TaggedData|BreakTaggedData (Utilities|Array|Get(aref) _returnvalue_3))))
      (bind _returnvalue_11 (Metadata|MutableMetadata _aspcgattribute_set_5))
      (LocalVariables|SetInMetadataX _returnvalue_11)
      (bind _aspcgattribute_set_6 (Utilities|Casting|CastToPCGParamData (PCG|TaggedData|BreakTaggedData (Utilities|Array|Get(aref) _returnvalue_4))))
      (bind _returnvalue_12 (Metadata|MutableMetadata _aspcgattribute_set_6))
      (LocalVariables|SetInMetaDataY _returnvalue_12)
      (bind _aspcgattribute_set_7 (Utilities|Casting|CastToPCGParamData (PCG|TaggedData|BreakTaggedData (Utilities|Array|Get(aref) _returnvalue_5))))
      (bind _returnvalue_13 (Metadata|MutableMetadata _aspcgattribute_set_7))
      (LocalVariables|SetInMetaDataZ _returnvalue_13)
      (bind _aspcgattribute_set_8 (Utilities|Casting|CastToPCGParamData (PCG|TaggedData|BreakTaggedData (Utilities|Array|Get(aref) _returnvalue_6))))
      (bind _returnvalue_14 (Metadata|MutableMetadata _aspcgattribute_set_8))
      (LocalVariables|SetInMetaDataW _returnvalue_14)))
  (bind _returnvalue_15 (Game|ConstructObjectfromClass "/Script/PCG.PCGParamData"))
  (bind _returnvalue_16 (Metadata|MutableMetadata _returnvalue_15))
  (bind _output_get (LocalVariables|SetOutMetadata _returnvalue_16))
  (bind _returnvalue_17 (PCG|Metadata|AddEntry _output_get -1))
  (LocalVariables|SetOutMetaDataEntry _returnvalue_17)
  (switch Utilities|FlowControl|Switch|SwitchonPCGAttributeTypes (Variables|Default|GetOutputAttributeType)
    (:NewEnumerator0
      (bind _returnvalue_18 (PCG|Metadata|GetFloatAttributebyMetadataKey 0 (LocalVariables|GetInMetadataX) (Utilities|Array|Get(aref) (PCG|Metadata|GetAttributes (LocalVariables|GetInMetadataX)) (Utilities|Array|LastIndex (PCG|Metadata|GetAttributes (LocalVariables|GetInMetadataX))))))
      (LocalVariables|SetInAttributeXFloat _returnvalue_18)
      (PCG|Metadata|CreateFloatAttribute (LocalVariables|GetOutMetadata) _attributenameout_2 _inattributexfloat_3 false true)
      (PCG|Metadata|SetFloatAttributebyMetadataKey (LocalVariables|GetOutMetaDataEntry) (LocalVariables|GetOutMetadata) _attributenameout_2 _inattributexfloat_3))
    (:NewEnumerator1
      (bind _returnvalue_19 (PCG|Metadata|GetFloatAttributebyMetadataKey 0 (LocalVariables|GetInMetadataX) (Utilities|Array|Get(aref) (PCG|Metadata|GetAttributes (LocalVariables|GetInMetadataX)) (Utilities|Array|LastIndex (PCG|Metadata|GetAttributes (LocalVariables|GetInMetadataX))))))
      (LocalVariables|SetInAttributeXFloat _returnvalue_19)
      (bind _returnvalue_20 (PCG|Metadata|GetFloatAttributebyMetadataKey 0 (LocalVariables|GetInMetaDataY) (Utilities|Array|Get(aref) (PCG|Metadata|GetAttributes (LocalVariables|GetInMetaDataY)) (Utilities|Array|LastIndex (PCG|Metadata|GetAttributes (LocalVariables|GetInMetaDataY))))))
      (LocalVariables|SetInAttributeYFloat _returnvalue_20)
      (bind _returnvalue_21 (PCG|Metadata|GetFloatAttributebyMetadataKey 0 (LocalVariables|GetInMetaDataZ) (Utilities|Array|Get(aref) _attributenames_4 (Utilities|Array|LastIndex _attributetypes_4))))
      (LocalVariables|SetInAttributeZFloat _returnvalue_21)
      (PCG|Metadata|CreateVectorAttribute (LocalVariables|GetOutMetadata) _attributenameout_3 _returnvalue false true)
      (PCG|Metadata|SetVectorAttributebyMetadataKey (LocalVariables|GetOutMetaDataEntry) (LocalVariables|GetOutMetadata) _attributenameout_3 _returnvalue))
    (:NewEnumerator2
      (bind _returnvalue_22 (PCG|Metadata|GetFloatAttributebyMetadataKey 0 (LocalVariables|GetInMetadataX) (Utilities|Array|Get(aref) (PCG|Metadata|GetAttributes (LocalVariables|GetInMetadataX)) (Utilities|Array|LastIndex (PCG|Metadata|GetAttributes (LocalVariables|GetInMetadataX))))))
      (LocalVariables|SetInAttributeXFloat _returnvalue_22)
      (bind _returnvalue_23 (PCG|Metadata|GetFloatAttributebyMetadataKey 0 (LocalVariables|GetInMetaDataY) (Utilities|Array|Get(aref) _attributenames (Utilities|Array|LastIndex _attributetypes))))
      (LocalVariables|SetInAttributeYFloat _returnvalue_23)
      (bind _returnvalue_24 (PCG|Metadata|GetFloatAttributebyMetadataKey 0 (LocalVariables|GetInMetaDataZ) (Utilities|Array|Get(aref) _attributenames_2 (Utilities|Array|LastIndex _attributetypes_2))))
      (LocalVariables|SetInAttributeZFloat _returnvalue_24)
      (bind _returnvalue_25 (PCG|Metadata|GetFloatAttributebyMetadataKey 0 (LocalVariables|GetInMetaDataW) (Utilities|Array|Get(aref) _attributenames_3 (Utilities|Array|LastIndex _attributetypes_3))))
      (LocalVariables|SetInAttributeWFloat _returnvalue_25)
      (PCG|Metadata|CreateVector4Attribute (LocalVariables|GetOutMetadata) _attributenameout _returnvalue_2 false true)
      (PCG|Metadata|SetVector4AttributebyMetadataKey (LocalVariables|GetOutMetaDataEntry) (LocalVariables|GetOutMetadata) _attributenameout _returnvalue_2)))
  (return (Utilities|Struct|MakePCGDataCollection (Utilities|Array|MakeArray (PCG|TaggedData|MakeTaggedData _returnvalue_15)))))

(fn NodeColorOverride
  (return "(R=1.000000,G=0.600000,B=0.000000,A=1.000000)"))
