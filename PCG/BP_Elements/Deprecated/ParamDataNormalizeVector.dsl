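; Deprecated PCG Blueprint element, titled "NormalizeVector Param" with an orange node color, that normalizes a vector held in param data.
; It casts the first input params entry to param data and reads the last attribute in its metadata as a vector at key 0.
; It returns new param data with one entry, storing the normalized vector as both default and value of the attribute named by AttributeNameOut.
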(fn ExecuteWithContext
  (params (InContext :type PCGContext%20Structure%20%28by%20ref%29) (Input :type PCGData%20Collection%20Structure%20%28by%20ref%29))
  (bind _inmetadataavector (LocalVariables|GetInMetadataAVector))
  (bind _returnvalue (Math|Vector|Normalize _inmetadataavector))
  (bind _attributenameout (Variables|Default|GetAttributeNameOut))
  (LocalVariables|SetContext InContext)
  (bind _output_get (LocalVariables|SetInputCollection Input))
  (bind _returnvalue_2 (Data|GetParams _output_get))
  (bind _aspcgattribute_set (Utilities|Casting|CastToPCGParamData (PCG|TaggedData|BreakTaggedData (Utilities|Array|Get(aref) _returnvalue_2))))
  (bind _returnvalue_3 (Metadata|MutableMetadata _aspcgattribute_set))
  (LocalVariables|SetInMetadataA _returnvalue_3)
  (bind _returnvalue_4 (PCG|Metadata|GetVectorAttributebyMetadataKey 0 (LocalVariables|GetInMetadataA) (Utilities|Array|Get(acopy) (PCG|Metadata|GetAttributes (LocalVariables|GetInMetadataA)) (Utilities|Array|LastIndex (PCG|Metadata|GetAttributes (LocalVariables|GetInMetadataA))))))
  (LocalVariables|SetInMetadataAVector _returnvalue_4)
  (bind _returnvalue_5 (Game|ConstructObjectfromClass "/Script/PCG.PCGParamData"))
  (bind _returnvalue_6 (Metadata|MutableMetadata _returnvalue_5))
  (bind _output_get_2 (LocalVariables|SetOutMetadata _returnvalue_6))
  (bind _returnvalue_7 (PCG|Metadata|AddEntry _output_get_2 -1))
  (LocalVariables|SetOutMetaDataEntry _returnvalue_7)
  (PCG|Metadata|CreateVectorAttribute (LocalVariables|GetOutMetadata) _attributenameout _returnvalue false true)
  (PCG|Metadata|SetVectorAttributebyMetadataKey (LocalVariables|GetOutMetaDataEntry) (LocalVariables|GetOutMetadata) _attributenameout _returnvalue)
  (return (Utilities|Struct|MakePCGDataCollection (Utilities|Array|MakeArray (PCG|TaggedData|MakeTaggedData _returnvalue_5)))))

(fn NodeColorOverride
  (return "(R=1.000000,G=0.600000,B=0.000000,A=1.000000)"))

(fn NodeTitleOverride
  (return "NormalizeVector Param"))
