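; Deprecated PCG Blueprint node that applies a selectable math operation to param data attributes and outputs a new single-entry param set.
; It reads the last attribute at key 0 from input A, and from B for binary operations.
; Float, Int64, Vector and Vector4 results come from Select nodes keyed on OperationMode.
; Vector4 truncate, absolute and fraction mistakenly use B, and vector fractions are truncated to zero.
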(fn ExecuteWithContext
  (params (InContext :type PCGContext%20Structure%20%28by%20ref%29) (Input :type PCGData%20Collection%20Structure%20%28by%20ref%29))
  (bind _inmetadatab (LocalVariables|GetInMetaDataB))
  (bind (_attributenames _attributetypes) (PCG|Metadata|GetAttributes _inmetadatab))
  (bind _returnvalue (Utilities|Array|LastIndex _attributetypes))
  (bind _inmetadataa (LocalVariables|GetInMetadataA))
  (bind (_attributenames_2 _attributetypes_2) (PCG|Metadata|GetAttributes _inmetadataa))
  (bind _returnvalue_2 (Utilities|Array|LastIndex _attributenames_2))
  (bind _inmetadatabvector4 (LocalVariables|GetInMetadataBVector4))
  (bind (_x _y _z _w) (Math|Vector4|BreakVector4 _inmetadatabvector4))
  (bind _inmetadataavector4 (LocalVariables|GetInMetadataAVector4))
  (bind (_x_2 _y_2 _z_2 _w_2) (Math|Vector4|BreakVector4 _inmetadataavector4))
  (bind _returnvalue_3 (Math|Float|SafeDivide _w_2 _w))
  (bind _returnvalue_4 (Math|Float|Ceil _x_2))
  (bind _returnvalue_5 (Math|Float|Ceil _y_2))
  (bind _returnvalue_6 (Math|Float|Ceil _z_2))
  (bind _returnvalue_7 (Math|Float|Ceil _w_2))
  (bind _returnvalue_8 (Math|Float|Round _x_2))
  (bind _returnvalue_9 (Math|Conversions|ToFloat(Integer) _returnvalue_8))
  (bind _returnvalue_10 (Math|Float|Round _y_2))
  (bind _returnvalue_11 (Math|Conversions|ToFloat(Integer) _returnvalue_10))
  (bind _returnvalue_12 (Math|Float|Round _z_2))
  (bind _returnvalue_13 (Math|Conversions|ToFloat(Integer) _returnvalue_12))
  (bind _returnvalue_14 (Math|Float|Truncate _x))
  (bind _returnvalue_15 (Math|Conversions|ToFloat(Integer) _returnvalue_14))
  (bind _returnvalue_16 (Math|Float|Truncate _y))
  (bind _returnvalue_17 (Math|Conversions|ToFloat(Integer) _returnvalue_16))
  (bind _returnvalue_18 (Math|Float|Truncate _z))
  (bind _returnvalue_19 (Math|Conversions|ToFloat(Integer) _returnvalue_18))
  (bind _inmetadatabvector (LocalVariables|GetInMetadataBVector))
  (bind _z_3 (.z _inmetadatabvector))
  (bind _inmetadataavector (LocalVariables|GetInMetadataAVector))
  (bind _z_4 (.z _inmetadataavector))
  (bind _returnvalue_20 (Math|Float|Min(Float) _z_4 _z_3))
  (bind _y_3 (.y _inmetadatabvector))
  (bind _y_4 (.y _inmetadataavector))
  (bind _returnvalue_21 (Math|Float|Min(Float) _y_4 _y_3))
  (bind _x_3 (.x _inmetadatabvector))
  (bind _x_4 (.x _inmetadataavector))
  (bind _returnvalue_22 (Math|Float|Min(Float) _x_4 _x_3))
  (bind _returnvalue_23 (Math|Vector|MakeVector _returnvalue_22 _returnvalue_21 _returnvalue_20))
  (bind _returnvalue_24 (Math|Float|Round _w_2))
  (bind _returnvalue_25 (Math|Conversions|ToFloat(Integer) _returnvalue_24))
  (bind _returnvalue_26 (Math|Float|Fraction _x))
  (bind _returnvalue_27 (Math|Float|Truncate _returnvalue_26))
  (bind _returnvalue_28 (Math|Float|Fraction _y))
  (bind _returnvalue_29 (Math|Float|Truncate _returnvalue_28))
  (bind _returnvalue_30 (Math|Float|Truncate _w))
  (bind _returnvalue_31 (Math|Conversions|ToFloat(Integer) _returnvalue_30))
  (bind _returnvalue_32 (Math|Vector4|MakeVector4 _returnvalue_15 _returnvalue_17 _returnvalue_19 _returnvalue_31))
  (bind _returnvalue_33 (Math|Float|Fraction _w))
  (bind _returnvalue_34 (Math|Float|Truncate _returnvalue_33))
  (bind _returnvalue_35 (Math|Float|Fraction _x_4))
  (bind _returnvalue_36 (Math|Float|Fraction _y_4))
  (bind _returnvalue_37 (Math|Float|Fraction _z_4))
  (bind _returnvalue_38 (Math|Float|Truncate _returnvalue_35))
  (bind _returnvalue_39 (Math|Conversions|ToFloat(Integer) _returnvalue_38))
  (bind _inmetadataafloat (LocalVariables|GetInMetadataAFloat))
  (bind _returnvalue_40 (Math|Float|Floor _inmetadataafloat))
  (bind _returnvalue_41 (Math|Float|Truncate _returnvalue_36))
  (bind _returnvalue_42 (Math|Conversions|ToFloat(Integer) _returnvalue_41))
  (bind _returnvalue_43 (Math|Float|Truncate _returnvalue_37))
  (bind _returnvalue_44 (Math|Conversions|ToFloat(Integer) _returnvalue_43))
  (bind _returnvalue_45 (Math|Float|Fraction _z))
  (bind _returnvalue_46 (Math|Float|Truncate _returnvalue_45))
  (bind _returnvalue_47 (Math|Conversions|ToFloat(Integer) _returnvalue_46))
  (bind _returnvalue_48 (Math|Vector|VectorGetAbs _inmetadataavector))
  (bind _returnvalue_49 (Math|Float|Absolute(Float) _x))
  (bind _returnvalue_50 (Math|Float|Absolute(Float) _y))
  (bind _returnvalue_51 (Math|Float|Absolute(Float) _z))
  (bind _returnvalue_52 (Math|Float|Absolute(Float) _w))
  (bind _returnvalue_53 (Math|Conversions|ToFloat(Integer) _returnvalue_7))
  (bind _returnvalue_54 (Math|Conversions|ToFloat(Integer) _returnvalue_6))
  (bind _returnvalue_55 (Math|Conversions|ToFloat(Integer) _returnvalue_5))
  (bind _returnvalue_56 (Math|Conversions|ToFloat(Integer) _returnvalue_4))
  (bind _returnvalue_57 (Math|Vector4|MakeVector4 _returnvalue_56 _returnvalue_55 _returnvalue_54 _returnvalue_53))
  (bind _returnvalue_58 (Math|Float|SafeDivide _x_4 _x_3))
  (bind _returnvalue_59 (Math|Float|SafeDivide _y_4 _y_3))
  (bind _returnvalue_60 (Math|Float|SafeDivide _z_4 _z_3))
  (bind _returnvalue_61 (Math|Float|SafeDivide _z_2 _z))
  (bind _returnvalue_62 (Math|Vector4|MakeVector4 _returnvalue_9 _returnvalue_11 _returnvalue_13 _returnvalue_25))
  (bind _returnvalue_63 (Math|Float|Ceil _z_4))
  (bind _returnvalue_64 (Math|Conversions|ToFloat(Integer) _returnvalue_63))
  (bind _returnvalue_65 (Math|Vector4|MakeVector4 _returnvalue_49 _returnvalue_50 _returnvalue_51 _returnvalue_52))
  (bind _inmetadatabfloat (LocalVariables|GetInMetadataBFloat))
  (bind _returnvalue_66 (Math|Float|SafeDivide _inmetadataafloat _inmetadatabfloat))
  (bind _inmetadataaint64 (LocalVariables|GetInMetadataAInt64))
  (bind _returnvalue_67 (Math|Integer64|Absolute(Integer64) _inmetadataaint64))
  (bind _returnvalue_68 (Math|Float|Truncate _z_4))
  (bind _returnvalue_69 (Math|Conversions|ToFloat(Integer) _returnvalue_68))
  (bind _returnvalue_70 (Math|Float|Truncate _y_4))
  (bind _returnvalue_71 (Math|Conversions|ToFloat(Integer) _returnvalue_70))
  (bind _returnvalue_72 (Math|Float|Truncate _x_4))
  (bind _returnvalue_73 (Math|Conversions|ToFloat(Integer) _returnvalue_72))
  (bind _returnvalue_74 (Math|Vector|MakeVector _returnvalue_73 _returnvalue_71 _returnvalue_69))
  (bind _returnvalue_75 (Math|Float|Ceil _y_4))
  (bind _returnvalue_76 (Math|Conversions|ToFloat(Integer) _returnvalue_75))
  (bind _returnvalue_77 (Math|Float|Ceil _x_4))
  (bind _returnvalue_78 (Math|Conversions|ToFloat(Integer) _returnvalue_77))
  (bind _returnvalue_79 (Math|Vector|MakeVector _returnvalue_78 _returnvalue_76 _returnvalue_64))
  (bind _returnvalue_80 (Math|Float|Max(Float) _w_2 _w))
  (bind _returnvalue_81 (Math|Float|Max(Float) _z_2 _z))
  (bind _returnvalue_82 (Math|Float|Max(Float) _y_2 _y))
  (bind _returnvalue_83 (Math|Float|Max(Float) _x_2 _x))
  (bind _returnvalue_84 (Math|Vector4|MakeVector4 _returnvalue_83 _returnvalue_82 _returnvalue_81 _returnvalue_80))
  (bind _returnvalue_85 (Math|Float|SafeDivide _y_2 _y))
  (bind _returnvalue_86 (Math|Float|SafeDivide _x_2 _x))
  (bind _returnvalue_87 (Math|Vector4|MakeVector4 _returnvalue_86 _returnvalue_85 _returnvalue_61 _returnvalue_3))
  (bind _returnvalue_88 (Math|Conversions|ToFloat(Integer) _returnvalue_34))
  (bind _returnvalue_89 (Math|Conversions|ToFloat(Integer) _returnvalue_29))
  (bind _returnvalue_90 (Math|Conversions|ToFloat(Integer) _returnvalue_27))
  (bind _returnvalue_91 (Math|Vector4|MakeVector4 _returnvalue_90 _returnvalue_89 _returnvalue_47 _returnvalue_88))
  (bind _returnvalue_92 (Math|Float|Round _inmetadataafloat))
  (bind _returnvalue_93 (Math|Conversions|ToFloat(Integer) _returnvalue_92))
  (bind _returnvalue_94 (Math|Float|Truncate _inmetadataafloat))
  (bind _returnvalue_95 (Math|Conversions|ToFloat(Integer) _returnvalue_94))
  (bind _returnvalue_96 (Math|Float|Max(Float) _z_4 _z_3))
  (bind _returnvalue_97 (Math|Float|Max(Float) _y_4 _y_3))
  (bind _returnvalue_98 (Math|Float|Max(Float) _x_4 _x_3))
  (bind _returnvalue_99 (Math|Vector|MakeVector _returnvalue_98 _returnvalue_97 _returnvalue_96))
  (bind _returnvalue_100 (Math|Float|Ceil _inmetadataafloat))
  (bind _returnvalue_101 (Math|Float|Round _z_4))
  (bind _returnvalue_102 (Math|Conversions|ToFloat(Integer) _returnvalue_101))
  (bind _returnvalue_103 (Math|Float|Round _y_4))
  (bind _returnvalue_104 (Math|Conversions|ToFloat(Integer) _returnvalue_103))
  (bind _returnvalue_105 (Math|Float|Round _x_4))
  (bind _returnvalue_106 (Math|Conversions|ToFloat(Integer) _returnvalue_105))
  (bind _returnvalue_107 (Math|Vector|MakeVector _returnvalue_106 _returnvalue_104 _returnvalue_102))
  (bind _returnvalue_108 (Math|Float|Floor _x_2))
  (bind _returnvalue_109 (Math|Conversions|ToFloat(Integer) _returnvalue_108))
  (bind _returnvalue_110 (Math|Float|Floor _y_2))
  (bind _returnvalue_111 (Math|Conversions|ToFloat(Integer) _returnvalue_110))
  (bind _returnvalue_112 (Math|Float|Floor _z_2))
  (bind _returnvalue_113 (Math|Conversions|ToFloat(Integer) _returnvalue_112))
  (bind _returnvalue_114 (Math|Float|Floor _w_2))
  (bind _returnvalue_115 (Math|Conversions|ToFloat(Integer) _returnvalue_114))
  (bind _returnvalue_116 (Math|Float|Min(Float) _w_2 _w))
  (bind _returnvalue_117 (Math|Float|Min(Float) _z_2 _z))
  (bind _returnvalue_118 (Math|Float|Min(Float) _y_2 _y))
  (bind _returnvalue_119 (Math|Float|Min(Float) _x_2 _x))
  (bind _returnvalue_120 (Math|Vector4|MakeVector4 _returnvalue_119 _returnvalue_118 _returnvalue_117 _returnvalue_116))
  (bind _returnvalue_121 (Math|Vector4|MakeVector4 _returnvalue_109 _returnvalue_111 _returnvalue_113 _returnvalue_115))
  (bind _returnvalue_122 (Math|Conversions|ToFloat(Integer) _returnvalue_40))
  (bind _returnvalue_123 (Math|Conversions|ToFloat(Integer) _returnvalue_100))
  (bind _returnvalue_124 (Math|Float|Fraction _inmetadataafloat))
  (bind _returnvalue_125 (Math|Float|Floor _x_4))
  (bind _returnvalue_126 (Math|Float|Floor _y_4))
  (bind _returnvalue_127 (Math|Float|Floor _z_4))
  (bind _returnvalue_128 (Math|Float|Absolute(Float) _inmetadataafloat))
  (bind _returnvalue_129 (Math|Vector|MakeVector _returnvalue_39 _returnvalue_42 _returnvalue_44))
  (bind _returnvalue_130 (Math|Vector|MakeVector _returnvalue_58 _returnvalue_59 _returnvalue_60))
  (bind _returnvalue_131 (Math|Conversions|ToFloat(Integer) _returnvalue_127))
  (bind _returnvalue_132 (Math|Conversions|ToFloat(Integer) _returnvalue_126))
  (bind _returnvalue_133 (Math|Conversions|ToFloat(Integer) _returnvalue_125))
  (bind _returnvalue_134 (Math|Vector|MakeVector _returnvalue_133 _returnvalue_132 _returnvalue_131))
  (bind _returnvalue_135 (Math|Float|Min(Float) _inmetadataafloat _inmetadatabfloat))
  (bind _returnvalue_136 (Math|Float|Max(Float) _inmetadataafloat _inmetadatabfloat))
  (bind _inmetadatabint64 (LocalVariables|GetInMetadataBInt64))
  (bind _returnvalue_137 (Math|Integer64|Max(Integer64) _inmetadataaint64 _inmetadatabint64))
  (bind _returnvalue_138 (Math|Integer64|Min(Integer64) _inmetadataaint64 _inmetadatabint64))
  (bind _output (Utilities|Array|Get(acopy) _attributenames_2 _returnvalue_2))
  (bind _output_2 (Utilities|Array|Get(aref) _attributenames _returnvalue))
  (bind _output_3 (Utilities|Array|Get(aref) _attributetypes_2 _returnvalue_2))
  (bind _returnvalue_139 (* _inmetadataafloat _inmetadatabfloat))
  (bind _returnvalue_140 (* _inmetadataavector _inmetadatabvector))
  (bind _returnvalue_141 (* _inmetadataavector4 _inmetadatabvector4))
  (bind _returnvalue_142 (+ _inmetadataaint64 _inmetadatabint64))
  (bind _returnvalue_143 (- _inmetadataaint64 _inmetadatabint64))
  (bind _returnvalue_144 (* _inmetadataaint64 _inmetadatabint64))
  (bind _returnvalue_145 (/ _inmetadataaint64 _inmetadatabint64))
  (bind _returnvalue_146 (+ _inmetadataafloat _inmetadatabfloat))
  (bind _returnvalue_147 (- _inmetadataafloat _inmetadatabfloat))
  (bind _returnvalue_148 (+ _inmetadataavector _inmetadatabvector))
  (bind _returnvalue_149 (- _inmetadataavector _inmetadatabvector))
  (bind _returnvalue_150 (+ _inmetadataavector4 _inmetadatabvector4))
  (bind _returnvalue_151 (- _inmetadataavector4 _inmetadatabvector4))
  (bind _operationmode (Variables|Default|GetOperationMode))
  (bind _returnvalue_152 (Utilities|Select :Index _operationmode :NewEnumerator0 _inmetadataafloat :NewEnumerator2 _returnvalue_135 :NewEnumerator3 _returnvalue_136 :NewEnumerator4 _returnvalue_146 :NewEnumerator5 _returnvalue_147 :NewEnumerator6 _returnvalue_139 :NewEnumerator7 _returnvalue_66 :NewEnumerator8 _returnvalue_93 :NewEnumerator9 _returnvalue_95 :NewEnumerator10 _returnvalue_122 :NewEnumerator11 _returnvalue_123 :NewEnumerator12 _returnvalue_128 :NewEnumerator13 _returnvalue_124))
  (bind _operationmode_2 (Variables|Default|GetOperationMode))
  (bind _returnvalue_153 (Utilities|Select :Index _operationmode_2 :NewEnumerator0 _inmetadataavector :NewEnumerator2 _returnvalue_23 :NewEnumerator3 _returnvalue_99 :NewEnumerator4 _returnvalue_148 :NewEnumerator5 _returnvalue_149 :NewEnumerator6 _returnvalue_140 :NewEnumerator7 _returnvalue_130 :NewEnumerator8 _returnvalue_107 :NewEnumerator9 _returnvalue_74 :NewEnumerator10 _returnvalue_134 :NewEnumerator11 _returnvalue_79 :NewEnumerator12 _returnvalue_48 :NewEnumerator13 _returnvalue_129))
  (bind _operationmode_3 (Variables|Default|GetOperationMode))
  (bind _returnvalue_154 (Utilities|Select :Index _operationmode_3 :NewEnumerator0 _inmetadataavector4 :NewEnumerator2 _returnvalue_120 :NewEnumerator3 _returnvalue_84 :NewEnumerator4 _returnvalue_150 :NewEnumerator5 _returnvalue_151 :NewEnumerator6 _returnvalue_141 :NewEnumerator7 _returnvalue_87 :NewEnumerator8 _returnvalue_62 :NewEnumerator9 _returnvalue_32 :NewEnumerator10 _returnvalue_121 :NewEnumerator11 _returnvalue_57 :NewEnumerator12 _returnvalue_65 :NewEnumerator13 _returnvalue_91))
  (bind _operationmode_4 (Variables|Default|GetOperationMode))
  (bind _returnvalue_155 (Utilities|Select :Index _operationmode_4 :NewEnumerator0 _inmetadataaint64 :NewEnumerator2 _returnvalue_138 :NewEnumerator3 _returnvalue_137 :NewEnumerator4 _returnvalue_142 :NewEnumerator5 _returnvalue_143 :NewEnumerator6 _returnvalue_144 :NewEnumerator7 _returnvalue_145 :NewEnumerator8 _inmetadataaint64 :NewEnumerator9 _inmetadataaint64 :NewEnumerator10 _inmetadataaint64 :NewEnumerator11 _inmetadataaint64 :NewEnumerator12 _returnvalue_67 :NewEnumerator13 0))
  (bind _attributenameout (Variables|Default|GetAttributeNameOut))
  (bind _inmetadatab_2 (LocalVariables|GetInMetaDataB))
  (bind _attributenameout_2 (Variables|Default|GetAttributeNameOut))
  (bind _attributenameout_3 (Variables|Default|GetAttributeNameOut))
  (bind _attributenameout_4 (Variables|Default|GetAttributeNameOut))
  (bind _inputcollection (LocalVariables|GetInputCollection))
  (LocalVariables|SetContext InContext)
  (LocalVariables|SetInputCollection Input)
  (bind _result (Utilities|Set|ToArray (PCG|Input&Output|CustomInputLabels self)))
  (bind _returnvalue_156 (Data|GetParamsbyPinLabel _inputcollection (Utilities|Array|Get(aref) _result)))
  (bind _returnvalue_157 (Data|GetParamsbyPinLabel _inputcollection (Utilities|Array|Get(aref) _result 1)))
  (if (> (Utilities|Array|Length _returnvalue_156) 0)
    (bind _aspcgattribute_set (Utilities|Casting|CastToPCGParamData (PCG|TaggedData|BreakTaggedData (Utilities|Array|Get(aref) _returnvalue_156))))
    (bind _returnvalue_158 (Metadata|MutableMetadata _aspcgattribute_set))
    (LocalVariables|SetInMetadataA _returnvalue_158)
    (LocalVariables|SetAttributeType _output_3)
    (switch Utilities|FlowControl|Switch|SwitchonEPCGMetadataTypes _output_3
      (:Float
        (bind _returnvalue_159 (PCG|Metadata|GetFloatAttributebyMetadataKey 0 (LocalVariables|GetInMetadataA) _output))
        (LocalVariables|SetInMetadataAFloat _returnvalue_159))
      (:Integer64
        (bind _returnvalue_160 (PCG|Metadata|GetInteger64AttributebyMetadataKey 0 (LocalVariables|GetInMetadataA) _output))
        (LocalVariables|SetInMetadataAInt64 _returnvalue_160))
      (:Vector
        (bind _returnvalue_161 (PCG|Metadata|GetVectorAttributebyMetadataKey 0 (LocalVariables|GetInMetadataA) _output))
        (LocalVariables|SetInMetadataAVector _returnvalue_161))
      (:Vector4
        (bind _returnvalue_162 (PCG|Metadata|GetVector4AttributebyMetadataKey 0 (LocalVariables|GetInMetadataA) _output))
        (LocalVariables|SetInMetadataAVector4 _returnvalue_162)))
    (switch Utilities|FlowControl|Switch|SwitchonPCGAttributeOperationMode (Variables|Default|GetOperationMode)
      (:NewEnumerator0
        _)
      (:NewEnumerator2
        (bind _aspcgattribute_set_2 (Utilities|Casting|CastToPCGParamData (PCG|TaggedData|BreakTaggedData (Utilities|Array|Get(aref) _returnvalue_157)) :!label _label))
        (bind _returnvalue_163 (Metadata|MutableMetadata _aspcgattribute_set_2))
        (LocalVariables|SetInMetaDataB _returnvalue_163)
        (switch Utilities|FlowControl|Switch|SwitchonEPCGMetadataTypes (LocalVariables|GetAttributeType)
          (:Float
            (bind _returnvalue_164 (PCG|Metadata|GetFloatAttributebyMetadataKey 0 _inmetadatab_2 _output_2))
            (LocalVariables|SetInMetadataBFloat _returnvalue_164))
          (:Integer64
            (bind _returnvalue_165 (PCG|Metadata|GetInteger64AttributebyMetadataKey 0 _inmetadatab_2 _output_2))
            (LocalVariables|SetInMetadataBInt64 _returnvalue_165))
          (:Vector
            (bind _returnvalue_166 (PCG|Metadata|GetVectorAttributebyMetadataKey 0 _inmetadatab_2 _output_2))
            (LocalVariables|SetInMetadataBVector _returnvalue_166))
          (:Vector4
            (bind _returnvalue_167 (PCG|Metadata|GetVector4AttributebyMetadataKey 0 _inmetadatab_2 _output_2))
            (LocalVariables|SetInMetadataBVector4 _returnvalue_167))))
      (:NewEnumerator3
        (goto _label))
      (:NewEnumerator4
        (goto _label))
      (:NewEnumerator5
        (goto _label))
      (:NewEnumerator6
        (goto _label))
      (:NewEnumerator7
        (goto _label))
      (:NewEnumerator8
        _)
      (:NewEnumerator9
        _)
      (:NewEnumerator10
        _)
      (:NewEnumerator11
        _)
      (:NewEnumerator12
        _)
      (:NewEnumerator13
        _))
    (bind _returnvalue_168 (Game|ConstructObjectfromClass "/Script/PCG.PCGParamData"))
    (bind _returnvalue_169 (Metadata|MutableMetadata _returnvalue_168))
    (bind _output_get (LocalVariables|SetOutMetadata _returnvalue_169))
    (bind _returnvalue_170 (PCG|Metadata|AddEntry _output_get -1))
    (LocalVariables|SetOutMetaDataEntry _returnvalue_170)
    (switch Utilities|FlowControl|Switch|SwitchonEPCGMetadataTypes (LocalVariables|GetAttributeType)
      (:Float
        (PCG|Metadata|CreateFloatAttribute (LocalVariables|GetOutMetadata) _attributenameout_4 _returnvalue_152 false true)
        (PCG|Metadata|SetFloatAttributebyMetadataKey (LocalVariables|GetOutMetaDataEntry) (LocalVariables|GetOutMetadata) _attributenameout_4 _returnvalue_152))
      (:Integer64
        (PCG|Metadata|CreateInteger64Attribute (LocalVariables|GetOutMetadata) _attributenameout_2 _returnvalue_155 false true)
        (PCG|Metadata|SetInteger64AttributebyMetadataKey (LocalVariables|GetOutMetaDataEntry) (LocalVariables|GetOutMetadata) _attributenameout_2 _returnvalue_155))
      (:Vector
        (PCG|Metadata|CreateVectorAttribute (LocalVariables|GetOutMetadata) _attributenameout_3 _returnvalue_153 false true)
        (PCG|Metadata|SetVectorAttributebyMetadataKey (LocalVariables|GetOutMetaDataEntry) (LocalVariables|GetOutMetadata) _attributenameout_3 _returnvalue_153))
      (:Vector4
        (PCG|Metadata|CreateVector4Attribute (LocalVariables|GetOutMetadata) _attributenameout _returnvalue_154 false true)
        (PCG|Metadata|SetVector4AttributebyMetadataKey (LocalVariables|GetOutMetaDataEntry) (LocalVariables|GetOutMetadata) _attributenameout _returnvalue_154)))
    (return (Utilities|Struct|MakePCGDataCollection (Utilities|Array|MakeArray (PCG|TaggedData|MakeTaggedData _returnvalue_168))))
    (else
      (return 0))))

(fn NodeColorOverride
  (return "(R=1.000000,G=0.600000,B=0.000000,A=1.000000)"))

(fn NodeTitleOverride
  (return (Utilities|String|StringToName (Utilities|String|Append "Param " (Utilities|String|EnumtoString (Variables|Default|GetOperationMode))))))
