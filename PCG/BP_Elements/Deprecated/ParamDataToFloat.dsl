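; Deprecated custom PCG Blueprint element that converts an attribute from input param data into a single float attribute in new param data.
; It reads entry 0 of the last attribute on the first input param, switching on metadata type.
; Float passes through, Integer64 is converted, and Vector or Vector4 yields the component chosen by AttributeComponent; other types become zero.
; The result is written under AttributeNameOut, and the node color is overridden to orange.
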(fn Execute
  (params (Input :type PCGData%20Collection%20Structure%20%28by%20ref%29))
  (bind _inmetadata (LocalVariables|GetInMetadata))
  (bind (_attributenames _attributetypes) (PCG|Metadata|GetAttributes _inmetadata))
  (bind _returnvalue (Utilities|Array|LastIndex _attributenames))
  (bind _output (Utilities|Array|Get(aref) _attributenames _returnvalue))
  (bind _output_2 (Utilities|Array|Get(aref) _attributetypes _returnvalue))
  (bind _inmetadata_2 (LocalVariables|GetInMetadata))
  (bind _attributecomponent (Variables|Default|GetAttributeComponent))
  (bind _returnvalue_2 (Data|GetParams Input))
  (bind _aspcgattribute_set (Utilities|Casting|CastToPCGParamData (PCG|TaggedData|BreakTaggedData (Utilities|Array|Get(aref) _returnvalue_2))))
  (bind _returnvalue_3 (Metadata|MutableMetadata _aspcgattribute_set))
  (LocalVariables|SetInMetadata _returnvalue_3)
  (bind _returnvalue_4 (Game|ConstructObjectfromClass "/Script/PCG.PCGParamData"))
  (bind _returnvalue_5 (Metadata|MutableMetadata _returnvalue_4))
  (bind _output_get (LocalVariables|SetOutMetadata _returnvalue_5))
  (bind _returnvalue_6 (PCG|Metadata|AddEntry _output_get -1))
  (LocalVariables|SetMetadataEntryKey _returnvalue_6)
  (switch Utilities|FlowControl|Switch|SwitchonEPCGMetadataTypes _output_2
    (:Float
      (bind _returnvalue_7 (PCG|Metadata|GetFloatAttributebyMetadataKey 0 _inmetadata_2 _output)))
    (:Integer64
      (bind _returnvalue_8 (PCG|Metadata|GetInteger64AttributebyMetadataKey 0 _inmetadata_2 _output)))
    (:Vector
      (bind _returnvalue_9 (PCG|Metadata|GetVectorAttributebyMetadataKey 0 _inmetadata_2 _output)))
    (:Vector4
      (bind _returnvalue_10 (PCG|Metadata|GetVector4AttributebyMetadataKey 0 _inmetadata_2 _output))))
  (PCG|Metadata|CreateFloatAttribute (LocalVariables|GetOutMetadata) (Variables|Default|GetAttributeNameOut) (Utilities|Select :Index _output_2 :Float _returnvalue_7 :Double 0.0 :Integer32 0.0 :Integer64 (Math|Conversions|ToFloat(Integer64) _returnvalue_8) :Vector2 0.0 :Vector (Utilities|Select :Index _attributecomponent :NewEnumerator0 (.x _returnvalue_9) :NewEnumerator1 (.y _returnvalue_9) :NewEnumerator2 (.z _returnvalue_9) :NewEnumerator3 0.0) :Vector4 (Utilities|Select :Index _attributecomponent :NewEnumerator0 (Math|Vector4|BreakVector4 _returnvalue_10) :NewEnumerator1 (.Y (Math|Vector4|BreakVector4 _returnvalue_10)) :NewEnumerator2 (.Z (Math|Vector4|BreakVector4 _returnvalue_10)) :NewEnumerator3 (.W (Math|Vector4|BreakVector4 _returnvalue_10))) :Quaternion 0.0 :Transform 0.0 :String 0.0 :Boolean 0.0 :Rotator 0.0 :Name 0.0 :SoftObjectPath 0.0 :SoftClassPath 0.0) false true)
  (PCG|Metadata|SetFloatAttributebyMetadataKey (LocalVariables|GetMetadataEntryKey) (LocalVariables|GetOutMetadata) (Variables|Default|GetAttributeNameOut) (Utilities|Select :Index _output_2 :Float _returnvalue_7 :Double 0.0 :Integer32 0.0 :Integer64 (Math|Conversions|ToFloat(Integer64) _returnvalue_8) :Vector2 0.0 :Vector (Utilities|Select :Index _attributecomponent :NewEnumerator0 (.x _returnvalue_9) :NewEnumerator1 (.y _returnvalue_9) :NewEnumerator2 (.z _returnvalue_9) :NewEnumerator3 0.0) :Vector4 (Utilities|Select :Index _attributecomponent :NewEnumerator0 (Math|Vector4|BreakVector4 _returnvalue_10) :NewEnumerator1 (.Y (Math|Vector4|BreakVector4 _returnvalue_10)) :NewEnumerator2 (.Z (Math|Vector4|BreakVector4 _returnvalue_10)) :NewEnumerator3 (.W (Math|Vector4|BreakVector4 _returnvalue_10))) :Quaternion 0.0 :Transform 0.0 :String 0.0 :Boolean 0.0 :Rotator 0.0 :Name 0.0 :SoftObjectPath 0.0 :SoftClassPath 0.0))
  (return (Utilities|Struct|MakePCGDataCollection (Utilities|Array|MakeArray (PCG|TaggedData|MakeTaggedData _returnvalue_4)))))

(fn NodeColorOverride
  (return "(R=1.000000,G=0.600000,B=0.000000,A=1.000000)"))
