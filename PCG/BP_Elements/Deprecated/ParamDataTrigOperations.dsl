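; Custom PCG element (in a Deprecated folder) that applies a selectable trig or degree/radian conversion to attribute set values, titled "Param" plus the mode.
; It reads the last attribute at entry 0 of the first input, handling only Float.
; A Select picks the result by OperationMode, but only the Atan2 branch is implemented.
; That branch reads a second input and returns new param data holding the result under AttributeNameOut.
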(fn ExecuteWithContext
  (params (InContext :type PCGContext%20Structure%20%28by%20ref%29) (Input :type PCGData%20Collection%20Structure%20%28by%20ref%29))
  (bind _inmetadataa (LocalVariables|GetInMetadataA))
  (bind (_attributenames _attributetypes) (PCG|Metadata|GetAttributes _inmetadataa))
  (bind _returnvalue (Utilities|Array|LastIndex _attributenames))
  (bind _inmetadataafloat (LocalVariables|GetInMetadataAFloat))
  (bind _returnvalue_2 (Math|Trig|Sin(Radians) _inmetadataafloat))
  (bind _returnvalue_3 (Math|Trig|Tan(Radians) _inmetadataafloat))
  (bind _inmetadatabfloat (LocalVariables|GetInMetadataBFloat))
  (bind _returnvalue_4 (Math|Trig|Atan2(Radians) _inmetadataafloat _inmetadatabfloat))
  (bind _returnvalue_5 (Math|Trig|DegreesToRadians _inmetadataafloat))
  (bind _returnvalue_6 (Math|Trig|RadiansToDegrees _inmetadataafloat))
  (bind _returnvalue_7 (Math|Trig|Acos(Radians) _inmetadataafloat))
  (bind _returnvalue_8 (Math|Trig|Asin(Radians) _inmetadataafloat))
  (bind _returnvalue_9 (Math|Trig|Atan(Radians) _inmetadataafloat))
  (bind _returnvalue_10 (Math|Trig|Cos(Radians) _inmetadataafloat))
  (bind _output (Utilities|Array|Get(aref) _attributetypes _returnvalue))
  (bind _operationmode (Variables|Default|GetOperationMode))
  (bind _returnvalue_11 (Utilities|Select :Index _operationmode :NewEnumerator0 _returnvalue_7 :NewEnumerator1 _returnvalue_8 :NewEnumerator2 _returnvalue_9 :NewEnumerator6 _returnvalue_4 :NewEnumerator3 _returnvalue_10 :NewEnumerator4 _returnvalue_2 :NewEnumerator5 _returnvalue_3 :NewEnumerator7 _returnvalue_5 :NewEnumerator8 _returnvalue_6))
  (bind _attributenameout (Variables|Default|GetAttributeNameOut))
  (bind _inputcollection (LocalVariables|GetInputCollection))
  (LocalVariables|SetContext InContext)
  (LocalVariables|SetInputCollection Input)
  (bind _result (Utilities|Set|ToArray (PCG|Input&Output|CustomInputLabels self)))
  (bind _returnvalue_12 (Data|GetParamsbyPinLabel _inputcollection (Utilities|Array|Get(aref) _result)))
  (bind _returnvalue_13 (Data|GetParamsbyPinLabel _inputcollection (Utilities|Array|Get(aref) _result 1)))
  (bind _aspcgattribute_set (Utilities|Casting|CastToPCGParamData (PCG|TaggedData|BreakTaggedData (Utilities|Array|Get(aref) _returnvalue_12))))
  (bind _returnvalue_14 (Metadata|MutableMetadata _aspcgattribute_set))
  (LocalVariables|SetInMetadataA _returnvalue_14)
  (LocalVariables|SetAttributeType _output)
  (switch Utilities|FlowControl|Switch|SwitchonEPCGMetadataTypes _output
    (:Float
      (bind _returnvalue_15 (PCG|Metadata|GetFloatAttributebyMetadataKey 0 (LocalVariables|GetInMetadataA) (Utilities|Array|Get(acopy) _attributenames _returnvalue)))
      (LocalVariables|SetInMetadataAFloat _returnvalue_15)
      (switch Utilities|FlowControl|Switch|SwitchonPCGAttributeTrigOperationMode (Variables|Default|GetOperationMode)
        (:NewEnumerator0
          _)
        (:NewEnumerator1
          _)
        (:NewEnumerator2
          _)
        (:NewEnumerator6
          (bind _aspcgattribute_set_2 (Utilities|Casting|CastToPCGParamData (PCG|TaggedData|BreakTaggedData (Utilities|Array|Get(aref) _returnvalue_13))))
          (bind _returnvalue_16 (Metadata|MutableMetadata _aspcgattribute_set_2))
          (LocalVariables|SetInMetaDataB _returnvalue_16)
          (bind _returnvalue_17 (PCG|Metadata|GetFloatAttributebyMetadataKey 0 (LocalVariables|GetInMetaDataB) (Utilities|Array|Get(aref) (PCG|Metadata|GetAttributes (LocalVariables|GetInMetaDataB)) (Utilities|Array|LastIndex (PCG|Metadata|GetAttributes (LocalVariables|GetInMetaDataB))))))
          (LocalVariables|SetInMetadataBFloat _returnvalue_17)
          (bind _returnvalue_18 (Game|ConstructObjectfromClass "/Script/PCG.PCGParamData"))
          (bind _returnvalue_19 (Metadata|MutableMetadata _returnvalue_18))
          (bind _output_get (LocalVariables|SetOutMetadata _returnvalue_19))
          (bind _returnvalue_20 (PCG|Metadata|AddEntry _output_get -1))
          (LocalVariables|SetOutMetaDataEntry _returnvalue_20)
          (PCG|Metadata|CreateFloatAttribute (LocalVariables|GetOutMetadata) _attributenameout _returnvalue_11 false true)
          (PCG|Metadata|SetFloatAttributebyMetadataKey (LocalVariables|GetOutMetaDataEntry) (LocalVariables|GetOutMetadata) _attributenameout _returnvalue_11)
          (return (Utilities|Struct|MakePCGDataCollection (Utilities|Array|MakeArray (PCG|TaggedData|MakeTaggedData _returnvalue_18)))))
        (:NewEnumerator3
          _)
        (:NewEnumerator4
          _)
        (:NewEnumerator5
          _)
        (:NewEnumerator7
          _)
        (:NewEnumerator8
          _)))))

(fn NodeColorOverride
  (return "(R=1.000000,G=0.600000,B=0.000000,A=1.000000)"))

(fn NodeTitleOverride
  (return (Utilities|String|StringToName (Utilities|String|Append "Param " (Utilities|String|EnumtoString (Variables|Default|GetOperationMode))))))
