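; Deprecated custom PCG Blueprint node, titled "VectorEqual Param", that tests whether vectors from two attribute sets are equal.
; It takes the first param data on each of its two custom input pins and reads the last attribute as a vector at metadata key 0.
; It compares them within a small tolerance and outputs a new attribute set holding the result as a float, named by the AttributeNameOut variable.
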(fn ExecuteWithContext
  (params (InContext :type PCGContext%20Structure%20%28by%20ref%29) (Input :type PCGData%20Collection%20Structure%20%28by%20ref%29))
  (bind (_attributenames _attributetypes) (PCG|Metadata|GetAttributes (LocalVariables|GetInMetaDataB)))
  (bind _inmetadatabvector (LocalVariables|GetInMetadataBVector))
  (bind _inmetadataavector (LocalVariables|GetInMetadataAVector))
  (bind _returnvalue (Utilities|Operators|Equal(==) _inmetadataavector _inmetadatabvector 0.0001))
  (bind _returnvalue_2 (Math|Conversions|ToFloat(Boolean) _returnvalue))
  (bind _attributenameout (Variables|Default|GetAttributeNameOut))
  (bind _inputcollection (LocalVariables|GetInputCollection))
  (LocalVariables|SetContext InContext)
  (LocalVariables|SetInputCollection Input)
  (bind _result (Utilities|Set|ToArray (PCG|Input&Output|CustomInputLabels self)))
  (bind _returnvalue_3 (Data|GetParamsbyPinLabel _inputcollection (Utilities|Array|Get(aref) _result)))
  (bind _returnvalue_4 (Data|GetParamsbyPinLabel _inputcollection (Utilities|Array|Get(aref) _result 1)))
  (bind _aspcgattribute_set (Utilities|Casting|CastToPCGParamData (PCG|TaggedData|BreakTaggedData (Utilities|Array|Get(aref) _returnvalue_3))))
  (bind _returnvalue_5 (Metadata|MutableMetadata _aspcgattribute_set))
  (LocalVariables|SetInMetadataA _returnvalue_5)
  (bind _returnvalue_6 (PCG|Metadata|GetVectorAttributebyMetadataKey 0 (LocalVariables|GetInMetadataA) (Utilities|Array|Get(acopy) (PCG|Metadata|GetAttributes (LocalVariables|GetInMetadataA)) (Utilities|Array|LastIndex (PCG|Metadata|GetAttributes (LocalVariables|GetInMetadataA))))))
  (LocalVariables|SetInMetadataAVector _returnvalue_6)
  (bind _aspcgattribute_set_2 (Utilities|Casting|CastToPCGParamData (PCG|TaggedData|BreakTaggedData (Utilities|Array|Get(aref) _returnvalue_4))))
  (bind _returnvalue_7 (Metadata|MutableMetadata _aspcgattribute_set_2))
  (LocalVariables|SetInMetaDataB _returnvalue_7)
  (bind _returnvalue_8 (PCG|Metadata|GetVectorAttributebyMetadataKey 0 (LocalVariables|GetInMetaDataB) (Utilities|Array|Get(aref) _attributenames (Utilities|Array|LastIndex _attributetypes))))
  (LocalVariables|SetInMetadataBVector _returnvalue_8)
  (bind _returnvalue_9 (Game|ConstructObjectfromClass "/Script/PCG.PCGParamData"))
  (bind _returnvalue_10 (Metadata|MutableMetadata _returnvalue_9))
  (bind _output_get (LocalVariables|SetOutMetadata _returnvalue_10))
  (bind _returnvalue_11 (PCG|Metadata|AddEntry _output_get -1))
  (LocalVariables|SetOutMetaDataEntry _returnvalue_11)
  (PCG|Metadata|CreateFloatAttribute (LocalVariables|GetOutMetadata) _attributenameout _returnvalue_2 false true)
  (PCG|Metadata|SetFloatAttributebyMetadataKey (LocalVariables|GetOutMetaDataEntry) (LocalVariables|GetOutMetadata) _attributenameout _returnvalue_2)
  (return (Utilities|Struct|MakePCGDataCollection (Utilities|Array|MakeArray (PCG|TaggedData|MakeTaggedData _returnvalue_9)))))

(fn NodeColorOverride
  (return "(R=1.000000,G=0.600000,B=0.000000,A=1.000000)"))

(fn NodeTitleOverride
  (return "VectorEqual Param"))
