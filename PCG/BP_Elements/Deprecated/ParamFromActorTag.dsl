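; Custom PCG Blueprint element, filed as deprecated, that outputs param data holding transform values read from level actors with a configured tag.
; It creates one metadata entry, then optionally writes Location, rotation as RotationX/Y/Z axis vectors, and Scale as vector attributes.
; Every actor writes to that same entry, so later actors overwrite earlier ones.
; Scale is nested under the rotation branch, so it requires both GetRotation and GetScale.
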(fn ExecuteWithContext
  (params (InContext :type PCGContext%20Structure%20%28by%20ref%29) (Input :type PCGData%20Collection%20Structure%20%28by%20ref%29))
  (bind _returnvalue (Utilities|Name|MakeLiteralName "RotationZ"))
  (bind _returnvalue_2 (Utilities|Name|MakeLiteralName "Scale"))
  (bind _returnvalue_3 (Utilities|Name|MakeLiteralName "RotationX"))
  (bind _returnvalue_4 (Utilities|Name|MakeLiteralName "RotationY"))
  (bind _returnvalue_5 (Utilities|Name|MakeLiteralName "Location"))
  (bind _metadata (LocalVariables|GetMetadata))
  (bind _metadata_2 (LocalVariables|GetMetadata))
  (bind _metadata_3 (LocalVariables|GetMetadata))
  (bind _metadata_4 (LocalVariables|GetMetadata))
  (bind _metadata_5 (LocalVariables|GetMetadata))
  (Utilities|FlowControl|Sequence
    (:then_0
      (bind _returnvalue_6 (Game|ConstructObjectfromClass "/Script/PCG.PCGParamData"))
      (bind _returnvalue_7 (Metadata|MutableMetadata _returnvalue_6))
      (bind _output_get (LocalVariables|SetMetadata _returnvalue_7))
      (bind _returnvalue_8 (PCG|Metadata|AddEntry _output_get -1))
      (LocalVariables|SetMetadataEntryKey _returnvalue_8))
    (:then_1
      (bind _outactors (Actor|GetAllActorswithTag (Variables|Default|GetActorTag)))
      (for _array_element _outactors
        (if (Variables|Default|GetLocation)
          (PCG|Metadata|CreateVectorAttribute _metadata_5 _returnvalue_5 (Math|Vector|MakeVector 0.0) false true)
          (PCG|Metadata|SetVectorAttributebyMetadataKey (LocalVariables|GetMetadataEntryKey) _metadata_5 _returnvalue_5 (.location (Transformation|GetActorTransform _array_element))))
        (if (Variables|Default|GetRotation)
          (PCG|Metadata|CreateVectorAttribute _metadata_3 _returnvalue_3 (Math|Vector|MakeVector 0.0) false true)
          (PCG|Metadata|SetVectorAttributebyMetadataKey (LocalVariables|GetMetadataEntryKey) _metadata_3 _returnvalue_3 (Math|Rotator|BreakRotIntoAxes (.rotation (Transformation|GetActorTransform _array_element))))
          (PCG|Metadata|CreateVectorAttribute _metadata_4 _returnvalue_4 (Math|Vector|MakeVector 0.0) false true)
          (PCG|Metadata|SetVectorAttributebyMetadataKey (LocalVariables|GetMetadataEntryKey) _metadata_4 _returnvalue_4 (.Y (Math|Rotator|BreakRotIntoAxes (.rotation (Transformation|GetActorTransform _array_element)))))
          (PCG|Metadata|CreateVectorAttribute _metadata _returnvalue (Math|Vector|MakeVector 0.0) false true)
          (PCG|Metadata|SetVectorAttributebyMetadataKey (LocalVariables|GetMetadataEntryKey) _metadata _returnvalue (.Z (Math|Rotator|BreakRotIntoAxes (.rotation (Transformation|GetActorTransform _array_element)))))
          (if (Variables|Default|GetScale)
            (PCG|Metadata|CreateVectorAttribute _metadata_2 _returnvalue_2 (Math|Vector|MakeVector 0.0) false true)
            (PCG|Metadata|SetVectorAttributebyMetadataKey (LocalVariables|GetMetadataEntryKey) _metadata_2 _returnvalue_2 (.scale (Transformation|GetActorTransform _array_element))))))
      (return (Utilities|Struct|MakePCGDataCollection (Utilities|Array|MakeArray (PCG|TaggedData|MakeTaggedData _returnvalue_6)))))))

(fn NodeColorOverride
  (return "(R=1.000000,G=0.600000,B=0.000000,A=1.000000)"))
