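; Deprecated PCG Blueprint element that outputs one point for each actor of a configured class in the level.
; ExecuteWithContext gathers the actors, creates point data with name, class and tag attributes, then runs IterationLoop over them.
; Each iteration takes bounds from the actor or defaults, optionally moving scale from transform into bounds, and seeds by position.
; Optional culling drops hidden actors and those with zero density in the volume's spatial data.
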(fn ExecuteWithContext
  (params (InContext :type PCGContext%20Structure%20%28by%20ref%29) (Input :type PCGData%20Collection%20Structure%20%28by%20ref%29))
  (bind _outactors (Actor|GetAllActorsOfClass (Variables|Default|GetActorClass)))
  (Variables|Hidden|SetActors _outactors)
  (bind _returnvalue (Game|ConstructObjectfromClass "/Script/PCG.PCGPointData"))
  (bind _returnvalue_2 (PCG|Temporary|GetComponent InContext))
  (Variables|/Script/PCG.PCGSpatialData|SetTargetActor (Components|GetOwner _returnvalue_2) _returnvalue)
  (bind _returnvalue_3 (Metadata|MutableMetadata _returnvalue))
  (PCG|Metadata|CreateStringAttribute _returnvalue_3 "ActorName" "" false true)
  (PCG|Metadata|CreateStringAttribute _returnvalue_3 "ActorClass" (Utilities|GetClassDisplayName (Variables|Default|GetActorClass)) false true)
  (PCG|Metadata|CreateStringAttribute _returnvalue_3 "ActorTag" "None" false true)
  (bind _returnvalue_4 (PCG|Temporary|GetActorData InContext))
  (bind _aspcgspatial_data (Utilities|Casting|CastToPCGSpatialData _returnvalue_4))
  (Variables|Hidden|SetVolumeActorSpatialData _aspcgspatial_data)
  (bind _outdata (PCG|FlowControl|IterationLoop InContext (Math|Conversions|ToInteger64(Integer) (Utilities|Array|Length (Variables|Hidden|GetActors))) 0 0 _returnvalue))
  (return (Utilities|Struct|MakePCGDataCollection (Utilities|Array|MakeArray (PCG|TaggedData|MakeTaggedData :Data _outdata)))))

(fn IterationLoopBody
  (params (InContext :type PCGContext%20Structure%20%28by%20ref%29) (Iteration :type Integer64) (InA :type PCGSpatial%20Data%20Object%20Reference) (InB :type PCGSpatial%20Data%20Object%20Reference) (OutMetadata :type PCGMetadata%20Object%20Reference))
  (bind _returnvalue (Math|Conversions|ToInteger(Integer64) Iteration))
  (bind _applyscaletobounds (Variables|Default|GetApplyScaleToBounds))
  (bind _actors (Variables|Hidden|GetActors))
  (bind _output (Utilities|Array|Get(acopy) _actors _returnvalue))
  (bind _returnvalue_2 (Transformation|GetActorTransform _output))
  (bind _scale (.scale _returnvalue_2))
  (bind _returnvalue_3 (Math|Vector|SelectVector "1.000000,1.000000,1.000000" _scale _applyscaletobounds))
  (bind _rotation (.rotation _returnvalue_2))
  (bind _location (.location _returnvalue_2))
  (bind _returnvalue_4 (Math|Transform|MakeTransform _location _rotation _returnvalue_3))
  (bind _steepness (Variables|Default|GetSteepness))
  (bind _boundsmax (Variables|Default|GetBoundsMax))
  (bind _setboundsfromactorbounds (Variables|Default|GetSetBoundsFromActorBounds))
  (bind _boundsmin (Variables|Default|GetBoundsMin))
  (bind _density (Variables|Default|GetDensity))
  (bind _returnvalue_5 (PCG|Helpers|GetActorLocalBoundsPCG _output))
  (bind _returnvalue_6 (PCG|Helpers|ComputeSeedfromPosition _location))
  (PCG|Metadata|SetStringAttribute (Utilities|Struct|MakePCGPoint _returnvalue_4 _density (Math|Vector|SelectVector (Math|Vector|SelectVector (* (Utilities|Struct|BreakBox _returnvalue_5) _scale) (Utilities|Struct|BreakBox _returnvalue_5) _applyscaletobounds) _boundsmin _setboundsfromactorbounds) (Math|Vector|SelectVector (Math|Vector|SelectVector (* (.Max (Utilities|Struct|BreakBox _returnvalue_5)) _scale) (.Max (Utilities|Struct|BreakBox _returnvalue_5)) _applyscaletobounds) _boundsmax _setboundsfromactorbounds) "(X=1.000000,Y=1.000000,Z=1.000000,W=1.000000)" _steepness _returnvalue_6) OutMetadata "ActorName" (Utilities|GetObjectName _output))
  (PCG|Metadata|SetStringAttribute (Utilities|Struct|MakePCGPoint _returnvalue_4 _density (Math|Vector|SelectVector (Math|Vector|SelectVector (* (Utilities|Struct|BreakBox _returnvalue_5) _scale) (Utilities|Struct|BreakBox _returnvalue_5) _applyscaletobounds) _boundsmin _setboundsfromactorbounds) (Math|Vector|SelectVector (Math|Vector|SelectVector (* (.Max (Utilities|Struct|BreakBox _returnvalue_5)) _scale) (.Max (Utilities|Struct|BreakBox _returnvalue_5)) _applyscaletobounds) _boundsmax _setboundsfromactorbounds) "(X=1.000000,Y=1.000000,Z=1.000000,W=1.000000)" _steepness _returnvalue_6) OutMetadata "ActorTag" (Utilities|String|ToString(Name) (select (> (Utilities|Array|Length (Variables|/Script/Engine.Actor|GetTags _output)) 0) (Utilities|Array|Get(aref) (Variables|/Script/Engine.Actor|GetTags _output)) "None")))
  (return (Utilities|Struct|MakePCGPoint _returnvalue_4 _density (Math|Vector|SelectVector (Math|Vector|SelectVector (* (Utilities|Struct|BreakBox _returnvalue_5) _scale) (Utilities|Struct|BreakBox _returnvalue_5) _applyscaletobounds) _boundsmin _setboundsfromactorbounds) (Math|Vector|SelectVector (Math|Vector|SelectVector (* (.Max (Utilities|Struct|BreakBox _returnvalue_5)) _scale) (.Max (Utilities|Struct|BreakBox _returnvalue_5)) _applyscaletobounds) _boundsmax _setboundsfromactorbounds) "(X=1.000000,Y=1.000000,Z=1.000000,W=1.000000)" _steepness _returnvalue_6) (and (Math|Boolean|NANDBoolean (Variables|Default|GetCullActorsHiddenInGame) (Variables|/Script/Engine.Actor|GetbHidden _output)) (Math|Boolean|NANDBoolean (Variables|Default|GetCullActorsOutsideVolume) (not (> (Distribution|GetDensityatPosition (Variables|Hidden|GetVolumeActorSpatialData) _location) 0))))))
