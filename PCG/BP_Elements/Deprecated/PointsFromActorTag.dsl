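; Deprecated PCG Blueprint element that outputs one point for each actor in the world carrying the ActorTag.
; It creates point data initialized from the context's actor data, keeping that target actor.
; Each point takes the actor's transform and a position-derived seed, with bounds from actor local bounds or the BoundsMin/BoundsMax variables.
; ApplyScaleToBounds bakes actor scale into actor-derived bounds and resets point scale to one.
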(fn ExecuteWithContext
  (params (InContext :type PCGContext%20Structure%20%28by%20ref%29) (Input :type PCGData%20Collection%20Structure%20%28by%20ref%29))
  (Utilities|FlowControl|Sequence
    (:then_0
      (bind _returnvalue (PCG|Temporary|GetActorData InContext))
      (bind _aspcgspatial_data (Utilities|Casting|CastToPCGSpatialData _returnvalue))
      (LocalVariables|SetInputTargetActor (Variables|/Script/PCG.PCGSpatialData|GetTargetActor _aspcgspatial_data))
      (bind _returnvalue_2 (Game|ConstructObjectfromClass "/Script/PCG.PCGPointData"))
      (SpatialData|InitializefromData _returnvalue_2 _aspcgspatial_data 0 true true)
      (bind _output_get (LocalVariables|SetOutputPCGPointData _returnvalue_2))
      (Variables|/Script/PCG.PCGSpatialData|SetTargetActor (LocalVariables|GetInputTargetActor) _output_get)
      (Utilities|Array|Add (Utilities|Struct|BreakPCGDataCollection (LocalVariables|GetOutputTaggedData)) (PCG|TaggedData|MakeTaggedData _output_get)))
    (:then_1
      (bind _outactors (Actor|GetAllActorswithTag (Variables|Default|GetActorTag)))
      (for _array_element _outactors
        (bind _returnvalue_3 (PCG|Helpers|GetActorLocalBoundsPCG _array_element))
        (bind _returnvalue_4 (PCG|Helpers|ComputeSeedfromPosition (.location (Transformation|GetActorTransform _array_element))))
        (LocalVariables|SetPoint (Utilities|Struct|MakePCGPoint (Math|Transform|MakeTransform (.location (Transformation|GetActorTransform _array_element)) (.rotation (Transformation|GetActorTransform _array_element)) (Math|Vector|SelectVector "1.000000,1.000000,1.000000" (.scale (Transformation|GetActorTransform _array_element)) (Variables|Default|GetApplyScaleToBounds))) 1.0 (Math|Vector|SelectVector (Math|Vector|SelectVector (* (Utilities|Struct|BreakBox _returnvalue_3) (.scale (Transformation|GetActorTransform _array_element))) (Utilities|Struct|BreakBox _returnvalue_3) (Variables|Default|GetApplyScaleToBounds)) (Variables|Default|GetBoundsMin) (Variables|Default|GetSetBoundsFromActorBounds)) (Math|Vector|SelectVector (Math|Vector|SelectVector (* (.Max (Utilities|Struct|BreakBox _returnvalue_3)) (.scale (Transformation|GetActorTransform _array_element))) (.Max (Utilities|Struct|BreakBox _returnvalue_3)) (Variables|Default|GetApplyScaleToBounds)) (Variables|Default|GetBoundsMax) (Variables|Default|GetSetBoundsFromActorBounds)) "(X=1.000000,Y=1.000000,Z=1.000000,W=1.000000)" 1.0 _returnvalue_4))
        (Utilities|Array|Add (LocalVariables|GetOutputPCGPoints) (LocalVariables|GetPoint)))
      (SpatialData|SetPoints (LocalVariables|GetOutputPCGPointData) (LocalVariables|GetOutputPCGPoints))
      (return (LocalVariables|GetOutputTaggedData)))))
