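; PCG Blueprint element that aligns points to landscape surface normals.
; It renders an orthographic world-normal capture over the PCG volume's bounds, padded slightly, into a render target.
; That texture becomes render target data projected onto the 2D landscape input, and the input points are projected onto the result.
; A point loop then turns each point's interpolated color into a rotation whose Z axis follows the normal.
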(fn ExecuteWithContext
  (params (InContext :type PCGContext%20Structure%20%28by%20ref%29) (Input :type PCGData%20Collection%20Structure%20%28by%20ref%29))
  (bind (_origin _boxextent) (Collision|GetComponentBounds (Actor|GetComponentbyClass (LocalVariables|GetPCGVolumeActor) "/Script/Engine.BrushComponent")))
  (bind (_origin_2 _boxextent_2) (Collision|GetActorBounds (LocalVariables|GetPCGVolumeActor) true false))
  (bind _inputcollection (LocalVariables|GetInputCollection))
  (LocalVariables|SetInputCollection Input)
  ;[1 1 1] Prepare inputs
  (bind _result (Utilities|Set|ToArray (PCG|Input&Output|CustomInputLabels self)))
  (bind _returnvalue (Data|GetInputsbyPinLabel _inputcollection (Utilities|Array|Get(aref) _result)))
  (bind _returnvalue_2 (Data|GetInputsbyPinLabel _inputcollection (Utilities|Array|Get(aref) _result 1)))
  (if (and (> (Utilities|Array|Length _returnvalue) 0) (> (Utilities|Array|Length _returnvalue_2) 0))
    (LocalVariables|SetLandscapeData (Utilities|Array|Get(aref) _returnvalue))
    (LocalVariables|SetPointData (Utilities|Array|Get(aref) _returnvalue_2))
    (bind _aspcgspatial_data (Utilities|Casting|CastToPCGSpatialData (PCG|TaggedData|BreakTaggedData (LocalVariables|GetLandscapeData))))
    (if (== (SpatialData|GetDimension _aspcgspatial_data) 2)
      (bind _output_get (LocalVariables|SetInputPCGSpatialData _aspcgspatial_data))
      (LocalVariables|SetTargetActor (Variables|/Script/PCG.PCGSpatialData|GetTargetActor _output_get))
      (bind _returnvalue_3 (PCG|Temporary|GetOriginalComponent InContext))
      (LocalVariables|SetPCGVolumeActor (Components|GetOwner _returnvalue_3))
      (LocalVariables|SetExtents (* (Math|Vector|SelectVector _boxextent _boxextent_2 (Utilities|IsValid (Actor|GetComponentbyClass (LocalVariables|GetPCGVolumeActor) "/Script/Engine.BrushComponent") :!class /Script/Engine.KismetSystemLibrary)) (Math|Vector|MakeVector 1.1 1.1 1.1)))
      (Utilities|FlowControl|Sequence
        (:then_0
          ;[1 1 1] Render terrain normals to RT
          (bind _texturetarget (Default|RTRenderWorldOrtho "SCS_Normal" (LocalVariables|GetPCGVolumeActor) (LocalVariables|GetExtents) true (Utilities|Array|MakeArray) (Variables|Default|GetTexels%20Per%20Meter))))
        (:then_1
          ;[1 1 1] Build RenderTargetData, transfer Terrain Normals to Color, Project on Terrain
          (bind _returnvalue_4 (Game|ConstructObjectfromClass "/Script/PCG.PCGRenderTargetData"))
          (Variables|/Script/PCG.PCGSpatialData|SetTargetActor (LocalVariables|GetPCGVolumeActor) _returnvalue_4)
          (Variables|/Script/PCG.PCGTexture2DSingleBaseData|SetbUseDensitySourceChannel false _returnvalue_4)
          (RenderTarget|Initialize _returnvalue_4 _texturetarget (Math|Transform|MakeTransform (.location (Transformation|GetActorTransform (LocalVariables|GetPCGVolumeActor))) (.rotation (Transformation|GetActorTransform (LocalVariables|GetPCGVolumeActor))) (Math|Vector|MakeVector (Math|Float|Max(Float) (.x (LocalVariables|GetExtents)) (.y (LocalVariables|GetExtents))) (Math|Float|Max(Float) (.x (LocalVariables|GetExtents)) (.y (LocalVariables|GetExtents))) (.z (LocalVariables|GetExtents)))) false false))
        (:then_2
          ;[1 1 1] Project points from input1 (Surface sampler ?) to projected RenderTargetData, Colors are interpolated
          (bind _aspcgspatial_data_2 (Utilities|Casting|CastToPCGSpatialData (PCG|TaggedData|BreakTaggedData (LocalVariables|GetPointData)))))
        (:then_3
          (bind _outdata (PCG|FlowControl|PointLoop InContext (SpatialData|ToPointDatawithContext (SpatialData|ProjectOn _aspcgspatial_data_2 (SpatialData|ProjectOn _returnvalue_4 (LocalVariables|GetInputPCGSpatialData) "(bProjectPositions=True,bProjectRotations=True,bProjectScales=False,ColorBlendMode=SourceValue,AttributeList=\"\",AttributeMode=ExcludeAttributes,AttributeMergeOperation=TargetValue,TagMergeOperation=Source)") "(bProjectPositions=True,bProjectRotations=True,bProjectScales=False,ColorBlendMode=SourceValue,AttributeList=\"\",AttributeMode=ExcludeAttributes,AttributeMergeOperation=TargetValue,TagMergeOperation=Source)") InContext) 0))
          (Variables|/Script/PCG.PCGSpatialData|SetTargetActor (LocalVariables|GetTargetActor) _outdata)
          (LocalVariables|SetOutputTaggedData (Utilities|Array|MakeArray (PCG|TaggedData|MakeTaggedData _outdata)))
          (bind _returnvalue_5 (Data|GetAllSettings (LocalVariables|GetInputCollection)))
          (Utilities|Array|AppendArray (LocalVariables|GetOutputTaggedData) _returnvalue_5)))
      (else
        _))
    (return (Utilities|Struct|MakePCGDataCollection (LocalVariables|GetOutputTaggedData)))
    (else
      (return 0))))

(fn PointLoopBody
  (params (InContext :type PCGContext%20Structure%20%28by%20ref%29) (InData :type PCGPoint%20Data%20Object%20Reference) (InPoint :type PCGPoint%20Structure%20%28by%20ref%29) (OutMetadata :type PCGMetadata%20Object%20Reference) (Iteration :type Integer64))
  (bind (_transform _boundsmin _boundsmax _color) (Utilities|Struct|BreakPCGPoint InPoint))
  (bind _structout (Utilities|Struct|SetmembersinPCGPoint InPoint (Math|Transform|MakeTransform (.location _transform) (Math|Rotator|MakeRotfromZ (Math|Vector|Normalize (Math|Conversions|ToVector(Vector4) _color))) (.scale _transform))))
  (return _structout true))
