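; Deprecated PCG Blueprint element that projects one spatial data set onto another, by its name points onto a landscape.
; Execute picks source and target by tag filters, while ExecuteWithContext reads the target from the first custom input pin and the source from the second.
; Both use only the first match for each and return the projection as a single-entry collection, transferring position and rotation but not scale.
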(fn Execute
  (params (Input :type PCGData%20Collection%20Structure%20%28by%20ref%29))
  (bind _returnvalue (Data|GetAllSpatialInputs Input))
  (bind _aspcgspatial_data (Utilities|Casting|CastToPCGSpatialData (PCG|TaggedData|BreakTaggedData (Utilities|Array|Get(aref) (Class|PCGBPFunctionLibrary|FilterDatabyTags _returnvalue (Variables|Default|GetDataToProjectOntoTags))))))
  (bind _aspcgspatial_data_2 (Utilities|Casting|CastToPCGSpatialData (PCG|TaggedData|BreakTaggedData (Utilities|Array|Get(aref) (Class|PCGBPFunctionLibrary|FilterDatabyTags _returnvalue (Variables|Default|GetDataToProjectTags))))))
  (return (Utilities|Struct|MakePCGDataCollection (Utilities|Array|MakeArray (PCG|TaggedData|MakeTaggedData (SpatialData|ProjectOn _aspcgspatial_data_2 _aspcgspatial_data "(bProjectPositions=True,bProjectRotations=True,bProjectScales=False,ColorBlendMode=SourceValue,AttributeList=\"\",AttributeMode=ExcludeAttributes,AttributeMergeOperation=TargetValue,TagMergeOperation=Source)"))))))

(fn ExecuteWithContext
  (params (InContext :type PCGContext%20Structure%20%28by%20ref%29) (Input :type PCGData%20Collection%20Structure%20%28by%20ref%29))
  (bind _result (Utilities|Set|ToArray (PCG|Input&Output|CustomInputLabels self)))
  (bind _returnvalue (Data|GetInputsbyPinLabel Input (Utilities|Array|Get(aref) _result)))
  (bind _returnvalue_2 (Data|GetInputsbyPinLabel Input (Utilities|Array|Get(aref) _result 1)))
  (bind _aspcgspatial_data (Utilities|Casting|CastToPCGSpatialData (PCG|TaggedData|BreakTaggedData (Utilities|Array|Get(aref) _returnvalue))))
  (bind _aspcgspatial_data_2 (Utilities|Casting|CastToPCGSpatialData (PCG|TaggedData|BreakTaggedData (Utilities|Array|Get(aref) _returnvalue_2))))
  (return (Utilities|Struct|MakePCGDataCollection (Utilities|Array|MakeArray (PCG|TaggedData|MakeTaggedData (SpatialData|ProjectOn _aspcgspatial_data_2 _aspcgspatial_data "(bProjectPositions=True,bProjectRotations=True,bProjectScales=False,ColorBlendMode=SourceValue,AttributeList=\"\",AttributeMode=ExcludeAttributes,AttributeMergeOperation=TargetValue,TagMergeOperation=Source)"))))))
