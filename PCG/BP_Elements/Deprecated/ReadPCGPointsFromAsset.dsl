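; Deprecated PCG Blueprint element that outputs point data read from a PCGPointList asset.
; It loads the points from the asset variable and copies the target actor from a spatial input.
; It builds a new PCGPointData from both, then returns it as a collection with the input's settings appended.
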(fn Execute
  (params (Input :type PCGData%20Collection%20Structure%20%28by%20ref%29))
  (bind _output_get (LocalVariables|SetInputCollection Input))
  (bind _returnvalue (Data|GetAllSpatialInputs _output_get))
  (Utilities|FlowControl|Sequence
    (:then_0
      (LocalVariables|SetOutPoints (Variables|/PCG/BP_Elements/Resources/PCGPointList.PCGPointList_C|GetPCGPoints (Variables|Default|GetPCGPointsAsset))))
    (:then_1
      (bind _aspcgspatial_data (Utilities|Casting|CastToPCGSpatialData (PCG|TaggedData|BreakTaggedData (Utilities|Array|Get(aref) _returnvalue))))
      (LocalVariables|SetTargetActor (Variables|/Script/PCG.PCGSpatialData|GetTargetActor _aspcgspatial_data)))
    (:then_2
      (bind _returnvalue_2 (Game|ConstructObjectfromClass "/Script/PCG.PCGPointData"))
      (SpatialData|SetPoints _returnvalue_2 (LocalVariables|GetOutPoints))
      (Variables|/Script/PCG.PCGSpatialData|SetTargetActor (LocalVariables|GetTargetActor) _returnvalue_2)
      (LocalVariables|SetOutputTaggedData (Utilities|Array|MakeArray (PCG|TaggedData|MakeTaggedData _returnvalue_2)))
      (bind _returnvalue_3 (Data|GetAllSettings (LocalVariables|GetInputCollection)))
      (Utilities|Array|AppendArray (LocalVariables|GetOutputTaggedData) _returnvalue_3)
      (return (Utilities|Struct|MakePCGDataCollection (LocalVariables|GetOutputTaggedData))))))
