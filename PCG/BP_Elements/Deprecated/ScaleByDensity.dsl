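; Custom PCG Blueprint element, filed under Deprecated, that scales each point uniformly according to its density.
; ExecuteWithContext converts the first spatial input to point data, runs a point loop, copies over the target actor and returns the result as a new collection.
; PointLoopBody raises density to Density Exp, remaps it with clamping to Scale Min–Scale Max, and multiplies existing scale by that factor, keeping location, rotation and every point.
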(fn ExecuteWithContext
  (params (InContext :type PCGContext%20Structure%20%28by%20ref%29) (Input :type PCGData%20Collection%20Structure%20%28by%20ref%29))
  (bind _returnvalue (Data|GetAllSpatialInputs Input))
  (bind _aspcgspatial_data (Utilities|Casting|CastToPCGSpatialData (PCG|TaggedData|BreakTaggedData (Utilities|Array|Get(aref) _returnvalue))))
  (bind _outdata (PCG|FlowControl|PointLoop InContext (SpatialData|ToPointDatawithContext _aspcgspatial_data InContext) 0))
  (Variables|/Script/PCG.PCGSpatialData|SetTargetActor (Variables|/Script/PCG.PCGSpatialData|GetTargetActor _aspcgspatial_data) _outdata)
  (return (Utilities|Struct|MakePCGDataCollection (Utilities|Array|MakeArray (PCG|TaggedData|MakeTaggedData _outdata)))))

(fn PointLoopBody
  (params (InContext :type PCGContext%20Structure%20%28by%20ref%29) (InData :type PCGPoint%20Data%20Object%20Reference) (InPoint :type PCGPoint%20Structure%20%28by%20ref%29) (OutMetadata :type PCGMetadata%20Object%20Reference) (Iteration :type Integer64))
  (bind (_transform _density) (Utilities|Struct|BreakPCGPoint InPoint))
  (bind _structout (Utilities|Struct|SetmembersinPCGPoint InPoint (Math|Transform|MakeTransform (.location _transform) (.rotation _transform) (* (.scale _transform) (Math|Vector|MakeVector (Math|Float|MapRangeClamped (Math|Float|Power _density (Variables|Default|GetDensity%20Exp)) 0.0 1.0 (Variables|Default|GetScale%20Min) (Variables|Default|GetScale%20Max)) (Math|Float|MapRangeClamped (Math|Float|Power _density (Variables|Default|GetDensity%20Exp)) 0.0 1.0 (Variables|Default|GetScale%20Min) (Variables|Default|GetScale%20Max)) (Math|Float|MapRangeClamped (Math|Float|Power _density (Variables|Default|GetDensity%20Exp)) 0.0 1.0 (Variables|Default|GetScale%20Min) (Variables|Default|GetScale%20Max)))))))
  (return _structout true))
