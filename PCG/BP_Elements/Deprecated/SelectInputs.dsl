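; Deprecated PCG Blueprint element that blends or selects between inputs A and B using a selector input.
; With a spatial selector, it converts all three to points and lerps A's and B's last attributes point by index, weighted by the selector's last attribute.
; The output attribute follows A's type (Float, Double, Vector, Vector4).
; With an attribute-set selector, its last attribute is an integer index passing A or B through whole.
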(fn ExecuteWithContext
  (params (InContext :type PCGContext%20Structure%20%28by%20ref%29) (Input :type PCGData%20Collection%20Structure%20%28by%20ref%29))
  (bind _ptdataselector (Variables|Hidden|GetPtDataSelector))
  (bind _returnvalue (Metadata|ConstMetadata _ptdataselector))
  (bind (_attributenames _attributetypes) (PCG|Metadata|GetAttributes _returnvalue))
  (bind _returnvalue_2 (Utilities|Array|LastIndex _attributenames))
  (bind _ptdataa (Variables|Hidden|GetPtDataA))
  (bind _returnvalue_3 (Metadata|ConstMetadata _ptdataa))
  (bind (_attributenames_2 _attributetypes_2) (PCG|Metadata|GetAttributes _returnvalue_3))
  (bind _returnvalue_4 (Utilities|Array|LastIndex _attributenames_2))
  (bind _output (Utilities|Array|Get(aref) _attributetypes_2 _returnvalue_4))
  (bind _context (LocalVariables|GetContext))
  (bind _outputattributename (Variables|Default|GetOutputAttributeName))
  (LocalVariables|SetContext InContext)
  (LocalVariables|SetInputsCollection Input)
  (bind _result (Utilities|Set|ToArray (PCG|Input&Output|CustomInputLabels self)))
  (bind _returnvalue_5 (Data|GetInputsbyPinLabel (LocalVariables|GetInputsCollection) (Utilities|Array|Get(aref) _result 1)))
  (LocalVariables|SetPinA%20Inputs _returnvalue_5)
  (bind _returnvalue_6 (Data|GetInputsbyPinLabel (LocalVariables|GetInputsCollection) (Utilities|Array|Get(aref) _result 2)))
  (LocalVariables|SetPinB%20Inputs _returnvalue_6)
  (bind _returnvalue_7 (Data|GetInputsbyPinLabel (LocalVariables|GetInputsCollection) (Utilities|Array|Get(aref) _result)))
  (bind _aspcgspatial_data (Utilities|Casting|CastToPCGSpatialData (PCG|TaggedData|BreakTaggedData (Utilities|Array|Get(aref) _returnvalue_7)))
    (:then
      (Utilities|FlowControl|Sequence
        (:then_0
          (bind _aspcgspatial_data_2 (Utilities|Casting|CastToPCGSpatialData (PCG|TaggedData|BreakTaggedData (Utilities|Array|Get(aref) (LocalVariables|GetPinA%20Inputs)))))
          (bind _aspcgspatial_data_3 (Utilities|Casting|CastToPCGSpatialData (PCG|TaggedData|BreakTaggedData (Utilities|Array|Get(aref) (LocalVariables|GetPinB%20Inputs)))))
          (Variables|Hidden|SetPtDataA (SpatialData|ToPointDatawithContext _aspcgspatial_data_2 _context))
          (Variables|Hidden|SetPtDataB (SpatialData|ToPointDatawithContext _aspcgspatial_data_3 _context))
          (Variables|Hidden|SetPtDataSelector (SpatialData|ToPointDatawithContext _aspcgspatial_data _context))
          (Variables|Hidden|SetInputAttributeNameA (Utilities|Array|Get(acopy) _attributenames_2 _returnvalue_4))
          (Variables|Hidden|SetInputAttributeNameB (Utilities|Array|Get(acopy) (PCG|Metadata|GetAttributes (Metadata|ConstMetadata (Variables|Hidden|GetPtDataB))) (Utilities|Array|LastIndex (PCG|Metadata|GetAttributes (Metadata|ConstMetadata (Variables|Hidden|GetPtDataB))))))
          (Variables|Hidden|SetSelectorAttributeName (Utilities|Array|Get(acopy) _attributenames _returnvalue_2))
          (Variables|Hidden|SetAttributeType _output)
          (Variables|Hidden|SetSelectorType (Utilities|Array|Get(acopy) _attributetypes _returnvalue_2)))
        (:then_1
          (bind _returnvalue_8 (Game|ConstructObjectfromClass "/Script/PCG.PCGPointData"))
          (Variables|/Script/PCG.PCGSpatialData|SetTargetActor (Variables|/Script/PCG.PCGSpatialData|GetTargetActor (SpatialData|ToPointDatawithContext _aspcgspatial_data_2 _context)) _returnvalue_8)
          (bind _returnvalue_9 (Metadata|MutableMetadata _returnvalue_8))
          (switch Utilities|FlowControl|Switch|SwitchonEPCGMetadataTypes _output
            (:Float
              (PCG|Metadata|CreateFloatAttribute _returnvalue_9 _outputattributename 0.0 true true))
            (:Double
              (PCG|Metadata|CreateDoubleAttribute _returnvalue_9 _outputattributename 0.0 true true))
            (:Vector
              (PCG|Metadata|CreateVectorAttribute _returnvalue_9 _outputattributename (0.0 0.0 0.0) true true))
            (:Vector4
              (PCG|Metadata|CreateVector4Attribute _returnvalue_9 _outputattributename (0.0 0.0 0.0 0.0) true true))))
        (:then_2
          (bind _outdata (PCG|FlowControl|IterationLoop (LocalVariables|GetContext) (Math|Conversions|ToInteger64(Integer) (Utilities|Array|Length (SpatialData|GetPoints (Variables|Hidden|GetPtDataA)))) 0 0 _returnvalue_8))
          (return (Utilities|Struct|MakePCGDataCollection (Utilities|Array|MakeArray (PCG|TaggedData|MakeTaggedData _outdata)))))))
    (:CastFailed
      (bind _aspcgattribute_set (Utilities|Casting|CastToPCGParamData (PCG|TaggedData|BreakTaggedData (Utilities|Array|Get(aref) _returnvalue_7))))
      (bind _returnvalue_10 (PCG|Metadata|GetInteger64AttributebyMetadataKey 0 (Metadata|ConstMetadata _aspcgattribute_set) (Utilities|Array|Get(acopy) (PCG|Metadata|GetAttributes (Metadata|ConstMetadata _aspcgattribute_set)) (Utilities|Array|LastIndex (PCG|Metadata|GetAttributes (Metadata|ConstMetadata _aspcgattribute_set))))))
      (return (Utilities|Struct|MakePCGDataCollection (Utilities|Select :Index (Math|Conversions|ToInteger(Integer64) _returnvalue_10) :Option%200 (LocalVariables|GetPinA%20Inputs) :Option%201 (LocalVariables|GetPinB%20Inputs)))))))

(fn IterationLoopBody
  (params (InContext :type PCGContext%20Structure%20%28by%20ref%29) (Iteration :type Integer64) (InA :type PCGSpatial%20Data%20Object%20Reference) (InB :type PCGSpatial%20Data%20Object%20Reference) (OutMetadata :type PCGMetadata%20Object%20Reference))
  (bind _returnvalue (Math|Conversions|ToInteger(Integer64) Iteration))
  (bind _ptdataa (Variables|Hidden|GetPtDataA))
  (bind _returnvalue_2 (Metadata|ConstMetadata _ptdataa))
  (bind _ptdatab (Variables|Hidden|GetPtDataB))
  (bind _returnvalue_3 (Metadata|ConstMetadata _ptdatab))
  (bind _ptdataselector (Variables|Hidden|GetPtDataSelector))
  (bind _returnvalue_4 (Metadata|ConstMetadata _ptdataselector))
  (bind _outputattributename (Variables|Default|GetOutputAttributeName))
  (bind _oumetadata (LocalVariables|GetouMetadata))
  (bind _inputattributenamea (Variables|Hidden|GetInputAttributeNameA))
  (bind _inputattributenameb (Variables|Hidden|GetInputAttributeNameB))
  (bind _ptsel (LocalVariables|GetptSel))
  (bind _outpt (LocalVariables|GetoutPt))
  (bind _selectorattributename (Variables|Hidden|GetSelectorAttributeName))
  (bind _selectorfloat (LocalVariables|GetSelectorFloat))
  (LocalVariables|SetouMetadata OutMetadata)
  (LocalVariables|SetptA (SpatialData|GetPoint (Variables|Hidden|GetPtDataA) _returnvalue))
  (LocalVariables|SetptB (SpatialData|GetPoint (Variables|Hidden|GetPtDataB) _returnvalue))
  (LocalVariables|SetptSel (SpatialData|GetPoint (Variables|Hidden|GetPtDataSelector) _returnvalue))
  (bind _outpoint (PCG|CopyPoint (LocalVariables|GetptA) false))
  (LocalVariables|SetoutPt _outpoint)
  (if (Utilities|Enum|Equal(Enum) (Variables|Hidden|GetSelectorType) "Double")
    (bind _returnvalue_5 (PCG|Metadata|GetDoubleAttribute _ptsel _returnvalue_4 _selectorattributename))
    (LocalVariables|SetSelectorFloat _returnvalue_5)
    (else
      (bind _returnvalue_6 (PCG|Metadata|GetFloatAttribute _ptsel _returnvalue_4 _selectorattributename))
      (LocalVariables|SetSelectorFloat _returnvalue_6)))
  (switch Utilities|FlowControl|Switch|SwitchonEPCGMetadataTypes (Variables|Hidden|GetAttributeType)
    (:Float
      (bind _returnvalue_7 (PCG|Metadata|GetFloatAttribute (LocalVariables|GetptA) _returnvalue_2 _inputattributenamea))
      (bind _returnvalue_8 (PCG|Metadata|GetFloatAttribute (LocalVariables|GetptB) _returnvalue_3 _inputattributenameb))
      (PCG|Metadata|SetFloatAttribute _outpt _oumetadata _outputattributename (Math|Float|Lerp _returnvalue_7 _returnvalue_8 _selectorfloat)))
    (:Double
      (bind _returnvalue_9 (PCG|Metadata|GetDoubleAttribute (LocalVariables|GetptA) _returnvalue_2 _inputattributenamea))
      (bind _returnvalue_10 (PCG|Metadata|GetDoubleAttribute (LocalVariables|GetptB) _returnvalue_3 _inputattributenameb))
      (PCG|Metadata|SetDoubleAttribute _outpt _oumetadata _outputattributename (Math|Float|Lerp _returnvalue_9 _returnvalue_10 _selectorfloat)))
    (:Vector
      (bind _returnvalue_11 (PCG|Metadata|GetVectorAttribute (LocalVariables|GetptA) _returnvalue_2 _inputattributenamea))
      (bind _returnvalue_12 (PCG|Metadata|GetVectorAttribute (LocalVariables|GetptB) _returnvalue_3 _inputattributenameb))
      (PCG|Metadata|SetVectorAttribute _outpt _oumetadata _outputattributename (Math|Vector|Lerp(Vector) _returnvalue_11 _returnvalue_12 _selectorfloat)))
    (:Vector4
      (bind _returnvalue_13 (PCG|Metadata|GetVector4Attribute (LocalVariables|GetptA) _returnvalue_2 _inputattributenamea))
      (bind _returnvalue_14 (PCG|Metadata|GetVector4Attribute (LocalVariables|GetptB) _returnvalue_3 _inputattributenameb))
      (PCG|Metadata|SetVector4Attribute _outpt _oumetadata _outputattributename (Math|Vector4|MakeVector4 (Math|Float|Lerp (Math|Vector4|BreakVector4 _returnvalue_13) (Math|Vector4|BreakVector4 _returnvalue_14) _selectorfloat) (Math|Float|Lerp (.Y (Math|Vector4|BreakVector4 _returnvalue_13)) (.Y (Math|Vector4|BreakVector4 _returnvalue_14)) _selectorfloat) (Math|Float|Lerp (.Z (Math|Vector4|BreakVector4 _returnvalue_13)) (.Z (Math|Vector4|BreakVector4 _returnvalue_14)) _selectorfloat) (Math|Float|Lerp (.W (Math|Vector4|BreakVector4 _returnvalue_13)) (.W (Math|Vector4|BreakVector4 _returnvalue_14)) _selectorfloat)))))
  (return (LocalVariables|GetoutPt) true))
