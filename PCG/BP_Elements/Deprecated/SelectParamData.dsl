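; Deprecated PCG Blueprint element that passes through one of two attribute set inputs, chosen by an index value.
; It reads the last attribute of the first param data on the first custom pin as an Int64 at metadata key 0.
; That value selects whether the param data from the second or third pin becomes the output collection.
; A color override tints the node orange.
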(fn ExecuteWithContext
  (params (InContext :type PCGContext%20Structure%20%28by%20ref%29) (Input :type PCGData%20Collection%20Structure%20%28by%20ref%29))
  (bind _result (Utilities|Set|ToArray (PCG|Input&Output|CustomInputLabels self)))
  (bind _returnvalue (Data|GetParamsbyPinLabel Input (Utilities|Array|Get(aref) _result)))
  (bind _aspcgattribute_set (Utilities|Casting|CastToPCGParamData (PCG|TaggedData|BreakTaggedData (Utilities|Array|Get(aref) _returnvalue))))
  (bind _returnvalue_2 (PCG|Metadata|GetInteger64AttributebyMetadataKey 0 (Metadata|ConstMetadata _aspcgattribute_set) (Utilities|Array|Get(acopy) (PCG|Metadata|GetAttributes (Metadata|ConstMetadata _aspcgattribute_set)) (Utilities|Array|LastIndex (PCG|Metadata|GetAttributes (Metadata|ConstMetadata _aspcgattribute_set))))))
  (bind _returnvalue_3 (Data|GetParamsbyPinLabel Input (Utilities|Array|Get(aref) _result 1)))
  (bind _returnvalue_4 (Data|GetParamsbyPinLabel Input (Utilities|Array|Get(aref) _result 2)))
  (return (Utilities|Struct|MakePCGDataCollection (Utilities|Select :Index (Math|Conversions|ToInteger(Integer64) _returnvalue_2) :Option%200 _returnvalue_3 :Option%201 _returnvalue_4))))

(fn NodeColorOverride
  (return "(R=1.000000,G=0.597202,B=0.000000,A=1.000000)"))
