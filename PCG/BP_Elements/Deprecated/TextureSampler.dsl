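; Deprecated PCG Blueprint element named TextureSampler, kept as a stub.
; When executed, it prints a "TextureSampler - Deprecated" message and returns its input data collection unchanged.
; It overrides the node title to show "TextureSampler-DEPRECATED" in the PCG graph.
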(fn ExecuteWithContext
  (params (InContext :type PCGContext%20Structure%20%28by%20ref%29) (Input :type PCGData%20Collection%20Structure%20%28by%20ref%29))
  ;[1 1 1] DEPRECATED
  (Development|PrintString "TextureSampler - Deprecated")
  (return Input))

(fn NodeTitleOverride
  (return "TextureSampler-DEPRECATED"))
