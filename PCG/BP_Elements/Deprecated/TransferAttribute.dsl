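; Deprecated PCG Blueprint element that copies a named attribute from source points onto target points under a target attribute name.
; ExecuteWithContext converts both inputs to point data, looks up the source attribute's type and creates a matching attribute on new output data.
; An iteration loop rebuilds each target point, reading the same-index source point when counts match, otherwise the first.
; Only Float, Double, Vector and Vector4 attributes are handled.
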(fn ExecuteWithContext
  (params (InContext :type PCGContext%20Structure%20%28by%20ref%29) (Input :type PCGData%20Collection%20Structure%20%28by%20ref%29))
  (bind _targetpointdata (Variables|Hidden|GetTargetPointData))
  (bind _returnvalue (SpatialData|GetPoints _targetpointdata))
  (bind _returnvalue_2 (Utilities|Array|Length _returnvalue))
  (bind _source_spatial_data (Variables|Hidden|GetSource%20Spatial%20Data))
  (bind _targetattributename (Variables|Default|GetTargetAttributeName))
  (bind _targmeta (Variables|Hidden|GetTargMeta))
  (LocalVariables|SetContext InContext)
  (Utilities|FlowControl|Sequence
    (:then_0
      (bind _result (Utilities|Set|ToArray (PCG|Input&Output|CustomInputLabels self)))
      (bind _returnvalue_3 (Data|GetInputsbyPinLabel Input (Utilities|Array|Get(aref) _result)))
      (bind _returnvalue_4 (Data|GetInputsbyPinLabel Input (Utilities|Array|Get(aref) _result 1)))
      (bind _aspcgspatial_data (Utilities|Casting|CastToPCGSpatialData (PCG|TaggedData|BreakTaggedData (Utilities|Array|Get(aref) _returnvalue_3))))
      (bind _output_get (Variables|Hidden|SetSource%20Spatial%20Data _aspcgspatial_data))
      (bind _aspcgspatial_data_2 (Utilities|Casting|CastToPCGSpatialData (PCG|TaggedData|BreakTaggedData (Utilities|Array|Get(aref) _returnvalue_4))))
      (bind _output_get_2 (Variables|Hidden|SetTarget%20Spatial%20Data _aspcgspatial_data_2))
      (Variables|Hidden|SetSourcePointData (SpatialData|ToPointDatawithContext _output_get (LocalVariables|GetContext)))
      (Variables|Hidden|SetTargetPointData (SpatialData|ToPointDatawithContext _output_get_2 (LocalVariables|GetContext))))
    (:then_2
      (bind _returnvalue_5 (Game|ConstructObjectfromClass "/Script/PCG.PCGPointData"))
      (bind _output_get_3 (LocalVariables|SetOutputPCGPointData _returnvalue_5))
      (SpatialData|InitializefromData _output_get_3 (Variables|Hidden|GetTarget%20Spatial%20Data) 0 true true)
      (bind _returnvalue_6 (Metadata|MutableMetadata _output_get_3))
      (Variables|Hidden|SetTargMeta _returnvalue_6)
      (bind _returnvalue_7 (Metadata|MutableMetadata _source_spatial_data))
      (Variables|Hidden|SetSrcMeta (Metadata|ConstMetadata _source_spatial_data))
      (bind _output_get_4 (Variables|Hidden|SetAttrType (Utilities|Array|Get(aref) (.AttributeTypes (PCG|Metadata|GetAttributes _returnvalue_7)) (Utilities|Array|FindItem (PCG|Metadata|GetAttributes _returnvalue_7) (Variables|Default|GetSourceAttributeName)))))
      (switch Utilities|FlowControl|Switch|SwitchonEPCGMetadataTypes _output_get_4
        (:Float
          (PCG|Metadata|CreateFloatAttribute _targmeta _targetattributename 0.0 true true))
        (:Double
          (PCG|Metadata|CreateDoubleAttribute _targmeta _targetattributename 0.0 true true))
        (:Vector
          (PCG|Metadata|CreateVectorAttribute _targmeta _targetattributename (Math|Vector|MakeVector 0.0) true true))
        (:Vector4
          (PCG|Metadata|CreateVector4Attribute _targmeta _targetattributename (Math|Vector4|MakeVector4 0.0) true true))))
    (:then_3
      (Variables|Hidden|SetSourceAndTargetsAreSameSize (== (Utilities|Array|Length (SpatialData|GetPoints (Variables|Hidden|GetSourcePointData))) _returnvalue_2))
      (bind _outdata (PCG|FlowControl|IterationLoop (LocalVariables|GetContext) (Math|Conversions|ToInteger64(Integer) _returnvalue_2) 0 0 (LocalVariables|GetOutputPCGPointData)))
      (return (Utilities|Struct|MakePCGDataCollection (Utilities|Array|MakeArray (PCG|TaggedData|MakeTaggedData _outdata)))))))

(fn IterationLoopBody
  (params (InContext :type PCGContext%20Structure%20%28by%20ref%29) (Iteration :type Integer64) (InA :type PCGSpatial%20Data%20Object%20Reference) (InB :type PCGSpatial%20Data%20Object%20Reference) (OutMetadata :type PCGMetadata%20Object%20Reference))
  (bind _returnvalue (Math|Conversions|ToInteger(Integer64) Iteration))
  (bind _targetpointdata (Variables|Hidden|GetTargetPointData))
  (bind _returnvalue_2 (SpatialData|GetPoint _targetpointdata _returnvalue))
  (bind (_transform _density _boundsmin _boundsmax _color _steepness _seed) (Utilities|Struct|BreakPCGPoint _returnvalue_2))
  (LocalVariables|SetSourcePoint (SpatialData|GetPoint (Variables|Hidden|GetSourcePointData) (select (Variables|Hidden|GetSourceAndTargetsAreSameSize) (Math|Conversions|ToInteger(Integer64) Iteration) 0)))
  (bind _output_get (LocalVariables|SetOutputPoint (Utilities|Struct|MakePCGPoint _transform _density _boundsmin _boundsmax _color _steepness _seed)))
  (PCG|Metadata|InitializeMetadata _output_get (Variables|Hidden|GetTargMeta) _returnvalue_2 (Metadata|ConstMetadata (Variables|Hidden|GetTargetPointData)))
  (switch Utilities|FlowControl|Switch|SwitchonEPCGMetadataTypes (Variables|Hidden|GetAttrType)
    (:Float
      (bind _returnvalue_3 (PCG|Metadata|GetFloatAttribute (LocalVariables|GetSourcePoint) (Variables|Hidden|GetSrcMeta) (Variables|Default|GetSourceAttributeName)))
      (PCG|Metadata|SetFloatAttribute (LocalVariables|GetOutputPoint) (Variables|Hidden|GetTargMeta) (Variables|Default|GetTargetAttributeName) _returnvalue_3))
    (:Double
      (bind _returnvalue_4 (PCG|Metadata|GetDoubleAttribute (LocalVariables|GetSourcePoint) (Variables|Hidden|GetSrcMeta) (Variables|Default|GetSourceAttributeName)))
      (PCG|Metadata|SetDoubleAttribute (LocalVariables|GetOutputPoint) (Variables|Hidden|GetTargMeta) (Variables|Default|GetTargetAttributeName) _returnvalue_4))
    (:Vector
      (bind _returnvalue_5 (PCG|Metadata|GetVectorAttribute (LocalVariables|GetSourcePoint) (Variables|Hidden|GetSrcMeta) (Variables|Default|GetSourceAttributeName)))
      (PCG|Metadata|SetVectorAttribute (LocalVariables|GetOutputPoint) (Variables|Hidden|GetTargMeta) (Variables|Default|GetTargetAttributeName) _returnvalue_5))
    (:Vector4
      (bind _returnvalue_6 (PCG|Metadata|GetVector4Attribute (LocalVariables|GetSourcePoint) (Variables|Hidden|GetSrcMeta) (Variables|Default|GetSourceAttributeName)))
      (PCG|Metadata|SetVector4Attribute (LocalVariables|GetOutputPoint) (Variables|Hidden|GetTargMeta) (Variables|Default|GetTargetAttributeName) _returnvalue_6)))
  (return (LocalVariables|GetOutputPoint) true))

(fn NodeColorOverride
  (return "(R=0.949020,G=0.949020,B=0.949020,A=1.000000)"))

(fn NodeTitleOverride
  (return (Utilities|String|StringToName (Utilities|String|Append "TransferAttr " (Utilities|String|ToString(Name) (Variables|Default|GetSourceAttributeName)) " to " (Utilities|String|ToString(Name) (Variables|Default|GetTargetAttributeName))))))
