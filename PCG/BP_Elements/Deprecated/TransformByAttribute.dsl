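; PCG Blueprint element, in a Deprecated folder, that transforms points using per-point vector metadata attributes.
; It converts the first spatial input to point data, runs a point loop and copies the target actor to the output.
; Toggles enable translation (world or point-local), rotation (absolute or combined) and scale (absolute or multiplied).
; The loop body selects between these results and original values to rebuild each point's transform.
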(fn ExecuteWithContext
  (params (InContext :type PCGContext%20Structure%20%28by%20ref%29) (Input :type PCGData%20Collection%20Structure%20%28by%20ref%29))
  (bind _returnvalue (Data|GetAllSpatialInputs Input))
  (bind _aspcgspatial_data (Utilities|Casting|CastToPCGSpatialData (PCG|TaggedData|BreakTaggedData (Utilities|Array|Get(aref) _returnvalue))))
  (bind _outdata (PCG|FlowControl|PointLoop InContext (SpatialData|ToPointDatawithContext _aspcgspatial_data InContext) 0))
  (Variables|/Script/PCG.PCGSpatialData|SetTargetActor (Variables|/Script/PCG.PCGSpatialData|GetTargetActor _aspcgspatial_data) _outdata)
  (return (Utilities|Struct|MakePCGDataCollection (Utilities|Array|MakeArray (PCG|TaggedData|MakeTaggedData _outdata)))))

(fn PointLoopBody
  (params (InContext :type PCGContext%20Structure%20%28by%20ref%29) (InData :type PCGPoint%20Data%20Object%20Reference) (InPoint :type PCGPoint%20Structure%20%28by%20ref%29) (OutMetadata :type PCGMetadata%20Object%20Reference) (Iteration :type Integer64))
  (bind _transform (Utilities|Struct|BreakPCGPoint InPoint))
  (LocalVariables|SetinPt InPoint)
  (LocalVariables|SetInMetadata (Metadata|ConstMetadata InData))
  (Utilities|FlowControl|Sequence
    (:then_0
      (LocalVariables|SetInLocation (.location _transform))
      (LocalVariables|SetInRotation (.rotation _transform))
      (LocalVariables|SetInScale (.scale _transform)))
    (:then_1
      (if (Variables|Default|GetTransformLocation)
        (bind _returnvalue (PCG|Metadata|GetVectorAttribute (LocalVariables|GetinPt) (LocalVariables|GetInMetadata) (Variables|Default|GetTranslationAttribute)))
        (LocalVariables|SettranslationAttrib _returnvalue)))
    (:then_2
      (if (Variables|Default|GetTransformRotation)
        (bind _returnvalue_2 (PCG|Metadata|GetVectorAttribute (LocalVariables|GetinPt) (LocalVariables|GetInMetadata) (Variables|Default|GetRotationAttribute)))
        (LocalVariables|SetrotationAttrib _returnvalue_2)))
    (:then_3
      (if (Variables|Default|GetTransformScale)
        (bind _returnvalue_3 (PCG|Metadata|GetVectorAttribute (LocalVariables|GetinPt) (LocalVariables|GetInMetadata) (Variables|Default|GetScaleAttribute)))
        (LocalVariables|SetscaleAttrib _returnvalue_3)))
    (:then_4
      (bind _structout (Utilities|Struct|SetmembersinPCGPoint InPoint (Math|Transform|MakeTransform (Math|Vector|SelectVector (Math|Vector|SelectVector (+ (LocalVariables|GetInLocation) (LocalVariables|GettranslationAttrib)) (+ (LocalVariables|GetInLocation) (Math|Matrix|TransformPosition(Matrix) (Math|Transform|ToMatrix(Transform) (Math|Conversions|ToTransform(Rotator) (LocalVariables|GetInRotation))) (LocalVariables|GettranslationAttrib))) (Variables|Default|GetWorldSpaceTranslation)) (LocalVariables|GetInLocation) (Variables|Default|GetTransformLocation)) (Math|Rotator|SelectRotator (Math|Rotator|SelectRotator (Math|Rotator|MakeRotator (.x (LocalVariables|GetrotationAttrib)) (.y (LocalVariables|GetrotationAttrib)) (.z (LocalVariables|GetrotationAttrib))) (Math|Rotator|CombineRotators (Math|Rotator|MakeRotator (.x (LocalVariables|GetrotationAttrib)) (.y (LocalVariables|GetrotationAttrib)) (.z (LocalVariables|GetrotationAttrib))) (LocalVariables|GetInRotation)) (Variables|Default|GetAbsoluteRotation)) (LocalVariables|GetInRotation) (Variables|Default|GetTransformRotation)) (Math|Vector|SelectVector (Math|Vector|SelectVector (LocalVariables|GetscaleAttrib) (* (LocalVariables|GetInScale) (LocalVariables|GetscaleAttrib)) (Variables|Default|GetAbsoluteScale)) (LocalVariables|GetInScale) (Variables|Default|GetTransformScale)))))
      (return _structout true))))
