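; Deprecated Blueprint PCG element that randomizes the transform of points, using each point's random stream.
; ExecuteWithContext converts only the first spatial input to point data, runs PointLoop and copies the target actor to the output.
; PointLoopBody draws offset, rotation and scale within Min/Max ranges.
; Boolean variables choose world-space or point-local offset, and whether rotation and scale replace or combine with existing values.
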(fn ExecuteWithContext
  (params (InContext :type PCGContext%20Structure%20%28by%20ref%29) (Input :type PCGData%20Collection%20Structure%20%28by%20ref%29))
  (bind _returnvalue (Data|GetAllSpatialInputs Input))
  (if (> (Utilities|Array|Length _returnvalue) 0)
    (bind _aspcgspatial_data (Utilities|Casting|CastToPCGSpatialData (PCG|TaggedData|BreakTaggedData (Utilities|Array|Get(aref) _returnvalue))))
    (bind _outdata (PCG|FlowControl|PointLoop InContext (SpatialData|ToPointDatawithContext _aspcgspatial_data InContext) 0))
    (Variables|/Script/PCG.PCGSpatialData|SetTargetActor (Variables|/Script/PCG.PCGSpatialData|GetTargetActor _aspcgspatial_data) _outdata)
    (return (Utilities|Struct|MakePCGDataCollection (Utilities|Array|MakeArray (PCG|TaggedData|MakeTaggedData _outdata))))
    (else
      (return 0))))

(fn PointLoopBody
  (params (InContext :type PCGContext%20Structure%20%28by%20ref%29) (InData :type PCGPoint%20Data%20Object%20Reference) (InPoint :type PCGPoint%20Structure%20%28by%20ref%29) (OutMetadata :type PCGMetadata%20Object%20Reference) (Iteration :type Integer64))
  (bind _returnvalue (PCG|Helpers|GetRandomStreamfromPoint InPoint))
  (bind _structout (Utilities|Struct|SetmembersinPCGPoint InPoint (Math|Transform|MakeTransform (Math|Vector|SelectVector (+ (.location (Utilities|Struct|BreakPCGPoint InPoint)) (Math|Vector|MakeVector (Math|Random|RandomFloatinRangefromStream _returnvalue (.x (Variables|Default|GetOffsetMin)) (.x (Variables|Default|GetOffsetMax))) (Math|Random|RandomFloatinRangefromStream _returnvalue (.y (Variables|Default|GetOffsetMin)) (.y (Variables|Default|GetOffsetMax))) (Math|Random|RandomFloatinRangefromStream _returnvalue (.z (Variables|Default|GetOffsetMin)) (.z (Variables|Default|GetOffsetMax))))) (+ (.location (Utilities|Struct|BreakPCGPoint InPoint)) (Math|Matrix|TransformPosition(Matrix) (Math|Transform|ToMatrix(Transform) (Math|Conversions|ToTransform(Rotator) (.rotation (Utilities|Struct|BreakPCGPoint InPoint)))) (Math|Vector|MakeVector (Math|Random|RandomFloatinRangefromStream _returnvalue (.x (Variables|Default|GetOffsetMin)) (.x (Variables|Default|GetOffsetMax))) (Math|Random|RandomFloatinRangefromStream _returnvalue (.y (Variables|Default|GetOffsetMin)) (.y (Variables|Default|GetOffsetMax))) (Math|Random|RandomFloatinRangefromStream _returnvalue (.z (Variables|Default|GetOffsetMin)) (.z (Variables|Default|GetOffsetMax)))))) (Variables|Default|GetWorldSpaceTranslation)) (Math|Rotator|SelectRotator (Math|Rotator|MakeRotator (Math|Random|RandomFloatinRangefromStream _returnvalue (.x (Variables|Default|GetRotationMin)) (.x (Variables|Default|GetRotationMax))) (Math|Random|RandomFloatinRangefromStream _returnvalue (.y (Variables|Default|GetRotationMin)) (.y (Variables|Default|GetRotationMax))) (Math|Random|RandomFloatinRangefromStream _returnvalue (.z (Variables|Default|GetRotationMin)) (.z (Variables|Default|GetRotationMax)))) (Math|Rotator|CombineRotators (Math|Rotator|MakeRotator (Math|Random|RandomFloatinRangefromStream _returnvalue (.x (Variables|Default|GetRotationMin)) (.x (Variables|Default|GetRotationMax))) (Math|Random|RandomFloatinRangefromStream _returnvalue (.y (Variables|Default|GetRotationMin)) (.y (Variables|Default|GetRotationMax))) (Math|Random|RandomFloatinRangefromStream _returnvalue (.z (Variables|Default|GetRotationMin)) (.z (Variables|Default|GetRotationMax)))) (.rotation (Utilities|Struct|BreakPCGPoint InPoint))) (Variables|Default|GetAbsoluteRotation)) (Math|Vector|SelectVector (Math|Vector|MakeVector (Math|Random|RandomFloatinRangefromStream _returnvalue (.x (Variables|Default|GetScaleMin)) (.x (Variables|Default|GetScaleMax))) (Math|Random|RandomFloatinRangefromStream _returnvalue (.y (Variables|Default|GetScaleMin)) (.y (Variables|Default|GetScaleMax))) (Math|Random|RandomFloatinRangefromStream _returnvalue (.z (Variables|Default|GetScaleMin)) (.z (Variables|Default|GetScaleMax)))) (* (.scale (Utilities|Struct|BreakPCGPoint InPoint)) (Math|Vector|MakeVector (Math|Random|RandomFloatinRangefromStream _returnvalue (.x (Variables|Default|GetScaleMin)) (.x (Variables|Default|GetScaleMax))) (Math|Random|RandomFloatinRangefromStream _returnvalue (.y (Variables|Default|GetScaleMin)) (.y (Variables|Default|GetScaleMax))) (Math|Random|RandomFloatinRangefromStream _returnvalue (.z (Variables|Default|GetScaleMin)) (.z (Variables|Default|GetScaleMax))))) (Variables|Default|GetAbsoluteScale)))))
  (return _structout true))
