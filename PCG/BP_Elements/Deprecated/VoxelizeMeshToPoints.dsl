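; Deprecated PCG Blueprint element, VoxelizeMeshToPoints, kept as a marker that points to the Mesh Sampler node as its replacement.
; ExecuteWithContext shows no logic, so no voxelization or point generation happens here.
; Node color and title overrides turn the node red and display the deprecation notice in its title.
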(fn ExecuteWithContext
  (params (InContext :type PCGContext%20Structure%20%28by%20ref%29) (Input :type PCGData%20Collection%20Structure%20%28by%20ref%29)))

(fn NodeColorOverride
  (return "(R=1.000000,G=0.000000,B=0.000000,A=1.000000)"))

(fn NodeTitleOverride
  (return "DEPRECATED: Replace with Mesh Sampler Node"))
