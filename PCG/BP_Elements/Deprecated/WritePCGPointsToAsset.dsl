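; Deprecated PCG Blueprint element that copies the points of its spatial input into a referenced PCGPointList asset.
; ExecuteWithContext takes the first spatial input, converts it to point data with the execution context, then clears the asset's point array and refills it.
; It returns all spatial inputs as the output collection, and contains no explicit step to dirty or save the asset.
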(fn ExecuteWithContext
  (params (InContext :type PCGContext%20Structure%20%28by%20ref%29) (Input :type PCGData%20Collection%20Structure%20%28by%20ref%29))
  (bind _pcgpointsasset (Variables|Default|GetPCGPointsAsset))
  (bind _pcgpoints (Variables|/PCG/BP_Elements/Resources/PCGPointList.PCGPointList_C|GetPCGPoints _pcgpointsasset))
  (bind _returnvalue (Data|GetAllSpatialInputs Input))
  (bind _aspcgspatial_data (Utilities|Casting|CastToPCGSpatialData (PCG|TaggedData|BreakTaggedData (Utilities|Array|Get(aref) _returnvalue))))
  (Utilities|Array|Clear _pcgpoints)
  (Utilities|Array|AppendArray _pcgpoints (SpatialData|GetPoints (SpatialData|ToPointDatawithContext _aspcgspatial_data InContext)))
  (return (Utilities|Struct|MakePCGDataCollection _returnvalue)))
